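# Niagara system for the fuse effect on the female Crush player body cosmetic, placing orange-tinted mesh and sprite particles on the character's skeletal mesh.
# Two emitters spawn continuously and a third fires a burst of larger, short-lived particles, the mesh-only one being continuous.
# Ten spawn groups index the bones or sockets that particles follow every update.
# A local Part_HiddenCheck maps spawn group and hand/foot hidden flags to VisibilityTag, masking particles on hidden parts.
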NiagaraSystem /BRCosmetics/Characters/Player/Female/Medium/Bodies/F_MED_Crush/FX/NS_Crush_Fuse
dependencies {
 usedDataInterfaces [&NiagaraDataInterfaceColorCurve, &NiagaraDataInterfaceArrayInt32]
 usedDynamicInputs []
 usedModules [&SystemState]
}
emitters {
 FX_Fuse_Hand_Modle {
  emitterUpdateScript [
   &EmitterState(Life Cycle Mode=::NewEnumerator1, Loop Duration=1, UseLoopDelay=false)
   &SpawnRate(SpawnRate=10, Use Spawn Probability=false)
  ]
  particleSpawnScript [
   /Niagara/Modules/Spawn/Initialization/InitializeParticle(
    Color=<rgba 1 0.5316 0.0955 0.018>
    Lifetime=&RandomRangeFloat(Minimum=2, Maximum=2.5)
    Mesh Scale=&VectorFromFloat(Value=&Multiply_Float(A=0.01, B=&RandomRangeFloat(Minimum=1.2, Maximum=2)))
    Module.Write Color=true
    Module.Write Lifetime=true
    Module.Write Mass=true
    Module.Write Material Random=true
    Module.Write Position=false
    Module.Write Scale=true
    Module.Write SpriteRotation=true
    Module.Write SpriteSize=true
    Position=&SimulationPosition()
    Sprite Rotation=&RandomRangeFloat(Minimum=-1, Maximum=1)
    Sprite Size=&Vector2DFromFloat(Value=7)
   )
   &InitialMeshOrientation(
    Rotation=&RandomRangeVector(Minimum=<0 0 0>, Maximum=<1 1 1>)
    Use Orientation Vector=true
    Use Rotation Vector=true
   )
   &FN_SetSpawnGroup(Number of Spawn Groups=10)
   /SetVariables(Particles.VisibilityTag=&MakeCustomIntFromBool(
     Bool=$User.bisFootHidden
     True INT=&SelectIntFromArray(
      Direct Array Index=&SpawnGroup()
      Int Selection Array=DataInterface() {propertyValues {IntData [0,0,0,0,1,1,1,1,1,1]}}
     )
     False INT=0
    )) {isSetParameters=true}
  ]
  particleUpdateScript [
   &ParticleState()
   &SkeletalMeshLocation(
    Bone / Socket Index=&SpawnGroup()
    Inherit Velocity=false
    Interpolate Mesh Sampling=false
    MaskSpawnGroup=false
    Mesh Sampling Type=::NewEnumerator5
    Sampled Position Offset Coordinate Space:ENiagaraMeshTransforms=::NewEnumerator2
    Skeletal Mesh=$User.SkeletalMesh
    Socket Sampling Mode=::NewEnumerator5
   )
   &SolveForcesAndVelocity(Clamp Velocity=false, Limit Acceleration=false)
   &FN_Light_Attributes(
    Color=&ColorFromCurve(ColorCurve=DataInterface() {propertyValues {
       AlphaCurve {keys [<at 0 0 RCIM_Cubic>,<at 0.279 1>,<at 1 0 RCIM_Cubic>]}
       BlueCurve {keys [<at 0 0 RCIM_Cubic>,<at 0.279 0>,<at 0.5266 0>,<at 1.0077 0.0776>]}
       GreenCurve {keys [<at 0 0.2118 RCIM_Cubic>,<at 0.279 0.3775>,<at 0.5266 0.5161>,<at 1.0077 0.4904>]}
       RedCurve {keys [<at 0 1 RCIM_Cubic>,<at 0.279 1>,<at 0.5266 1>,<at 1.0077 1>]}
      }})
    Create Color Attribute=true
    Create Exponent Attribute=true
    Create Radius Attribute=true
    Radius=20
    Scale Color=1.2
   )
   /SetVariables(Particles.VisibilityTag=&self:Part_HiddenCheck(is Hand Hidden:NiagaraBool=$User.bisHandHidden, is Foot Hidden:NiagaraBool=$User.bisFootHidden, Selection:NiagaraInt32=$Particles.SpawnGroup)) {isSetParameters=true, module=SetVariables_#1}
  ]
  rendererClasses [&NiagaraMeshRendererProperties]
  renderers [{rendererClass=&NiagaraMeshRendererProperties, rendererIndex=0}]
 }
 FX_Fuse_L_Hand001 {
  emitterUpdateScript [
   &EmitterState(Life Cycle Mode=::NewEnumerator1, Loop Duration=1, UseLoopDelay=false)
   &SpawnRate(SpawnRate=10, Use Spawn Probability=false)
  ]
  particleSpawnScript [
   &InitializeParticle(
    Color=<rgba 1 0.5316 0.0955 0.018>
    Lifetime=&RandomRangeFloat(Minimum=2, Maximum=2.5)
    Mesh Scale=&VectorFromFloat(Value=&Multiply_Float(A=0.01, B=&RandomRangeFloat(Minimum=1.2, Maximum=2)))
    Module.Write Color=true
    Module.Write Lifetime=true
    Module.Write Mass=true
    Module.Write Material Random=true
    Module.Write Position=false
    Module.Write Scale=true
    Module.Write SpriteRotation=true
    Module.Write SpriteSize=true
    Position=&SimulationPosition()
    Sprite Rotation=&RandomRangeFloat(Minimum=-1, Maximum=1)
    Sprite Size=&Vector2DFromFloat(Value=7)
   )
   &InitialMeshOrientation(
    Rotation=&RandomRangeVector(Minimum=<0 0 0>, Maximum=<1 1 1>)
    Use Orientation Vector=true
    Use Rotation Vector=true
   )
   &FN_SetSpawnGroup(Number of Spawn Groups=10)
  ]
  particleUpdateScript [
   &ParticleState()
   &SkeletalMeshLocation(
    Bone / Socket Index=&SpawnGroup()
    Inherit Velocity=false
    Interpolate Mesh Sampling=false
    MaskSpawnGroup=false
    Mesh Sampling Type=::NewEnumerator5
    Sampled Position Offset Coordinate Space:ENiagaraMeshTransforms=::NewEnumerator2
    Skeletal Mesh=$User.SkeletalMesh
    Socket Sampling Mode=::NewEnumerator5
   )
   &SolveForcesAndVelocity(Clamp Velocity=false, Limit Acceleration=false)
   &FN_Light_Attributes(
    Color=&ColorFromCurve(ColorCurve=DataInterface() {propertyValues {
       AlphaCurve {keys [<at 0 0 RCIM_Cubic>,<at 0.279 1>,<at 1 0 RCIM_Cubic>]}
       BlueCurve {keys [<at 0 0 RCIM_Cubic>,<at 0.279 0>,<at 0.5266 0>,<at 1.0077 0.0776>]}
       GreenCurve {keys [<at 0 0.2118 RCIM_Cubic>,<at 0.279 0.3775>,<at 0.5266 0.5161>,<at 1.0077 0.4904>]}
       RedCurve {keys [<at 0 1 RCIM_Cubic>,<at 0.279 1>,<at 0.5266 1>,<at 1.0077 1>]}
      }})
    Create Color Attribute=true
    Create Exponent Attribute=true
    Create Radius Attribute=true
    Radius=20
    Scale Color=1.2
   )
   /SetVariables(Particles.VisibilityTag=&self:Part_HiddenCheck(is Hand Hidden:NiagaraBool=$User.bisHandHidden, is Foot Hidden:NiagaraBool=$User.bisFootHidden, Selection:NiagaraInt32=$Particles.SpawnGroup)) {isSetParameters=true, module=SetVariables_#1}
  ]
  rendererClasses [&NiagaraMeshRendererProperties, &NiagaraSpriteRendererProperties]
  renderers [
   {rendererClass=&NiagaraMeshRendererProperties, rendererIndex=0}
   {rendererClass=&NiagaraSpriteRendererProperties, rendererIndex=1}
  ]
 }
 FX_Fuse_L_Hand002 {
  emitterUpdateScript [
   &EmitterState(Life Cycle Mode=::NewEnumerator1, Loop Behavior=::NewEnumerator1, Loop Duration=1, UseLoopDelay=false)
   &SpawnBurst_Instantaneous(Spawn Count=10, Use Spawn Probability=false) {module=SpawnBurst_Instantaneous001}
  ]
  particleSpawnScript [
   &InitializeParticle(
    Color=<rgba 1 0.5316 0.0955 0.018>
    Lifetime=&RandomRangeFloat(Minimum=0.2, Maximum=0.45)
    Mesh Scale=&VectorFromFloat(Value=&Multiply_Float(A=0.03, B=&RandomRangeFloat(Minimum=1.2, Maximum=2)))
    Module.Write Color=true
    Module.Write Lifetime=true
    Module.Write Mass=true
    Module.Write Material Random=true
    Module.Write Position=false
    Module.Write Scale=true
    Module.Write SpriteRotation=true
    Module.Write SpriteSize=true
    Position=&SimulationPosition()
    Sprite Rotation=&RandomRangeFloat(Minimum=-1, Maximum=1)
    Sprite Size=&Vector2DFromFloat(Value=7)
   )
   &InitialMeshOrientation(
    Rotation=&RandomRangeVector(Minimum=<0 0 0>, Maximum=<1 1 1>)
    Use Orientation Vector=true
    Use Rotation Vector=true
   )
   &FN_SetSpawnGroup(Number of Spawn Groups=10)
   /SetVariables(Particles.VisibilityTag=&MakeCustomIntFromBool(
     Bool=$User.bisFootHidden
     True INT=&SelectIntFromArray(
      Direct Array Index=&SpawnGroup()
      Int Selection Array=DataInterface() {propertyValues {IntData [0,0,0,0,1,1,1,1,1,1]}}
     )
     False INT=0
    )) {isSetParameters=true}
  ]
  particleUpdateScript [
   &ParticleState()
   &SkeletalMeshLocation(
    Bone / Socket Index=&SpawnGroup()
    Inherit Velocity=false
    Interpolate Mesh Sampling=false
    MaskSpawnGroup=false
    Mesh Sampling Type=::NewEnumerator5
    Sampled Position Offset Coordinate Space:ENiagaraMeshTransforms=::NewEnumerator2
    Skeletal Mesh=$User.SkeletalMesh
    Socket Sampling Mode=::NewEnumerator5
   )
   &SolveForcesAndVelocity(Clamp Velocity=false, Limit Acceleration=false)
   &FN_Light_Attributes(
    Color=&ColorFromCurve(ColorCurve=DataInterface() {propertyValues {
       AlphaCurve {keys [<at 0 0 RCIM_Cubic>,<at 0.279 1>,<at 1 0 RCIM_Cubic>]}
       BlueCurve {keys [<at 0 0 RCIM_Cubic>,<at 0.279 0>,<at 0.5266 0>,<at 1.0077 0.0776>]}
       GreenCurve {keys [<at 0 0.2118 RCIM_Cubic>,<at 0.279 0.3775>,<at 0.5266 0.5161>,<at 1.0077 0.4904>]}
       RedCurve {keys [<at 0 1 RCIM_Cubic>,<at 0.279 1>,<at 0.5266 1>,<at 1.0077 1>]}
      }})
    Create Color Attribute=true
    Create Exponent Attribute=true
    Create Radius Attribute=true
    Radius=20
    Scale Color=1.2
   )
   /SetVariables(Particles.VisibilityTag=&self:Part_HiddenCheck(is Hand Hidden:NiagaraBool=$User.bisHandHidden, is Foot Hidden:NiagaraBool=$User.bisFootHidden, Selection:NiagaraInt32=$Particles.SpawnGroup)) {isSetParameters=true, module=SetVariables_#1}
  ]
  rendererClasses [&NiagaraMeshRendererProperties, &NiagaraSpriteRendererProperties]
  renderers [
   {rendererClass=&NiagaraMeshRendererProperties, rendererIndex=0}
   {rendererClass=&NiagaraSpriteRendererProperties, rendererIndex=1}
  ]
 }
}
summary {
 systemName=NS_Crush_Fuse
 emitters [
  {
   emitterName=FX_Fuse_L_Hand001
   rendererClasses [&NiagaraMeshRendererProperties, &NiagaraSpriteRendererProperties]
  }
  {
   emitterName=FX_Fuse_L_Hand002
   rendererClasses [&NiagaraMeshRendererProperties, &NiagaraSpriteRendererProperties]
  }
  {emitterName=FX_Fuse_Hand_Modle, rendererClasses [&NiagaraMeshRendererProperties]}
 ]
 userVariables [
  User.bisFootHidden {type=NiagaraBool}
  User.bisHandHidden {type=NiagaraBool}
  User.SkeletalMesh {
   type=NiagaraDataInterfaceSkeletalMesh
   defaultValue {
    struct=NiagaraExt_VariableValue_DataInterface
    value {dataInterface=&self:NiagaraDataInterfaceSkeletalMesh_0}
   }
  }
 ]
}
