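# Physics asset for the F_MED_Crush female body's skirt and accessories, defining their secondary-motion bodies and joints.
# Twelve three-capsule skirt chains ring the hips with limited swing and twist, each hung from a tiny MSTR anchor constrained to the pelvis.
# A three-box back string hangs from spine_03, and belt-gun and hip-heart boxes hang from the pelvis.
# Pelvis, spine and thigh capsules supply the body shapes, linked by free-rotation constraints.
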PhysicsAsset /BRCosmetics/Characters/Player/Female/Medium/Bodies/F_MED_Crush/Meshes/F_MED_Crush_Skirt_FE_Physics
bodies {
 dyn_backString_01 [{shapeType=Box, extentX=2, extentY=4.9314, extentZ=1.0927, center=<0.8202 0 0>}]
 dyn_backString_02 [{shapeType=Box, extentX=2, extentY=4.9314, extentZ=1.0927, center=<0.8202 0 0>}]
 dyn_backString_03 [{shapeType=Box, extentX=2, extentY=4.9314, extentZ=1.0927, center=<0.8202 0 0>}]
 dyn_backString_MSTR [{shapeType=Capsule, length=0.505, radius=0.505}]
 dyn_belt_gun_01 [{shapeType=Box, extentX=15.1603, extentY=16.5509, extentZ=1.3033, center=<8.4555 -0.0002 3.2218>, rotation=<rot 0 0 -4.2015>}]
 dyn_hipHeart [{shapeType=Box, extentX=17.1269, extentY=16.2409, extentZ=2.6086, center=<0 0 1.4564>}]
 dyn_skirt_01_l [{shapeType=Capsule, length=5.9805, radius=2, shapeName=dyn_skirt_fr_01_capsule, center=<4.6779 0 0.8139>, rotation=<rot 90 0 0>}]
 dyn_skirt_01_l_MSTR [{shapeType=Capsule, length=0.505, radius=0.505}]
 dyn_skirt_01_r [{shapeType=Capsule, length=5.9805, radius=2, shapeName=dyn_skirt_fr_01_capsule, center=<4.6779 0 0.8139>, rotation=<rot 90 0 0>}]
 dyn_skirt_01_r_MSTR [{shapeType=Capsule, length=0.505, radius=0.505}]
 dyn_skirt_02_l [{shapeType=Capsule, length=5.9805, radius=2, shapeName=dyn_skirt_fr_01_capsule, center=<4.6779 0 0.7108>, rotation=<rot 90 0 0>}]
 dyn_skirt_02_r [{shapeType=Capsule, length=5.9805, radius=2, shapeName=dyn_skirt_fr_01_capsule, center=<4.6779 0 0.688>, rotation=<rot 90 0 0>}]
 dyn_skirt_03_l [{shapeType=Capsule, length=5.9805, radius=2, shapeName=dyn_skirt_fr_01_capsule, center=<4.6779 0 -0.271>, rotation=<rot 90 0 0>}]
 dyn_skirt_03_r [{shapeType=Capsule, length=5.9805, radius=2, shapeName=dyn_skirt_fr_01_capsule, center=<4.6779 0 0.8139>, rotation=<rot 90 0 0>}]
 dyn_skirt_bk2_01_l [{shapeType=Capsule, length=5.9805, radius=2, shapeName=dyn_skirt_fr_01_capsule, center=<4.6779 0 0.8139>, rotation=<rot 90 0 0>}]
 dyn_skirt_bk2_01_l_MSTR [{shapeType=Capsule, length=0.505, radius=0.505}]
 dyn_skirt_bk2_01_r [{shapeType=Capsule, length=5.9805, radius=2, shapeName=dyn_skirt_fr_01_capsule, center=<4.6779 0 0.8139>, rotation=<rot 90 0 0>}]
 dyn_skirt_bk2_01_r_MSTR [{shapeType=Capsule, length=0.505, radius=0.505}]
 dyn_skirt_bk2_02_l [{shapeType=Capsule, length=5.9805, radius=2, shapeName=dyn_skirt_fr_01_capsule, center=<4.6779 0 0.8139>, rotation=<rot 90 0 0>}]
 dyn_skirt_bk2_02_r [{shapeType=Capsule, length=5.9805, radius=2, shapeName=dyn_skirt_fr_01_capsule, center=<4.6779 0 0.8139>, rotation=<rot 90 0 0>}]
 dyn_skirt_bk2_03_l [{shapeType=Capsule, length=5.9805, radius=2, shapeName=dyn_skirt_fr_01_capsule, center=<4.6779 0 0.8139>, rotation=<rot 90 0 0>}]
 dyn_skirt_bk2_03_r [{shapeType=Capsule, length=5.9805, radius=2, shapeName=dyn_skirt_fr_01_capsule, center=<4.6779 0 0.8139>, rotation=<rot 90 0 0>}]
 dyn_skirt_bk_01 [{shapeType=Capsule, length=5.9805, radius=2, shapeName=dyn_skirt_fr_01_capsule, center=<4.6779 0 0.8139>, rotation=<rot 90 0 0>}]
 dyn_skirt_bk_01_MSTR [{shapeType=Capsule, length=0.505, radius=0.505}]
 dyn_skirt_bk_01_l [{shapeType=Capsule, length=5.9805, radius=2, shapeName=dyn_skirt_fr_01_capsule, center=<4.6779 0 0.8139>, rotation=<rot 90 0 0>}]
 dyn_skirt_bk_01_l_MSTR [{shapeType=Capsule, length=0.505, radius=0.505}]
 dyn_skirt_bk_01_r [{shapeType=Capsule, length=5.9805, radius=2, shapeName=dyn_skirt_fr_01_capsule, center=<4.6779 0 0.8139>, rotation=<rot 90 0 0>}]
 dyn_skirt_bk_01_r_MSTR [{shapeType=Capsule, length=0.505, radius=0.505}]
 dyn_skirt_bk_02 [{shapeType=Capsule, length=5.9805, radius=2, shapeName=dyn_skirt_fr_01_capsule, center=<4.6779 0 0.8139>, rotation=<rot 90 0 0>}]
 dyn_skirt_bk_02_l [{shapeType=Capsule, length=5.9805, radius=2, shapeName=dyn_skirt_fr_01_capsule, center=<4.6779 0 0.8139>, rotation=<rot 90 0 0>}]
 dyn_skirt_bk_02_r [{shapeType=Capsule, length=5.9805, radius=2, shapeName=dyn_skirt_fr_01_capsule, center=<4.6779 0 0.8139>, rotation=<rot 90 0 0>}]
 dyn_skirt_bk_03 [{shapeType=Capsule, length=5.9805, radius=2, shapeName=dyn_skirt_fr_01_capsule, center=<4.6779 0 0.8139>, rotation=<rot 90 0 0>}]
 dyn_skirt_bk_03_l [{shapeType=Capsule, length=5.9805, radius=2, shapeName=dyn_skirt_fr_01_capsule, center=<4.6779 0 0.8139>, rotation=<rot 90 0 0>}]
 dyn_skirt_bk_03_r [{shapeType=Capsule, length=5.9805, radius=2, shapeName=dyn_skirt_fr_01_capsule, center=<4.6779 0 0.8139>, rotation=<rot 90 0 0>}]
 dyn_skirt_fr2_01_l [{shapeType=Capsule, length=5.9805, radius=2, shapeName=dyn_skirt_fr_01_capsule, center=<4.6779 0 0.8139>, rotation=<rot 90 0 0>}]
 dyn_skirt_fr2_01_l_MSTR [{shapeType=Capsule, length=0.505, radius=0.505}]
 dyn_skirt_fr2_01_r [{shapeType=Capsule, length=5.9805, radius=2, shapeName=dyn_skirt_fr_01_capsule, center=<4.6779 0 0.8139>, rotation=<rot 90 0 0>}]
 dyn_skirt_fr2_01_r_MSTR [{shapeType=Capsule, length=0.505, radius=0.505}]
 dyn_skirt_fr2_02_l [{shapeType=Capsule, length=5.9805, radius=2, shapeName=dyn_skirt_fr_01_capsule, center=<4.6779 0 0.2188>, rotation=<rot 90 0 0>}]
 dyn_skirt_fr2_02_r [{shapeType=Capsule, length=5.9805, radius=2, shapeName=dyn_skirt_fr_01_capsule, center=<4.6779 0 0.8139>, rotation=<rot 90 0 0>}]
 dyn_skirt_fr2_03_l [{shapeType=Capsule, length=5.9805, radius=2, shapeName=dyn_skirt_fr_01_capsule, center=<4.6779 0 0.8139>, rotation=<rot 90 0 0>}]
 dyn_skirt_fr2_03_r [{shapeType=Capsule, length=5.9805, radius=2, shapeName=dyn_skirt_fr_01_capsule, center=<4.6779 0 0.8139>, rotation=<rot 90 0 0>}]
 dyn_skirt_fr_01 [{shapeType=Capsule, length=5.9805, radius=2, shapeName=dyn_skirt_fr_01_capsule, center=<5.6085 0 0.8139>, rotation=<rot 90 0 0>}]
 dyn_skirt_fr_01_MSTR [{shapeType=Capsule, length=0.505, radius=0.505}]
 dyn_skirt_fr_01_l [{shapeType=Capsule, length=5.9805, radius=2, shapeName=dyn_skirt_fr_01_capsule, center=<4.6779 0 0.8139>, rotation=<rot 90 0 0>}]
 dyn_skirt_fr_01_l_MSTR [{shapeType=Capsule, length=0.505, radius=0.505}]
 dyn_skirt_fr_01_r [{shapeType=Capsule, length=5.9805, radius=2, shapeName=dyn_skirt_fr_01_capsule, center=<4.6779 0 0.8139>, rotation=<rot 90 0 0>}]
 dyn_skirt_fr_01_r_MSTR [{shapeType=Capsule, length=0.505, radius=0.505}]
 dyn_skirt_fr_02 [{shapeType=Capsule, length=5.9805, radius=2, shapeName=dyn_skirt_fr_01_capsule, center=<5.6085 0 0.7037>, rotation=<rot 90 0 0>}]
 dyn_skirt_fr_02_l [{shapeType=Capsule, length=5.9805, radius=2, shapeName=dyn_skirt_fr_01_capsule, center=<4.6779 0 0.8139>, rotation=<rot 90 0 0>}]
 dyn_skirt_fr_02_r [{shapeType=Capsule, length=5.9805, radius=2, shapeName=dyn_skirt_fr_01_capsule, center=<4.6779 0 0.6246>, rotation=<rot 90 0 0>}]
 dyn_skirt_fr_03 [{shapeType=Capsule, length=5.9805, radius=2, shapeName=dyn_skirt_fr_01_capsule, center=<4.6779 0 0.8139>, rotation=<rot 90 0 0>}]
 dyn_skirt_fr_03_l [{shapeType=Capsule, length=5.9805, radius=2, shapeName=dyn_skirt_fr_01_capsule, center=<4.6779 0 0.8139>, rotation=<rot 90 0 0>}]
 dyn_skirt_fr_03_r [{shapeType=Capsule, length=5.9805, radius=2, shapeName=dyn_skirt_fr_01_capsule, center=<4.6779 0 0.8139>, rotation=<rot 90 0 0>}]
 pelvis [{shapeType=Capsule, length=7.458, radius=8.0636, center=<1.1549 -1.274 -0.4831>}]
 spine_02 [{shapeType=Capsule, length=8.1468, radius=6.2361}]
 spine_03 [{shapeType=Capsule, length=8.1468, radius=6.2361}]
 thigh_l [{shapeType=Capsule, length=33.0077, radius=7.051, center=<-25.1663 0.0001 0>, rotation=<rot 90 -174.9489 5.0512>}]
 thigh_r [{shapeType=Capsule, length=33.3529, radius=7.4973, center=<25.1663 -0.0001 0>, rotation=<rot 90 7.5524 7.5525>}]
}
constraints [
 {
  bone1Name=thigh_l
  bone2Name=pelvis
  swing1LimitDegrees=45
  swing1Motion=Free
  swing2LimitDegrees=45
  swing2Motion=Free
  twistLimitDegrees=45
  twistMotion=Free
 }
 {
  bone1Name=thigh_r
  bone2Name=pelvis
  swing1LimitDegrees=45
  swing1Motion=Free
  swing2LimitDegrees=45
  swing2Motion=Free
  twistLimitDegrees=45
  twistMotion=Free
 }
 {bone1Name=dyn_skirt_fr_01_l, bone2Name=dyn_skirt_fr_01_l_MSTR, swing1LimitDegrees=5, swing2LimitDegrees=20, twistLimitDegrees=5}
 {bone1Name=dyn_skirt_fr_02_l, bone2Name=dyn_skirt_fr_01_l, swing1LimitDegrees=5, swing2LimitDegrees=20, twistLimitDegrees=5}
 {bone1Name=dyn_skirt_fr_03_l, bone2Name=dyn_skirt_fr_02_l, swing1LimitDegrees=5, swing2LimitDegrees=20, twistLimitDegrees=5}
 {bone1Name=dyn_skirt_01_l, bone2Name=dyn_skirt_01_l_MSTR, swing1LimitDegrees=25, swing2LimitDegrees=20, twistLimitDegrees=5}
 {bone1Name=dyn_skirt_02_l, bone2Name=dyn_skirt_01_l, swing1LimitDegrees=25, swing2LimitDegrees=20, twistLimitDegrees=5}
 {bone1Name=dyn_skirt_03_l, bone2Name=dyn_skirt_02_l, swing1LimitDegrees=25, swing2LimitDegrees=20, twistLimitDegrees=5}
 {bone1Name=dyn_skirt_bk_01_l, bone2Name=dyn_skirt_bk_01_l_MSTR, swing1LimitDegrees=25, swing2LimitDegrees=20, twistLimitDegrees=5}
 {bone1Name=dyn_skirt_bk_02_l, bone2Name=dyn_skirt_bk_01_l, swing1LimitDegrees=25, swing2LimitDegrees=20, twistLimitDegrees=5}
 {bone1Name=dyn_skirt_bk_03_l, bone2Name=dyn_skirt_bk_02_l, swing1LimitDegrees=25, swing2LimitDegrees=20, twistLimitDegrees=5}
 {bone1Name=dyn_skirt_fr2_01_l, bone2Name=dyn_skirt_fr2_01_l_MSTR, swing1LimitDegrees=25, swing2LimitDegrees=20, twistLimitDegrees=5}
 {bone1Name=dyn_skirt_fr2_02_l, bone2Name=dyn_skirt_fr2_01_l, swing1LimitDegrees=15, swing2LimitDegrees=20, twistLimitDegrees=5}
 {bone1Name=dyn_skirt_fr2_03_l, bone2Name=dyn_skirt_fr2_02_l, swing1LimitDegrees=15, swing2LimitDegrees=20, twistLimitDegrees=5}
 {bone1Name=dyn_skirt_bk2_01_l, bone2Name=dyn_skirt_bk2_01_l_MSTR, swing1LimitDegrees=25, swing2LimitDegrees=20, twistLimitDegrees=5}
 {bone1Name=dyn_skirt_bk2_02_l, bone2Name=dyn_skirt_bk2_01_l, swing1LimitDegrees=25, swing2LimitDegrees=20, twistLimitDegrees=5}
 {bone1Name=dyn_skirt_bk2_03_l, bone2Name=dyn_skirt_bk2_02_l, swing1LimitDegrees=25, swing2LimitDegrees=20, twistLimitDegrees=5}
 {bone1Name=dyn_skirt_fr_01_r, bone2Name=dyn_skirt_fr_01_r_MSTR, swing1LimitDegrees=5, swing2LimitDegrees=20, twistLimitDegrees=5}
 {bone1Name=dyn_skirt_fr_02_r, bone2Name=dyn_skirt_fr_01_r, swing1LimitDegrees=5, swing2LimitDegrees=20, twistLimitDegrees=5}
 {bone1Name=dyn_skirt_fr_03_r, bone2Name=dyn_skirt_fr_02_r, swing1LimitDegrees=5, swing2LimitDegrees=20, twistLimitDegrees=5}
 {bone1Name=dyn_skirt_01_r, bone2Name=dyn_skirt_01_r_MSTR, swing1LimitDegrees=25, swing2LimitDegrees=20, twistLimitDegrees=5}
 {bone1Name=dyn_skirt_02_r, bone2Name=dyn_skirt_01_r, swing1LimitDegrees=25, swing2LimitDegrees=20, twistLimitDegrees=5}
 {bone1Name=dyn_skirt_03_r, bone2Name=dyn_skirt_02_r, swing1LimitDegrees=25, swing2LimitDegrees=20, twistLimitDegrees=5}
 {bone1Name=dyn_skirt_bk_01_r, bone2Name=dyn_skirt_bk_01_r_MSTR, swing1LimitDegrees=25, swing2LimitDegrees=20, twistLimitDegrees=5}
 {bone1Name=dyn_skirt_bk_03_r, bone2Name=dyn_skirt_bk_02_r, swing1LimitDegrees=25, swing2LimitDegrees=20, twistLimitDegrees=5}
 {bone1Name=dyn_skirt_bk_02_r, bone2Name=dyn_skirt_bk_01_r, swing1LimitDegrees=25, swing2LimitDegrees=20, twistLimitDegrees=5}
 {bone1Name=dyn_skirt_fr2_01_r, bone2Name=dyn_skirt_fr2_01_r_MSTR, swing1LimitDegrees=25, swing2LimitDegrees=20, twistLimitDegrees=5}
 {bone1Name=dyn_skirt_fr2_02_r, bone2Name=dyn_skirt_fr2_01_r, swing1LimitDegrees=25, swing2LimitDegrees=20, twistLimitDegrees=5}
 {bone1Name=dyn_skirt_fr2_03_r, bone2Name=dyn_skirt_fr2_02_r, swing1LimitDegrees=25, swing2LimitDegrees=20, twistLimitDegrees=5}
 {bone1Name=dyn_skirt_bk2_01_r, bone2Name=dyn_skirt_bk2_01_r_MSTR, swing1LimitDegrees=25, swing2LimitDegrees=20, twistLimitDegrees=5}
 {bone1Name=dyn_skirt_bk2_02_r, bone2Name=dyn_skirt_bk2_01_r, swing1LimitDegrees=25, swing2LimitDegrees=20, twistLimitDegrees=5}
 {bone1Name=dyn_skirt_bk2_03_r, bone2Name=dyn_skirt_bk2_02_r, swing1LimitDegrees=25, swing2LimitDegrees=20, twistLimitDegrees=5}
 {bone1Name=dyn_skirt_fr_01, bone2Name=dyn_skirt_fr_01_MSTR, swing1LimitDegrees=25, swing2LimitDegrees=20, twistLimitDegrees=5}
 {bone1Name=dyn_skirt_fr_02, bone2Name=dyn_skirt_fr_01, swing1LimitDegrees=25, swing2LimitDegrees=20, twistLimitDegrees=5}
 {bone1Name=dyn_skirt_fr_03, bone2Name=dyn_skirt_fr_02, swing1LimitDegrees=25, swing2LimitDegrees=20, twistLimitDegrees=5}
 {bone1Name=dyn_skirt_bk_01, bone2Name=dyn_skirt_bk_01_MSTR, swing1LimitDegrees=25, swing2LimitDegrees=20, twistLimitDegrees=5}
 {bone1Name=dyn_skirt_bk_02, bone2Name=dyn_skirt_bk_01, swing1LimitDegrees=25, swing2LimitDegrees=20, twistLimitDegrees=5}
 {bone1Name=dyn_skirt_bk_03, bone2Name=dyn_skirt_bk_02, swing1LimitDegrees=25, swing2LimitDegrees=20, twistLimitDegrees=5}
 {
  bone1Name=dyn_skirt_fr_01_l_MSTR
  bone2Name=pelvis
  swing1LimitDegrees=45
  swing1Motion=Free
  swing2LimitDegrees=45
  swing2Motion=Free
  twistLimitDegrees=45
  twistMotion=Free
 }
 {
  bone1Name=dyn_skirt_01_l_MSTR
  bone2Name=pelvis
  swing1LimitDegrees=45
  swing1Motion=Free
  swing2LimitDegrees=45
  swing2Motion=Free
  twistLimitDegrees=45
  twistMotion=Free
 }
 {
  bone1Name=dyn_skirt_bk_01_l_MSTR
  bone2Name=pelvis
  swing1LimitDegrees=45
  swing1Motion=Free
  swing2LimitDegrees=45
  swing2Motion=Free
  twistLimitDegrees=45
  twistMotion=Free
 }
 {
  bone1Name=dyn_skirt_fr2_01_l_MSTR
  bone2Name=pelvis
  swing1LimitDegrees=45
  swing1Motion=Free
  swing2LimitDegrees=45
  swing2Motion=Free
  twistLimitDegrees=45
  twistMotion=Free
 }
 {
  bone1Name=dyn_skirt_bk2_01_l_MSTR
  bone2Name=pelvis
  swing1LimitDegrees=45
  swing1Motion=Free
  swing2LimitDegrees=45
  swing2Motion=Free
  twistLimitDegrees=45
  twistMotion=Free
 }
 {
  bone1Name=dyn_skirt_fr_01_r_MSTR
  bone2Name=pelvis
  swing1LimitDegrees=45
  swing1Motion=Free
  swing2LimitDegrees=45
  swing2Motion=Free
  twistLimitDegrees=45
  twistMotion=Free
 }
 {
  bone1Name=dyn_skirt_01_r_MSTR
  bone2Name=pelvis
  swing1LimitDegrees=45
  swing1Motion=Free
  swing2LimitDegrees=45
  swing2Motion=Free
  twistLimitDegrees=45
  twistMotion=Free
 }
 {
  bone1Name=dyn_skirt_bk_01_r_MSTR
  bone2Name=pelvis
  swing1LimitDegrees=45
  swing1Motion=Free
  swing2LimitDegrees=45
  swing2Motion=Free
  twistLimitDegrees=45
  twistMotion=Free
 }
 {
  bone1Name=dyn_skirt_fr2_01_r_MSTR
  bone2Name=pelvis
  swing1LimitDegrees=45
  swing1Motion=Free
  swing2LimitDegrees=45
  swing2Motion=Free
  twistLimitDegrees=45
  twistMotion=Free
 }
 {
  bone1Name=dyn_skirt_bk2_01_r_MSTR
  bone2Name=pelvis
  swing1LimitDegrees=45
  swing1Motion=Free
  swing2LimitDegrees=45
  swing2Motion=Free
  twistLimitDegrees=45
  twistMotion=Free
 }
 {
  bone1Name=dyn_skirt_fr_01_MSTR
  bone2Name=pelvis
  swing1LimitDegrees=45
  swing1Motion=Free
  swing2LimitDegrees=45
  swing2Motion=Free
  twistLimitDegrees=45
  twistMotion=Free
 }
 {
  bone1Name=dyn_skirt_bk_01_MSTR
  bone2Name=pelvis
  swing1LimitDegrees=45
  swing1Motion=Free
  swing2LimitDegrees=45
  swing2Motion=Free
  twistLimitDegrees=45
  twistMotion=Free
 }
 {
  bone1Name=dyn_backString_MSTR
  bone2Name=spine_03
  swing1LimitDegrees=45
  swing1Motion=Free
  swing2LimitDegrees=45
  swing2Motion=Free
  twistLimitDegrees=45
  twistMotion=Free
 }
 {bone1Name=dyn_backString_01, bone2Name=dyn_backString_MSTR, swing1LimitDegrees=60, swing2LimitDegrees=60, twistLimitDegrees=10}
 {
  bone1Name=dyn_backString_02
  bone2Name=dyn_backString_01
  swing1LimitDegrees=45
  swing1Motion=Free
  swing2LimitDegrees=45
  swing2Motion=Free
  twistLimitDegrees=10
 }
 {
  bone1Name=dyn_backString_03
  bone2Name=dyn_backString_02
  swing1LimitDegrees=45
  swing1Motion=Free
  swing2LimitDegrees=45
  swing2Motion=Free
  twistLimitDegrees=10
 }
 {
  bone1Name=spine_02
  bone2Name=pelvis
  swing1LimitDegrees=45
  swing1Motion=Free
  swing2LimitDegrees=45
  swing2Motion=Free
  twistLimitDegrees=45
  twistMotion=Free
 }
 {
  bone1Name=spine_03
  bone2Name=spine_02
  swing1LimitDegrees=45
  swing1Motion=Free
  swing2LimitDegrees=45
  swing2Motion=Free
  twistLimitDegrees=45
  twistMotion=Free
 }
 {bone1Name=dyn_belt_gun_01, bone2Name=pelvis, swing1LimitDegrees=25, swing2LimitDegrees=15, twistLimitDegrees=25}
 {bone1Name=dyn_hipHeart, bone2Name=pelvis, swing1LimitDegrees=45, swing2LimitDegrees=25, twistLimitDegrees=45}
]
