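# Physics asset for the skirt mesh of the F_MED_Crush female cosmetic body, covering skirt, back string, belt gun and hip heart motion.
# Twelve three-box skirt chains hang from tiny MSTR anchor capsules on the pelvis, with limited swing and twist, stiffer on front-left and front-right.
# The back string chains off spine_03, while gun and heart constrain directly to the pelvis.
# Pelvis, spine and thigh capsules provide body shapes.
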PhysicsAsset /BRCosmetics/Characters/Player/Female/Medium/Bodies/F_MED_Crush/Meshes/F_MED_Crush_Skirt_msLow_Physics
bodies {
 dyn_backString_01 [{shapeType=Box, extentX=2, extentY=4.9314, extentZ=1.0927, center=<0.8202 0 0>}]
 dyn_backString_02 [{shapeType=Box, extentX=2, extentY=4.9314, extentZ=1.0927, center=<0.8202 0 0>}]
 dyn_backString_03 [{shapeType=Box, extentX=2, extentY=4.9314, extentZ=1.0927, center=<0.8202 0 0>}]
 dyn_backString_MSTR [{shapeType=Capsule, length=0.505, radius=0.505}]
 dyn_belt_gun_01 [{shapeType=Box, extentX=15.1603, extentY=16.5509, extentZ=1.3033, center=<8.4555 -0.0002 3.2218>, rotation=<rot 0 0 -4.2015>}]
 dyn_hipHeart [{shapeType=Box, extentX=17.1269, extentY=16.2409, extentZ=2.6086, center=<0 0 1.4564>}]
 dyn_skirt_01_l [{shapeType=Box, extentX=10.3816, extentY=13.0442, extentZ=0.7453, center=<4.9906 0 0>}]
 dyn_skirt_01_l_MSTR [{shapeType=Capsule, length=0.505, radius=0.505}]
 dyn_skirt_01_r [{shapeType=Box, extentX=10.3816, extentY=16.1813, extentZ=3.9411, center=<5.2189 0 0>}]
 dyn_skirt_01_r_MSTR [{shapeType=Capsule, length=0.505, radius=0.505}]
 dyn_skirt_02_l [{shapeType=Box, extentX=10.3816, extentY=13.0442, extentZ=2.3055, center=<5.1782 0 0.0001>}]
 dyn_skirt_02_r [{shapeType=Box, extentX=10.3816, extentY=16.1813, extentZ=3.3936, center=<5.0672 -0.0001 -0.0893>}]
 dyn_skirt_03_l [{shapeType=Box, extentX=10.3816, extentY=13.0442, extentZ=2.3055, center=<4.9641 0 0>}]
 dyn_skirt_03_r [{shapeType=Box, extentX=10.3816, extentY=16.1813, extentZ=3.3936, center=<4.8569 0 0>}]
 dyn_skirt_bk2_01_l [{shapeType=Box, extentX=10.3816, extentY=9.5419, extentZ=2.8934, center=<5.0777 0 0>}]
 dyn_skirt_bk2_01_l_MSTR [{shapeType=Capsule, length=0.505, radius=0.505}]
 dyn_skirt_bk2_01_r [{shapeType=Box, extentX=10.3816, extentY=13.2684, extentZ=3.4026, center=<5.3585 0 0>}]
 dyn_skirt_bk2_01_r_MSTR [{shapeType=Capsule, length=0.505, radius=0.505}]
 dyn_skirt_bk2_02_l [{shapeType=Box, extentX=10.3816, extentY=14.7674, extentZ=2.8934, center=<5.0432 0 0>}]
 dyn_skirt_bk2_02_r [{shapeType=Box, extentX=10.3816, extentY=13.2684, extentZ=3.4026, center=<5.1042 0 0>}]
 dyn_skirt_bk2_03_l [{shapeType=Box, extentX=10.3816, extentY=14.7674, extentZ=2.8934, center=<5.1763 0 0>}]
 dyn_skirt_bk2_03_r [{shapeType=Box, extentX=10.3816, extentY=13.2684, extentZ=3.4026, center=<5.1696 0 -0.0001>}]
 dyn_skirt_bk_01 [{shapeType=Box, extentX=10.3816, extentY=12.3702, extentZ=3.1213, center=<4.9054 0 0>}]
 dyn_skirt_bk_01_MSTR [{shapeType=Capsule, length=0.505, radius=0.505}]
 dyn_skirt_bk_01_l [{shapeType=Box, extentX=10.3816, extentY=13.6371, extentZ=2.464, center=<5.0423 0 0>}]
 dyn_skirt_bk_01_l_MSTR [{shapeType=Capsule, length=0.505, radius=0.505}]
 dyn_skirt_bk_01_r [{shapeType=Box, extentX=10.3816, extentY=14.9836, extentZ=1.7729, center=<5.0844 0 0.0001>}]
 dyn_skirt_bk_01_r_MSTR [{shapeType=Capsule, length=0.505, radius=0.505}]
 dyn_skirt_bk_02 [{shapeType=Box, extentX=10.3816, extentY=12.3702, extentZ=3.1213, center=<5.1641 0 -0.0001>}]
 dyn_skirt_bk_02_l [{shapeType=Box, extentX=10.3816, extentY=13.6371, extentZ=2.464, center=<5.1691 0 0>}]
 dyn_skirt_bk_02_r [{shapeType=Box, extentX=10.3816, extentY=14.9836, extentZ=1.7729, center=<4.9741 0 -0.0001>}]
 dyn_skirt_bk_03 [{shapeType=Box, extentX=10.3816, extentY=12.3702, extentZ=6.7971, center=<5.2588 0 0>}]
 dyn_skirt_bk_03_l [{shapeType=Box, extentX=10.3816, extentY=13.6371, extentZ=6.1397, center=<5.1055 0 0>}]
 dyn_skirt_bk_03_r [{shapeType=Box, extentX=10.3816, extentY=14.9836, extentZ=5.4487, center=<5.1057 0 -0.0001>}]
 dyn_skirt_fr2_01_l [{shapeType=Box, extentX=8.7015, extentY=23.6958, extentZ=2.4837, center=<5.347 0 0>}]
 dyn_skirt_fr2_01_l_MSTR [{shapeType=Capsule, length=0.505, radius=0.505}]
 dyn_skirt_fr2_01_r [{shapeType=Box, extentX=10.3816, extentY=17.4795, extentZ=1.8773, center=<5.1165 0 0.0001>}]
 dyn_skirt_fr2_01_r_MSTR [{shapeType=Capsule, length=0.505, radius=0.505}]
 dyn_skirt_fr2_02_l [{shapeType=Box, extentX=8.7015, extentY=23.6958, extentZ=2.4837, center=<5.1712 0 -0.0002>}]
 dyn_skirt_fr2_02_r [{shapeType=Box, extentX=10.3816, extentY=17.4795, extentZ=1.8075, center=<5.1338 0 -0.0002>}]
 dyn_skirt_fr2_03_l [{shapeType=Box, extentX=8.7015, extentY=23.6958, extentZ=2.4837, center=<5.2498 0 0>}]
 dyn_skirt_fr2_03_r [{shapeType=Box, extentX=10.3816, extentY=17.4795, extentZ=1.8075, center=<5.096 0 0>}]
 dyn_skirt_fr_01 [{shapeType=Box, extentX=10.3816, extentY=13.2247, extentZ=2.058, center=<5.0146 0 0.0001>}]
 dyn_skirt_fr_01_MSTR [{shapeType=Capsule, length=0.505, radius=0.505}]
 dyn_skirt_fr_01_l [{shapeType=Box, extentX=10.3816, extentY=14.241, extentZ=2.058, center=<5.0853 0 0>}]
 dyn_skirt_fr_01_l_MSTR [{shapeType=Capsule, length=0.505, radius=0.505}]
 dyn_skirt_fr_01_r [{shapeType=Box, extentX=10.3816, extentY=14.241, extentZ=2.058, center=<4.9553 0 -0.0001>}]
 dyn_skirt_fr_01_r_MSTR [{shapeType=Capsule, length=0.505, radius=0.505}]
 dyn_skirt_fr_02 [{shapeType=Box, extentX=10.3816, extentY=13.2247, extentZ=2, center=<5.0659 0 -0.0001>}]
 dyn_skirt_fr_02_l [{shapeType=Box, extentX=10.3816, extentY=14.241, extentZ=1.7669, center=<5.2173 0 -0.0001>}]
 dyn_skirt_fr_02_r [{shapeType=Box, extentX=10.3816, extentY=14.241, extentZ=2.796, center=<5.0219 0 -0.0001>}]
 dyn_skirt_fr_03 [{shapeType=Box, extentX=10.3816, extentY=13.2247, extentZ=2, center=<4.9765 0 0.0001>}]
 dyn_skirt_fr_03_l [{shapeType=Box, extentX=10.3816, extentY=14.241, extentZ=1.7669, center=<5.0377 0 0>}]
 dyn_skirt_fr_03_r [{shapeType=Box, extentX=10.3816, extentY=14.241, extentZ=2.796, center=<5.0707 0 0.0001>}]
 pelvis [{shapeType=Capsule, length=7.458, radius=8.0636, center=<1.1549 -1.274 -0.4831>}]
 spine_02 [{shapeType=Capsule, length=8.1468, radius=6.2361}]
 spine_03 [{shapeType=Capsule, length=8.1468, radius=6.2361}]
 thigh_l [{shapeType=Capsule, length=33.0077, radius=7.051, center=<-25.1663 0.0001 0>, rotation=<rot 90 -174.9489 5.0512>}]
 thigh_r [{shapeType=Capsule, length=33.3529, radius=7.4973, center=<25.1663 -0.0001 0>, rotation=<rot 90 7.5524 7.5525>}]
}
constraints [
 {
  bone1Name=thigh_l
  bone2Name=pelvis
  swing1LimitDegrees=45
  swing1Motion=Free
  swing2LimitDegrees=45
  swing2Motion=Free
  twistLimitDegrees=45
  twistMotion=Free
 }
 {
  bone1Name=thigh_r
  bone2Name=pelvis
  swing1LimitDegrees=45
  swing1Motion=Free
  swing2LimitDegrees=45
  swing2Motion=Free
  twistLimitDegrees=45
  twistMotion=Free
 }
 {bone1Name=dyn_skirt_fr_01_l, bone2Name=dyn_skirt_fr_01_l_MSTR, swing1LimitDegrees=5, swing2LimitDegrees=45, twistLimitDegrees=5}
 {bone1Name=dyn_skirt_fr_02_l, bone2Name=dyn_skirt_fr_01_l, swing1LimitDegrees=5, swing2LimitDegrees=45, twistLimitDegrees=5}
 {bone1Name=dyn_skirt_fr_03_l, bone2Name=dyn_skirt_fr_02_l, swing1LimitDegrees=5, swing2LimitDegrees=45, twistLimitDegrees=5}
 {bone1Name=dyn_skirt_01_l, bone2Name=dyn_skirt_01_l_MSTR, swing1LimitDegrees=25, swing2LimitDegrees=45, twistLimitDegrees=5}
 {bone1Name=dyn_skirt_02_l, bone2Name=dyn_skirt_01_l, swing1LimitDegrees=25, swing2LimitDegrees=45, twistLimitDegrees=5}
 {bone1Name=dyn_skirt_03_l, bone2Name=dyn_skirt_02_l, swing1LimitDegrees=25, swing2LimitDegrees=45, twistLimitDegrees=5}
 {bone1Name=dyn_skirt_bk_01_l, bone2Name=dyn_skirt_bk_01_l_MSTR, swing1LimitDegrees=25, swing2LimitDegrees=45, twistLimitDegrees=5}
 {bone1Name=dyn_skirt_bk_02_l, bone2Name=dyn_skirt_bk_01_l, swing1LimitDegrees=25, swing2LimitDegrees=45, twistLimitDegrees=5}
 {bone1Name=dyn_skirt_bk_03_l, bone2Name=dyn_skirt_bk_02_l, swing1LimitDegrees=25, swing2LimitDegrees=45, twistLimitDegrees=5}
 {bone1Name=dyn_skirt_fr2_01_l, bone2Name=dyn_skirt_fr2_01_l_MSTR, swing1LimitDegrees=25, swing2LimitDegrees=45, twistLimitDegrees=5}
 {bone1Name=dyn_skirt_fr2_02_l, bone2Name=dyn_skirt_fr2_01_l, swing1LimitDegrees=25, swing2LimitDegrees=45, twistLimitDegrees=5}
 {bone1Name=dyn_skirt_fr2_03_l, bone2Name=dyn_skirt_fr2_02_l, swing1LimitDegrees=25, swing2LimitDegrees=45, twistLimitDegrees=5}
 {bone1Name=dyn_skirt_bk2_01_l, bone2Name=dyn_skirt_bk2_01_l_MSTR, swing1LimitDegrees=25, swing2LimitDegrees=45, twistLimitDegrees=5}
 {bone1Name=dyn_skirt_bk2_02_l, bone2Name=dyn_skirt_bk2_01_l, swing1LimitDegrees=25, swing2LimitDegrees=45, twistLimitDegrees=5}
 {bone1Name=dyn_skirt_bk2_03_l, bone2Name=dyn_skirt_bk2_02_l, swing1LimitDegrees=25, swing2LimitDegrees=45, twistLimitDegrees=5}
 {bone1Name=dyn_skirt_fr_01_r, bone2Name=dyn_skirt_fr_01_r_MSTR, swing1LimitDegrees=5, swing2LimitDegrees=45, twistLimitDegrees=5}
 {bone1Name=dyn_skirt_fr_02_r, bone2Name=dyn_skirt_fr_01_r, swing1LimitDegrees=5, swing2LimitDegrees=45, twistLimitDegrees=5}
 {bone1Name=dyn_skirt_fr_03_r, bone2Name=dyn_skirt_fr_02_r, swing1LimitDegrees=5, swing2LimitDegrees=45, twistLimitDegrees=5}
 {bone1Name=dyn_skirt_01_r, bone2Name=dyn_skirt_01_r_MSTR, swing1LimitDegrees=25, swing2LimitDegrees=45, twistLimitDegrees=5}
 {bone1Name=dyn_skirt_02_r, bone2Name=dyn_skirt_01_r, swing1LimitDegrees=25, swing2LimitDegrees=45, twistLimitDegrees=5}
 {bone1Name=dyn_skirt_03_r, bone2Name=dyn_skirt_02_r, swing1LimitDegrees=25, swing2LimitDegrees=45, twistLimitDegrees=5}
 {bone1Name=dyn_skirt_bk_01_r, bone2Name=dyn_skirt_bk_01_r_MSTR, swing1LimitDegrees=25, swing2LimitDegrees=45, twistLimitDegrees=5}
 {bone1Name=dyn_skirt_bk_03_r, bone2Name=dyn_skirt_bk_02_r, swing1LimitDegrees=25, swing2LimitDegrees=45, twistLimitDegrees=5}
 {bone1Name=dyn_skirt_bk_02_r, bone2Name=dyn_skirt_bk_01_r, swing1LimitDegrees=25, swing2LimitDegrees=45, twistLimitDegrees=5}
 {bone1Name=dyn_skirt_fr2_01_r, bone2Name=dyn_skirt_fr2_01_r_MSTR, swing1LimitDegrees=25, swing2LimitDegrees=45, twistLimitDegrees=5}
 {bone1Name=dyn_skirt_fr2_02_r, bone2Name=dyn_skirt_fr2_01_r, swing1LimitDegrees=25, swing2LimitDegrees=45, twistLimitDegrees=5}
 {bone1Name=dyn_skirt_fr2_03_r, bone2Name=dyn_skirt_fr2_02_r, swing1LimitDegrees=25, swing2LimitDegrees=45, twistLimitDegrees=5}
 {bone1Name=dyn_skirt_bk2_01_r, bone2Name=dyn_skirt_bk2_01_r_MSTR, swing1LimitDegrees=25, swing2LimitDegrees=45, twistLimitDegrees=5}
 {bone1Name=dyn_skirt_bk2_02_r, bone2Name=dyn_skirt_bk2_01_r, swing1LimitDegrees=25, swing2LimitDegrees=45, twistLimitDegrees=5}
 {bone1Name=dyn_skirt_bk2_03_r, bone2Name=dyn_skirt_bk2_02_r, swing1LimitDegrees=25, swing2LimitDegrees=45, twistLimitDegrees=5}
 {bone1Name=dyn_skirt_fr_01, bone2Name=dyn_skirt_fr_01_MSTR, swing1LimitDegrees=25, swing2LimitDegrees=45, twistLimitDegrees=5}
 {bone1Name=dyn_skirt_fr_02, bone2Name=dyn_skirt_fr_01, swing1LimitDegrees=25, swing2LimitDegrees=45, twistLimitDegrees=5}
 {bone1Name=dyn_skirt_fr_03, bone2Name=dyn_skirt_fr_02, swing1LimitDegrees=25, swing2LimitDegrees=45, twistLimitDegrees=5}
 {bone1Name=dyn_skirt_bk_01, bone2Name=dyn_skirt_bk_01_MSTR, swing1LimitDegrees=25, swing2LimitDegrees=45, twistLimitDegrees=5}
 {bone1Name=dyn_skirt_bk_02, bone2Name=dyn_skirt_bk_01, swing1LimitDegrees=25, swing2LimitDegrees=45, twistLimitDegrees=5}
 {bone1Name=dyn_skirt_bk_03, bone2Name=dyn_skirt_bk_02, swing1LimitDegrees=25, swing2LimitDegrees=45, twistLimitDegrees=5}
 {
  bone1Name=dyn_skirt_fr_01_l_MSTR
  bone2Name=pelvis
  swing1LimitDegrees=45
  swing1Motion=Free
  swing2LimitDegrees=45
  swing2Motion=Free
  twistLimitDegrees=45
  twistMotion=Free
 }
 {
  bone1Name=dyn_skirt_01_l_MSTR
  bone2Name=pelvis
  swing1LimitDegrees=45
  swing1Motion=Free
  swing2LimitDegrees=45
  swing2Motion=Free
  twistLimitDegrees=45
  twistMotion=Free
 }
 {
  bone1Name=dyn_skirt_bk_01_l_MSTR
  bone2Name=pelvis
  swing1LimitDegrees=45
  swing1Motion=Free
  swing2LimitDegrees=45
  swing2Motion=Free
  twistLimitDegrees=45
  twistMotion=Free
 }
 {
  bone1Name=dyn_skirt_fr2_01_l_MSTR
  bone2Name=pelvis
  swing1LimitDegrees=45
  swing1Motion=Free
  swing2LimitDegrees=45
  swing2Motion=Free
  twistLimitDegrees=45
  twistMotion=Free
 }
 {
  bone1Name=dyn_skirt_bk2_01_l_MSTR
  bone2Name=pelvis
  swing1LimitDegrees=45
  swing1Motion=Free
  swing2LimitDegrees=45
  swing2Motion=Free
  twistLimitDegrees=45
  twistMotion=Free
 }
 {
  bone1Name=dyn_skirt_fr_01_r_MSTR
  bone2Name=pelvis
  swing1LimitDegrees=45
  swing1Motion=Free
  swing2LimitDegrees=45
  swing2Motion=Free
  twistLimitDegrees=45
  twistMotion=Free
 }
 {
  bone1Name=dyn_skirt_01_r_MSTR
  bone2Name=pelvis
  swing1LimitDegrees=45
  swing1Motion=Free
  swing2LimitDegrees=45
  swing2Motion=Free
  twistLimitDegrees=45
  twistMotion=Free
 }
 {
  bone1Name=dyn_skirt_bk_01_r_MSTR
  bone2Name=pelvis
  swing1LimitDegrees=45
  swing1Motion=Free
  swing2LimitDegrees=45
  swing2Motion=Free
  twistLimitDegrees=45
  twistMotion=Free
 }
 {
  bone1Name=dyn_skirt_fr2_01_r_MSTR
  bone2Name=pelvis
  swing1LimitDegrees=45
  swing1Motion=Free
  swing2LimitDegrees=45
  swing2Motion=Free
  twistLimitDegrees=45
  twistMotion=Free
 }
 {
  bone1Name=dyn_skirt_bk2_01_r_MSTR
  bone2Name=pelvis
  swing1LimitDegrees=45
  swing1Motion=Free
  swing2LimitDegrees=45
  swing2Motion=Free
  twistLimitDegrees=45
  twistMotion=Free
 }
 {
  bone1Name=dyn_skirt_fr_01_MSTR
  bone2Name=pelvis
  swing1LimitDegrees=45
  swing1Motion=Free
  swing2LimitDegrees=45
  swing2Motion=Free
  twistLimitDegrees=45
  twistMotion=Free
 }
 {
  bone1Name=dyn_skirt_bk_01_MSTR
  bone2Name=pelvis
  swing1LimitDegrees=45
  swing1Motion=Free
  swing2LimitDegrees=45
  swing2Motion=Free
  twistLimitDegrees=45
  twistMotion=Free
 }
 {
  bone1Name=dyn_backString_MSTR
  bone2Name=spine_03
  swing1LimitDegrees=45
  swing1Motion=Free
  swing2LimitDegrees=45
  swing2Motion=Free
  twistLimitDegrees=45
  twistMotion=Free
 }
 {bone1Name=dyn_backString_01, bone2Name=dyn_backString_MSTR, swing1LimitDegrees=60, swing2LimitDegrees=60, twistLimitDegrees=10}
 {
  bone1Name=dyn_backString_02
  bone2Name=dyn_backString_01
  swing1LimitDegrees=45
  swing1Motion=Free
  swing2LimitDegrees=45
  swing2Motion=Free
  twistLimitDegrees=10
 }
 {
  bone1Name=dyn_backString_03
  bone2Name=dyn_backString_02
  swing1LimitDegrees=45
  swing1Motion=Free
  swing2LimitDegrees=45
  swing2Motion=Free
  twistLimitDegrees=10
 }
 {
  bone1Name=spine_02
  bone2Name=pelvis
  swing1LimitDegrees=45
  swing1Motion=Free
  swing2LimitDegrees=45
  swing2Motion=Free
  twistLimitDegrees=45
  twistMotion=Free
 }
 {
  bone1Name=spine_03
  bone2Name=spine_02
  swing1LimitDegrees=45
  swing1Motion=Free
  swing2LimitDegrees=45
  swing2Motion=Free
  twistLimitDegrees=45
  twistMotion=Free
 }
 {bone1Name=dyn_belt_gun_01, bone2Name=pelvis, swing1LimitDegrees=25, swing2LimitDegrees=25, twistLimitDegrees=25}
 {bone1Name=dyn_hipHeart, bone2Name=pelvis, swing1LimitDegrees=45, swing2LimitDegrees=45, twistLimitDegrees=45}
]
