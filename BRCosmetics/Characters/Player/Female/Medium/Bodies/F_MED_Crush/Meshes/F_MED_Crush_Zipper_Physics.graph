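# Physics asset for the F_MED_Crush female medium body cosmetic, covering the dangling zipper bones on the left and right arms.
# Small capsules on each dyn arm zipper bone pair with larger capsules that approximate the forearms.
# A constraint ties each zipper to its same-side lowerarm, with narrow, matching swing and twist limit angles to keep its motion tight.
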PhysicsAsset /BRCosmetics/Characters/Player/Female/Medium/Bodies/F_MED_Crush/Meshes/F_MED_Crush_Zipper_Physics
bodies {
 dyn_l_arm_zipper [{shapeType=Capsule, length=1.4077, radius=0.6832, center=<1.2061 0 0>, rotation=<rot 90 0 0>}]
 dyn_r_arm_zipper [{shapeType=Capsule, length=1.5925, radius=0.5, center=<1.4332 0.2435 0.2098>, rotation=<rot 0 -81.897 -83.3382>}]
 lowerarm_l [{shapeType=Capsule, length=17.8928, radius=4.083, center=<12.1396 0 -0.8939>, rotation=<rot 84.8474 0.0005 0.0005>}]
 lowerarm_r [{shapeType=Capsule, length=17.8928, radius=4.083, center=<-12.201 0 0.2134>, rotation=<rot 84.8474 0.0005 -179.9995>}]
}
constraints [
 {bone1Name=dyn_r_arm_zipper, bone2Name=lowerarm_r, swing1LimitDegrees=15, swing2LimitDegrees=15, twistLimitDegrees=15}
 {bone1Name=dyn_l_arm_zipper, bone2Name=lowerarm_l, swing1LimitDegrees=15, swing2LimitDegrees=15, twistLimitDegrees=15}
]
