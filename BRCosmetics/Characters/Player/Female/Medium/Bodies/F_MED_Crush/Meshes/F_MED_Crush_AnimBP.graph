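# Secondary-motion AnimBlueprint for the F_MED_Crush female body, animating skirt, zipper, belt gun and hip heart over a pose copied from the attached parent mesh.
# RigidBody simulates the zipper, then bool blends choose between skirt RigidBody setups.
# Constraints pin back skirt bones to the pelvis and pull front bones with thigh helpers.
# Cached poses feed two bool-blended AnimDynamics chains with different stiffness, followed by an additive front-skirt rotation.
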AnimBlueprint /BRCosmetics/Characters/Player/Female/Medium/Bodies/F_MED_Crush/Meshes/F_MED_Crush_AnimBP
target_skeleton=&/F_MED_Crush_Skeleton
graphs [
 AnimGraph {nodes [
   n1 = Anim Dynamics - Bone: dyn_skirt_fr_02_l!(ComponentPose=Local To Component(LocalPose=Use cached pose 'Cache_pose_Crush_01'()), Alpha=0) -> (Pose) {
    angularDampingOverride=0.9999
    angularSpringConstant=75
    bAngularSpring=true
    bGravityOverrideInSimSpace=true
    bOverrideAngularDamping=true
    simulationSpace=RootRelative
    alphaBoolBlend {blendOption=Linear}
    alphaScaleBiasClamp {interpSpeedDecreasing=5}
    boundBone {boneName=dyn_skirt_fr_02_l}
    componentLinearAccScale=<1 1 1>
    componentLinearVelScale=<1 1 1>
    physicsBodyDefinitions [{
      sphereCollisionRadius=0
      boundBone {boneName=dyn_skirt_fr_02_l}
      boxExtents=<20 10 10>
      constraintSetup {angularLimitsMax=<25 0 25>, angularLimitsMin=<-25 -25 -25>, angularTarget=<0 0 0>}
      localJointOffset=<2 0 0>
     }]
   }
   n2 = Anim Dynamics - Bone: dyn_skirt_fr_02_r!(ComponentPose=@n1, Alpha=0) -> (Pose) {
    angularDampingOverride=0.9999
    angularSpringConstant=75
    bAngularSpring=true
    bGravityOverrideInSimSpace=true
    bOverrideAngularDamping=true
    simulationSpace=RootRelative
    alphaBoolBlend {blendOption=Linear}
    alphaScaleBiasClamp {bInterpResult=true, interpSpeedDecreasing=5}
    boundBone {boneName=dyn_skirt_fr_02_r}
    componentLinearAccScale=<1 1 1>
    componentLinearVelScale=<1 1 1>
    physicsBodyDefinitions [{
      sphereCollisionRadius=0
      boundBone {boneName=dyn_skirt_fr_02_r}
      boxExtents=<20 20 10>
      constraintSetup {angularLimitsMax=<25 0 25>, angularLimitsMin=<-25 -25 -25>, angularTarget=<0 0 0>}
      localJointOffset=<2 0 0>
     }]
   }
   n3 = Anim Dynamics - Bone: dyn_skirt_fr2_02_l!(ComponentPose=@n2, Alpha=1) -> (Pose) {
    angularDampingOverride=0.9999
    angularSpringConstant=75
    bAngularSpring=true
    bGravityOverrideInSimSpace=true
    bOverrideAngularDamping=true
    simulationSpace=RootRelative
    alphaBoolBlend {blendOption=Linear}
    alphaScaleBiasClamp {bInterpResult=true, interpSpeedDecreasing=5}
    boundBone {boneName=dyn_skirt_fr2_02_l}
    componentLinearAccScale=<1 1 1>
    componentLinearVelScale=<1 1 1>
    physicsBodyDefinitions [{
      sphereCollisionRadius=0
      boundBone {boneName=dyn_skirt_fr2_02_l}
      boxExtents=<20 10 10>
      constraintSetup {angularLimitsMax=<25 0 25>, angularLimitsMin=<-25 -25 -25>, angularTarget=<0 0 0>}
      localJointOffset=<2 0 0>
     }]
   }
   n4 = Anim Dynamics - Bone: dyn_skirt_fr2_02_r!(ComponentPose=@n3, Alpha=1) -> (Pose) {
    angularDampingOverride=0.9999
    angularSpringConstant=75
    bAngularSpring=true
    bGravityOverrideInSimSpace=true
    bOverrideAngularDamping=true
    simulationSpace=RootRelative
    alphaBoolBlend {blendOption=Linear}
    alphaScaleBiasClamp {bInterpResult=true, interpSpeedDecreasing=5}
    boundBone {boneName=dyn_skirt_fr2_02_r}
    componentLinearAccScale=<1 1 1>
    componentLinearVelScale=<1 1 1>
    physicsBodyDefinitions [{
      sphereCollisionRadius=0
      boundBone {boneName=dyn_skirt_fr2_02_r}
      boxExtents=<20 20 10>
      constraintSetup {angularLimitsMax=<25 0 25>, angularLimitsMin=<-25 -25 -25>, angularTarget=<0 0 0>}
      localJointOffset=<2 0 0>
     }]
   }
   n5 = Anim Dynamics - Bone: dyn_skirt_02_l!(ComponentPose=@n4, Alpha=1) -> (Pose) {
    angularDampingOverride=0.9999
    angularSpringConstant=75
    bAngularSpring=true
    bGravityOverrideInSimSpace=true
    bOverrideAngularDamping=true
    simulationSpace=RootRelative
    alphaBoolBlend {blendOption=Linear}
    alphaScaleBiasClamp {bInterpResult=true, interpSpeedDecreasing=5}
    boundBone {boneName=dyn_skirt_02_l}
    componentLinearAccScale=<1 1 1>
    componentLinearVelScale=<1 1 1>
    physicsBodyDefinitions [{
      sphereCollisionRadius=0
      boundBone {boneName=dyn_skirt_02_l}
      boxExtents=<20 10 10>
      constraintSetup {angularLimitsMax=<25 0 25>, angularLimitsMin=<-25 -25 -25>, angularTarget=<0 0 0>}
      localJointOffset=<2 0 0>
     }]
   }
   n6 = Anim Dynamics - Bone: dyn_skirt_01_r!(ComponentPose=@n5, Alpha=1) -> (Pose) {
    angularDampingOverride=0.9999
    angularSpringConstant=75
    bAngularSpring=true
    bGravityOverrideInSimSpace=true
    bOverrideAngularDamping=true
    simulationSpace=RootRelative
    alphaBoolBlend {blendOption=Linear}
    alphaScaleBiasClamp {bInterpResult=true, interpSpeedDecreasing=5}
    boundBone {boneName=dyn_skirt_01_r}
    componentLinearAccScale=<1 1 1>
    componentLinearVelScale=<1 1 1>
    physicsBodyDefinitions [{
      sphereCollisionRadius=0
      boundBone {boneName=dyn_skirt_01_r}
      boxExtents=<20 20 10>
      constraintSetup {angularLimitsMax=<25 0 25>, angularLimitsMin=<-25 -25 -25>, angularTarget=<0 0 0>}
      localJointOffset=<2 0 0>
     }]
   }
   n7 = Anim Dynamics - Bone: dyn_skirt_fr_03_l!(ComponentPose=@n6, Alpha=1) -> (Pose) {
    angularDampingOverride=0.9999
    angularSpringConstant=75
    bAngularSpring=true
    bGravityOverrideInSimSpace=true
    bOverrideAngularDamping=true
    simulationSpace=RootRelative
    alphaBoolBlend {blendOption=Linear}
    alphaScaleBiasClamp {interpSpeedDecreasing=5}
    boundBone {boneName=dyn_skirt_fr_03_l}
    componentLinearAccScale=<1 1 1>
    componentLinearVelScale=<1 1 1>
    physicsBodyDefinitions [{
      sphereCollisionRadius=0
      boundBone {boneName=dyn_skirt_fr_03_l}
      boxExtents=<20 10 10>
      constraintSetup {angularLimitsMax=<25 0 25>, angularLimitsMin=<-25 -25 -25>, angularTarget=<0 0 0>}
      localJointOffset=<2 0 0>
     }]
   }
   n8 = Anim Dynamics - Bone: dyn_skirt_fr_03_r!(ComponentPose=@n7, Alpha=1) -> (Pose) {
    angularDampingOverride=0.9999
    angularSpringConstant=75
    bAngularSpring=true
    bGravityOverrideInSimSpace=true
    bOverrideAngularDamping=true
    simulationSpace=RootRelative
    alphaBoolBlend {blendOption=Linear}
    alphaScaleBiasClamp {bInterpResult=true, interpSpeedDecreasing=5}
    boundBone {boneName=dyn_skirt_fr_03_r}
    componentLinearAccScale=<1 1 1>
    componentLinearVelScale=<1 1 1>
    physicsBodyDefinitions [{
      sphereCollisionRadius=0
      boundBone {boneName=dyn_skirt_fr_03_r}
      boxExtents=<20 20 10>
      constraintSetup {angularLimitsMax=<25 0 25>, angularLimitsMin=<-25 -25 -25>, angularTarget=<0 0 0>}
      localJointOffset=<2 0 0>
     }]
   }
   n9 = Anim Dynamics - Bone: dyn_skirt_fr2_03_l!(ComponentPose=@n8, Alpha=1) -> (Pose) {
    angularDampingOverride=0.9999
    angularSpringConstant=75
    bAngularSpring=true
    bGravityOverrideInSimSpace=true
    bOverrideAngularDamping=true
    simulationSpace=RootRelative
    alphaBoolBlend {blendOption=Linear}
    alphaScaleBiasClamp {bInterpResult=true, interpSpeedDecreasing=5}
    boundBone {boneName=dyn_skirt_fr2_03_l}
    componentLinearAccScale=<1 1 1>
    componentLinearVelScale=<1 1 1>
    physicsBodyDefinitions [{
      sphereCollisionRadius=0
      boundBone {boneName=dyn_skirt_fr2_03_l}
      boxExtents=<20 10 10>
      constraintSetup {angularLimitsMax=<25 0 25>, angularLimitsMin=<-25 -25 -25>, angularTarget=<0 0 0>}
      localJointOffset=<2 0 0>
     }]
   }
   n10 = Anim Dynamics - Bone: dyn_skirt_fr2_03_r!(ComponentPose=@n9, Alpha=1) -> (Pose) {
    angularDampingOverride=0.9999
    angularSpringConstant=75
    bAngularSpring=true
    bGravityOverrideInSimSpace=true
    bOverrideAngularDamping=true
    simulationSpace=RootRelative
    alphaBoolBlend {blendOption=Linear}
    alphaScaleBiasClamp {bInterpResult=true, interpSpeedDecreasing=5}
    boundBone {boneName=dyn_skirt_fr2_03_r}
    componentLinearAccScale=<1 1 1>
    componentLinearVelScale=<1 1 1>
    physicsBodyDefinitions [{
      sphereCollisionRadius=0
      boundBone {boneName=dyn_skirt_fr2_03_r}
      boxExtents=<20 20 10>
      constraintSetup {angularLimitsMax=<25 0 25>, angularLimitsMin=<-25 -25 -25>, angularTarget=<0 0 0>}
      localJointOffset=<2 0 0>
     }]
   }
   n11 = Anim Dynamics - Bone: dyn_skirt_03_l!(ComponentPose=@n10, Alpha=1) -> (Pose) {
    angularDampingOverride=0.9999
    angularSpringConstant=75
    bAngularSpring=true
    bGravityOverrideInSimSpace=true
    bOverrideAngularDamping=true
    simulationSpace=RootRelative
    alphaBoolBlend {blendOption=Linear}
    alphaScaleBiasClamp {bInterpResult=true, interpSpeedDecreasing=5}
    boundBone {boneName=dyn_skirt_03_l}
    componentLinearAccScale=<1 1 1>
    componentLinearVelScale=<1 1 1>
    physicsBodyDefinitions [{
      sphereCollisionRadius=0
      boundBone {boneName=dyn_skirt_03_l}
      boxExtents=<20 10 10>
      constraintSetup {angularLimitsMax=<25 0 25>, angularLimitsMin=<-25 -25 -25>, angularTarget=<0 0 0>}
      localJointOffset=<2 0 0>
     }]
   }
   n12 = Anim Dynamics - Bone: dyn_skirt_03_r!(ComponentPose=@n11, Alpha=1) -> (Pose) {
    angularDampingOverride=0.9999
    angularSpringConstant=75
    bAngularSpring=true
    bGravityOverrideInSimSpace=true
    bOverrideAngularDamping=true
    simulationSpace=RootRelative
    alphaBoolBlend {blendOption=Linear}
    alphaScaleBiasClamp {bInterpResult=true, interpSpeedDecreasing=5}
    boundBone {boneName=dyn_skirt_03_r}
    componentLinearAccScale=<1 1 1>
    componentLinearVelScale=<1 1 1>
    physicsBodyDefinitions [{
      sphereCollisionRadius=0
      boundBone {boneName=dyn_skirt_03_r}
      boxExtents=<20 20 10>
      constraintSetup {angularLimitsMax=<25 0 25>, angularLimitsMin=<-25 -25 -25>, angularTarget=<0 0 0>}
      localJointOffset=<2 0 0>
     }]
   }
   n13 = Anim Dynamics - Bone: dyn_belt_gun_01!(ComponentPose=@n12, Alpha=1) -> (Pose) {
    angularDampingOverride=0.9999
    angularSpringConstant=75
    bAngularSpring=true
    bGravityOverrideInSimSpace=true
    bOverrideAngularDamping=true
    simulationSpace=BoneRelative
    alphaBoolBlend {blendOption=Linear}
    boundBone {boneName=dyn_belt_gun_01}
    componentLinearAccScale=<1 1 1>
    componentLinearVelScale=<1 1 1>
    physicsBodyDefinitions [{
      sphereCollisionRadius=0
      boundBone {boneName=dyn_belt_gun_01}
      boxExtents=<10 20 20>
      constraintSetup {angularLimitsMax=<25 15 50>, angularLimitsMin=<-25 -15 -50>, angularTarget=<0 0 0>}
      localJointOffset=<2 0 0>
     }]
    relativeSpaceBone {boneName=root}
   }
   n14 = Anim Dynamics - Bone: dyn_hipHeart!(ComponentPose=@n13, Alpha=0.5) -> (Pose) {
    angularDampingOverride=0.9999
    angularSpringConstant=75
    bAngularSpring=true
    bGravityOverrideInSimSpace=true
    bOverrideAngularDamping=true
    simulationSpace=RootRelative
    alphaBoolBlend {blendOption=Linear}
    boundBone {boneName=dyn_hipHeart}
    componentLinearAccScale=<1 1 1>
    componentLinearVelScale=<1 1 1>
    physicsBodyDefinitions [{
      sphereCollisionRadius=0
      boundBone {boneName=dyn_hipHeart}
      boxExtents=<20 20 10>
      constraintSetup {angularLimitsMax=<25 0 25>, angularLimitsMin=<-25 -25 -25>, angularTarget=<0 0 0>}
      localJointOffset=<2 0 0>
     }]
   }
   n15 = Component To Local(ComponentPose=@n14)
   n16 = Anim Dynamics - Bone: dyn_skirt_fr_02_l(ComponentPose=Local To Component(LocalPose=Use cached pose 'Cache_pose_Crush_01'()), Alpha=0) {
    angularDampingOverride=0.9999
    angularSpringConstant=400
    bAngularSpring=true
    bGravityOverrideInSimSpace=true
    bOverrideAngularDamping=true
    bUseGravityOverride=true
    simulationSpace=BoneRelative
    alphaBoolBlend {blendOption=Linear}
    boundBone {boneName=dyn_skirt_fr_02_l}
    componentAppliedLinearAccClamp=<5000 5000 5000>
    componentLinearAccScale=<1 1 1>
    componentLinearVelScale=<1 1 1>
    gravityOverride {z=-980}
    physicsBodyDefinitions [{
      sphereCollisionRadius=0
      boundBone {boneName=dyn_skirt_fr_02_l}
      boxExtents=<20 10 10>
      constraintSetup {angularLimitsMax=<10 0 10>, angularLimitsMin=<-10 -15 -10>, angularTarget=<1 0 0>}
      localJointOffset=<1 0 0>
     }]
    relativeSpaceBone {boneName=pelvis}
   }
   n17 = Anim Dynamics - Bone: dyn_skirt_fr_02_r(ComponentPose=@n16, Alpha=0) {
    angularDampingOverride=0.9999
    angularSpringConstant=400
    bAngularSpring=true
    bGravityOverrideInSimSpace=true
    bOverrideAngularDamping=true
    bUseGravityOverride=true
    simulationSpace=BoneRelative
    alphaBoolBlend {blendOption=Linear}
    boundBone {boneName=dyn_skirt_fr_02_r}
    componentAppliedLinearAccClamp=<5000 5000 5000>
    componentLinearAccScale=<1 1 1>
    componentLinearVelScale=<1 1 1>
    gravityOverride {z=-980}
    physicsBodyDefinitions [{
      sphereCollisionRadius=0
      boundBone {boneName=dyn_skirt_fr_02_r}
      boxExtents=<20 20 10>
      constraintSetup {angularLimitsMax=<10 0 10>, angularLimitsMin=<-10 -15 -10>, angularTarget=<1 0 0>}
      localJointOffset=<1 0 0>
     }]
    relativeSpaceBone {boneName=pelvis}
   }
   n18 = Anim Dynamics - Bone: dyn_skirt_fr2_02_l(ComponentPose=@n17, Alpha=1) {
    angularDampingOverride=0.9999
    angularSpringConstant=400
    bAngularSpring=true
    bGravityOverrideInSimSpace=true
    bOverrideAngularDamping=true
    bUseGravityOverride=true
    simulationSpace=BoneRelative
    alphaBoolBlend {blendOption=Linear}
    boundBone {boneName=dyn_skirt_fr2_02_l}
    componentAppliedLinearAccClamp=<5000 5000 5000>
    componentLinearAccScale=<1 1 1>
    componentLinearVelScale=<1 1 1>
    gravityOverride {z=-980}
    physicsBodyDefinitions [{
      sphereCollisionRadius=0
      boundBone {boneName=dyn_skirt_fr2_02_l}
      boxExtents=<20 10 10>
      constraintSetup {angularLimitsMax=<10 0 10>, angularLimitsMin=<-10 -15 -10>, angularTarget=<1 0 0>}
      localJointOffset=<1 0 0>
     }]
    relativeSpaceBone {boneName=pelvis}
   }
   n19 = Anim Dynamics - Bone: dyn_skirt_fr2_02_r(ComponentPose=@n18, Alpha=1) {
    angularDampingOverride=0.9999
    angularSpringConstant=400
    bAngularSpring=true
    bGravityOverrideInSimSpace=true
    bOverrideAngularDamping=true
    bUseGravityOverride=true
    simulationSpace=BoneRelative
    alphaBoolBlend {blendOption=Linear}
    boundBone {boneName=dyn_skirt_fr2_02_r}
    componentAppliedLinearAccClamp=<5000 5000 5000>
    componentLinearAccScale=<1 1 1>
    componentLinearVelScale=<1 1 1>
    gravityOverride {z=-980}
    physicsBodyDefinitions [{
      sphereCollisionRadius=0
      boundBone {boneName=dyn_skirt_fr2_02_r}
      boxExtents=<20 20 10>
      constraintSetup {angularLimitsMax=<10 0 10>, angularLimitsMin=<-10 -15 -10>, angularTarget=<1 0 0>}
      localJointOffset=<1 0 0>
     }]
    relativeSpaceBone {boneName=pelvis}
   }
   n20 = Anim Dynamics - Bone: dyn_skirt_02_l(ComponentPose=@n19, Alpha=1) {
    angularDampingOverride=0.9999
    angularSpringConstant=400
    bAngularSpring=true
    bGravityOverrideInSimSpace=true
    bOverrideAngularDamping=true
    bUseGravityOverride=true
    simulationSpace=BoneRelative
    alphaBoolBlend {blendOption=Linear}
    boundBone {boneName=dyn_skirt_02_l}
    componentAppliedLinearAccClamp=<5000 5000 5000>
    componentLinearAccScale=<1 1 1>
    componentLinearVelScale=<1 1 1>
    gravityOverride {z=-980}
    physicsBodyDefinitions [{
      sphereCollisionRadius=0
      boundBone {boneName=dyn_skirt_02_l}
      boxExtents=<20 10 10>
      constraintSetup {angularLimitsMax=<10 0 10>, angularLimitsMin=<-10 -15 -10>, angularTarget=<1 0 0>}
      localJointOffset=<1 0 0>
     }]
    relativeSpaceBone {boneName=pelvis}
   }
   n21 = Anim Dynamics - Bone: dyn_skirt_01_r(ComponentPose=@n20, Alpha=1) {
    angularDampingOverride=0.9999
    angularSpringConstant=400
    bAngularSpring=true
    bGravityOverrideInSimSpace=true
    bOverrideAngularDamping=true
    bUseGravityOverride=true
    simulationSpace=BoneRelative
    alphaBoolBlend {blendOption=Linear}
    boundBone {boneName=dyn_skirt_01_r}
    componentAppliedLinearAccClamp=<5000 5000 5000>
    componentLinearAccScale=<1 1 1>
    componentLinearVelScale=<1 1 1>
    gravityOverride {z=-980}
    physicsBodyDefinitions [{
      sphereCollisionRadius=0
      boundBone {boneName=dyn_skirt_01_r}
      boxExtents=<20 20 10>
      constraintSetup {angularLimitsMax=<10 0 10>, angularLimitsMin=<-10 -15 -10>, angularTarget=<1 0 0>}
      localJointOffset=<1 0 0>
     }]
    relativeSpaceBone {boneName=pelvis}
   }
   n22 = Anim Dynamics - Bone: dyn_skirt_fr_03_l(ComponentPose=@n21, Alpha=1) {
    angularDampingOverride=0.9999
    angularSpringConstant=400
    bAngularSpring=true
    bGravityOverrideInSimSpace=true
    bOverrideAngularDamping=true
    bUseGravityOverride=true
    simulationSpace=BoneRelative
    alphaBoolBlend {blendOption=Linear}
    boundBone {boneName=dyn_skirt_fr_03_l}
    componentAppliedLinearAccClamp=<5000 5000 5000>
    componentLinearAccScale=<1 1 1>
    componentLinearVelScale=<1 1 1>
    gravityOverride {z=-980}
    physicsBodyDefinitions [{
      sphereCollisionRadius=0
      boundBone {boneName=dyn_skirt_fr_03_l}
      boxExtents=<20 10 10>
      constraintSetup {angularLimitsMax=<10 0 10>, angularLimitsMin=<-10 -15 -10>, angularTarget=<1 0 0>}
      localJointOffset=<1 0 0>
     }]
    relativeSpaceBone {boneName=pelvis}
   }
   n23 = Anim Dynamics - Bone: dyn_skirt_fr_03_r(ComponentPose=@n22, Alpha=1) {
    angularDampingOverride=0.9999
    angularSpringConstant=400
    bAngularSpring=true
    bGravityOverrideInSimSpace=true
    bOverrideAngularDamping=true
    bUseGravityOverride=true
    simulationSpace=BoneRelative
    alphaBoolBlend {blendOption=Linear}
    boundBone {boneName=dyn_skirt_fr_03_r}
    componentAppliedLinearAccClamp=<5000 5000 5000>
    componentLinearAccScale=<1 1 1>
    componentLinearVelScale=<1 1 1>
    gravityOverride {z=-980}
    physicsBodyDefinitions [{
      sphereCollisionRadius=0
      boundBone {boneName=dyn_skirt_fr_03_r}
      boxExtents=<20 20 10>
      constraintSetup {angularLimitsMax=<10 0 10>, angularLimitsMin=<-10 -15 -10>, angularTarget=<1 0 0>}
      localJointOffset=<1 0 0>
     }]
    relativeSpaceBone {boneName=pelvis}
   }
   n24 = Anim Dynamics - Bone: dyn_skirt_fr2_03_l(ComponentPose=@n23, Alpha=1) {
    angularDampingOverride=0.9999
    angularSpringConstant=400
    bAngularSpring=true
    bGravityOverrideInSimSpace=true
    bOverrideAngularDamping=true
    bUseGravityOverride=true
    simulationSpace=BoneRelative
    alphaBoolBlend {blendOption=Linear}
    boundBone {boneName=dyn_skirt_fr2_03_l}
    componentAppliedLinearAccClamp=<5000 5000 5000>
    componentLinearAccScale=<1 1 1>
    componentLinearVelScale=<1 1 1>
    gravityOverride {z=-980}
    physicsBodyDefinitions [{
      sphereCollisionRadius=0
      boundBone {boneName=dyn_skirt_fr2_03_l}
      boxExtents=<20 10 10>
      constraintSetup {angularLimitsMax=<10 0 10>, angularLimitsMin=<-10 -15 -10>, angularTarget=<1 0 0>}
      localJointOffset=<1 0 0>
     }]
    relativeSpaceBone {boneName=pelvis}
   }
   n25 = Anim Dynamics - Bone: dyn_skirt_fr2_03_r(ComponentPose=@n24, Alpha=1) {
    angularDampingOverride=0.9999
    angularSpringConstant=400
    bAngularSpring=true
    bGravityOverrideInSimSpace=true
    bOverrideAngularDamping=true
    bUseGravityOverride=true
    simulationSpace=BoneRelative
    alphaBoolBlend {blendOption=Linear}
    boundBone {boneName=dyn_skirt_fr2_03_r}
    componentAppliedLinearAccClamp=<5000 5000 5000>
    componentLinearAccScale=<1 1 1>
    componentLinearVelScale=<1 1 1>
    gravityOverride {z=-980}
    physicsBodyDefinitions [{
      sphereCollisionRadius=0
      boundBone {boneName=dyn_skirt_fr2_03_r}
      boxExtents=<20 20 10>
      constraintSetup {angularLimitsMax=<10 0 10>, angularLimitsMin=<-10 -15 -10>, angularTarget=<1 0 0>}
      localJointOffset=<1 0 0>
     }]
    relativeSpaceBone {boneName=pelvis}
   }
   n26 = Anim Dynamics - Bone: dyn_skirt_03_l(ComponentPose=@n25, Alpha=1) {
    angularDampingOverride=0.9999
    angularSpringConstant=400
    bAngularSpring=true
    bGravityOverrideInSimSpace=true
    bOverrideAngularDamping=true
    bUseGravityOverride=true
    simulationSpace=BoneRelative
    alphaBoolBlend {blendOption=Linear}
    boundBone {boneName=dyn_skirt_03_l}
    componentAppliedLinearAccClamp=<5000 5000 5000>
    componentLinearAccScale=<1 1 1>
    componentLinearVelScale=<1 1 1>
    gravityOverride {z=-980}
    physicsBodyDefinitions [{
      sphereCollisionRadius=0
      boundBone {boneName=dyn_skirt_03_l}
      boxExtents=<20 10 10>
      constraintSetup {angularLimitsMax=<10 0 10>, angularLimitsMin=<-10 -15 -10>, angularTarget=<1 0 0>}
      localJointOffset=<1 0 0>
     }]
    relativeSpaceBone {boneName=pelvis}
   }
   n27 = Anim Dynamics - Bone: dyn_skirt_03_r(ComponentPose=@n26, Alpha=1) {
    angularDampingOverride=0.9999
    angularSpringConstant=400
    bAngularSpring=true
    bGravityOverrideInSimSpace=true
    bOverrideAngularDamping=true
    bUseGravityOverride=true
    simulationSpace=BoneRelative
    alphaBoolBlend {blendOption=Linear}
    boundBone {boneName=dyn_skirt_03_r}
    componentAppliedLinearAccClamp=<5000 5000 5000>
    componentLinearAccScale=<1 1 1>
    componentLinearVelScale=<1 1 1>
    gravityOverride {z=-980}
    physicsBodyDefinitions [{
      sphereCollisionRadius=0
      boundBone {boneName=dyn_skirt_03_r}
      boxExtents=<20 20 10>
      constraintSetup {angularLimitsMax=<10 0 10>, angularLimitsMin=<-10 -15 -10>, angularTarget=<1 0 0>}
      localJointOffset=<1 0 0>
     }]
    relativeSpaceBone {boneName=pelvis}
   }
   n28 = Anim Dynamics - Bone: dyn_belt_gun_01(ComponentPose=@n27, Alpha=1) {
    angularDampingOverride=0.9999
    angularSpringConstant=400
    bAngularSpring=true
    bGravityOverrideInSimSpace=true
    bOverrideAngularDamping=true
    bUseGravityOverride=true
    bUsePlanarLimit=false
    simulationSpace=BoneRelative
    alphaBoolBlend {blendOption=Linear}
    boundBone {boneName=dyn_belt_gun_01}
    componentAppliedLinearAccClamp=<5000 5000 5000>
    componentLinearAccScale=<1 1 1>
    componentLinearVelScale=<1 1 1>
    gravityOverride {z=-980}
    physicsBodyDefinitions [{
      sphereCollisionRadius=0
      boundBone {boneName=dyn_belt_gun_01}
      boxExtents=<10 20 20>
      constraintSetup {angularLimitsMax=<25 15 50>, angularLimitsMin=<-25 -15 -50>, angularTarget=<1 0 0>}
      localJointOffset=<1 0 0>
     }]
    relativeSpaceBone {boneName=pelvis}
   }
   n29 = Anim Dynamics - Bone: dyn_hipHeart(ComponentPose=@n28, Alpha=0.5) {
    angularDampingOverride=0.9999
    angularSpringConstant=400
    bAngularSpring=true
    bGravityOverrideInSimSpace=true
    bOverrideAngularDamping=true
    bUseGravityOverride=true
    bUsePlanarLimit=false
    simulationSpace=BoneRelative
    alphaBoolBlend {blendOption=Linear}
    boundBone {boneName=dyn_hipHeart}
    componentAppliedLinearAccClamp=<5000 5000 5000>
    componentLinearAccScale=<1 1 1>
    componentLinearVelScale=<1 1 1>
    gravityOverride {z=-980}
    physicsBodyDefinitions [{
      sphereCollisionRadius=0
      boundBone {boneName=dyn_hipHeart}
      boxExtents=<20 20 10>
      constraintSetup {angularLimitsMax=<25 0 25>, angularLimitsMin=<-25 -25 -25>, angularTarget=<1 0 0>}
      localJointOffset=<1 0 0>
     }]
    relativeSpaceBone {boneName=pelvis}
   }
   n30 = Component To Local(ComponentPose=@n29)
   n31 = "Transform (Modify) Bone - Bone: dyn_skirt_fr_01_MSTR"!(
    ComponentPose=Local To Component(LocalPose=Blend Poses by bool(BlendPose_0=@n30, BlendPose_1=@n15) {blendType=Linear})
    bAlphaBoolEnabled:bool=True
    Rotation="10,0,0"
   ) -> (Pose) {
    alphaInputType=Bool
    rotationMode=BMM_Additive
    rotationSpace=BCS_BoneSpace
    scaleMode=BMM_Additive
    scaleSpace=BCS_BoneSpace
    translationMode=BMM_Additive
    translationSpace=BCS_BoneSpace
    alphaBoolBlend {blendInTime=0.2, blendOption=Sinusoidal, blendOutTime=0.2}
    boneToModify {boneName=dyn_skirt_fr_01_MSTR}
    rotation {pitch=10}
   }
   Output Pose(Result=Component To Local(ComponentPose=@n31)) {layerGroup=None}
   n32 = RigidBody(ComponentPose=Local To Component(LocalPose=Copy Pose From Mesh() {bUseAttachedParent=true})) {
    bClampLinearTranslationLimitToRefPose=true
    bOverrideWorldGravity=true
    overridePhysicsAsset=&/F_MED_Crush_Zipper_Physics
    simulationSpace=BaseBoneSpace
    alphaBoolBlend {blendOption=Linear}
    baseBoneRef {boneName=root}
    componentAppliedLinearAccClamp=<5000 5000 5000>
    overrideWorldGravity {z=-250}
   }
   n33 = "Constraint - dyn_thigh_01_l (RR)"!(ComponentPose=@n32, Alpha=1, ConstraintWeights_0=1, ConstraintWeights_1:float=0) -> (Pose) {
    alphaBoolBlend {blendOption=Linear}
    boneToModify {boneName=dyn_thigh_01_l}
    constraintSetup [{transformType=Rotation, targetBone {boneName=thigh_l}}, {transformType=Rotation, targetBone {boneName=pelvis}}]
    constraintWeights [1,1]
   }
   n34 = "Constraint - dyn_thigh_01_r (RR)"!(ComponentPose=@n33, Alpha=1, ConstraintWeights_0=1, ConstraintWeights_1:float=0) -> (Pose) {
    alphaBoolBlend {blendOption=Linear}
    boneToModify {boneName=dyn_thigh_01_r}
    constraintSetup [{transformType=Rotation, targetBone {boneName=thigh_r}}, {transformType=Rotation, targetBone {boneName=pelvis}}]
    constraintWeights [1,1]
   }
   n35 = "Constraint - dyn_skirt_fr_01_MSTR (RR)"!(ComponentPose=@n34, Alpha=1, ConstraintWeights_0=0.5, ConstraintWeights_1:float=0.5) -> (Pose) {
    alphaBoolBlend {blendOption=Linear}
    boneToModify {boneName=dyn_skirt_fr_01_MSTR}
    constraintSetup [
     {transformType=Rotation, targetBone {boneName=dyn_skirt_fr_01_l_MSTR}}
     {transformType=Rotation, targetBone {boneName=dyn_skirt_fr_01_r_MSTR}}
    ]
    constraintWeights [1,1]
   }
   n36 = "Constraint - dyn_skirt_fr_01 (RR)"!(ComponentPose=@n35, Alpha=1, ConstraintWeights_0=0.5, ConstraintWeights_1:float=0.5) -> (Pose) {
    alphaBoolBlend {blendOption=Linear}
    boneToModify {boneName=dyn_skirt_fr_01}
    constraintSetup [{transformType=Rotation, targetBone {boneName=dyn_skirt_fr_01_l}}, {transformType=Rotation, targetBone {boneName=dyn_skirt_fr_01_r}}]
    constraintWeights [1,1]
   }
   n37 = "Constraint - dyn_skirt_fr_02 (RR)"!(ComponentPose=@n36, Alpha=1, ConstraintWeights_0=0.5, ConstraintWeights_1:float=0.5) -> (Pose) {
    alphaBoolBlend {blendOption=Linear}
    boneToModify {boneName=dyn_skirt_fr_02}
    constraintSetup [{transformType=Rotation, targetBone {boneName=dyn_skirt_fr_02_l}}, {transformType=Rotation, targetBone {boneName=dyn_skirt_fr_02_r}}]
    constraintWeights [1,1]
   }
   n38 = "Constraint - dyn_skirt_fr_03 (RR)"!(ComponentPose=@n37, Alpha=1, ConstraintWeights_0=0.5, ConstraintWeights_1:float=0.5) -> (Pose) {
    alphaBoolBlend {blendOption=Linear}
    boneToModify {boneName=dyn_skirt_fr_03}
    constraintSetup [{transformType=Rotation, targetBone {boneName=dyn_skirt_fr_03_l}}, {transformType=Rotation, targetBone {boneName=dyn_skirt_fr_03_r}}]
    constraintWeights [1,1]
   }
   n39 = "Transform (Modify) Bone - Bone: dyn_shinkHips"(ComponentPose=@n38, Alpha=1, Translation="0,0,0", Rotation="0,0,0", Scale="0.1,0.1,0.1") {
    rotationMode=BMM_Additive
    scaleMode=BMM_Additive
    translationMode=BMM_Additive
    alphaBoolBlend {blendOption=Linear}
    boneToModify {boneName=dyn_shinkHips}
    scale=<0.1 0.1 0.1>
   }
   n40 = "Constraint - dyn_skirt_01_l (P)"(ComponentPose=@n39, Alpha=1, ConstraintWeights_0=1) {
    alphaBoolBlend {blendOption=Linear}
    boneToModify {boneName=dyn_skirt_01_l}
    constraintSetup [{targetBone {boneName=dyn_belt_gun_01}}]
    constraintWeights [1]
   }
   Save cached pose 'PreDressRBAN'(Pose=Component To Local(ComponentPose=@n40)) {cacheName=PreDressRBAN}
   n41 = RigidBody!(ComponentPose=Local To Component(LocalPose=Use cached pose 'PreDressRBAN'()), Alpha, OverrideWorldGravity:Vector="0,0,0", ExternalForce) -> (Pose) {
    bClampLinearTranslationLimitToRefPose=true
    bOverrideWorldGravity=true
    overridePhysicsAsset=&/F_MED_Crush_Skirt_msLow_Physics
    simulationSpace=BaseBoneSpace
    alphaBoolBlend {blendOption=Linear}
    baseBoneRef {boneName=root}
    componentAppliedLinearAccClamp=<3000 3000 1000>
    componentLinearAccScale=<0.4 0.4 0.42>
    overrideWorldGravity {z=-980}
   }
   n42 = "Constraint - dyn_skirt_bk_01 (P)"(ComponentPose=@n41, Alpha=1, ConstraintWeights_0=1) {
    alphaBoolBlend {blendOption=Linear}
    boneToModify {boneName=dyn_skirt_bk_01}
    constraintSetup [{targetBone {boneName=pelvis}}]
    constraintWeights [1]
   }
   n43 = "Constraint - dyn_skirt_bk_02 (P)"(ComponentPose=@n42, Alpha=0.3, ConstraintWeights_0=1) {
    alphaBoolBlend {blendOption=Linear}
    boneToModify {boneName=dyn_skirt_bk_02}
    constraintSetup [{targetBone {boneName=pelvis}}]
    constraintWeights [1]
   }
   n44 = "Constraint - dyn_skirt_bk_03 (P)"(ComponentPose=@n43, Alpha=0, ConstraintWeights_0=1) {
    alphaBoolBlend {blendOption=Linear}
    boneToModify {boneName=dyn_skirt_bk_03}
    constraintSetup [{targetBone {boneName=pelvis}}]
    constraintWeights [1]
   }
   n45 = "Constraint - dyn_skirt_bk2_01_l (P)"(ComponentPose=@n44, Alpha=1, ConstraintWeights_0=1) {
    alphaBoolBlend {blendOption=Linear}
    boneToModify {boneName=dyn_skirt_bk2_01_l}
    constraintSetup [{targetBone {boneName=pelvis}}]
    constraintWeights [1]
   }
   n46 = "Constraint - dyn_skirt_bk2_02_l (P)"(ComponentPose=@n45, Alpha=0.3, ConstraintWeights_0=1) {
    alphaBoolBlend {blendOption=Linear}
    boneToModify {boneName=dyn_skirt_bk2_02_l}
    constraintSetup [{targetBone {boneName=pelvis}}]
    constraintWeights [1]
   }
   n47 = "Constraint - dyn_skirt_bk2_03_l (P)"(ComponentPose=@n46, Alpha=0, ConstraintWeights_0=1) {
    alphaBoolBlend {blendOption=Linear}
    boneToModify {boneName=dyn_skirt_bk2_03_l}
    constraintSetup [{targetBone {boneName=pelvis}}]
    constraintWeights [1]
   }
   n48 = "Constraint - dyn_skirt_bk_01_l (P)"(ComponentPose=@n47, Alpha=1, ConstraintWeights_0=1) {
    alphaBoolBlend {blendOption=Linear}
    boneToModify {boneName=dyn_skirt_bk_01_l}
    constraintSetup [{targetBone {boneName=pelvis}}]
    constraintWeights [1]
   }
   n49 = "Constraint - dyn_skirt_bk_02_l (P)"(ComponentPose=@n48, Alpha=0.3, ConstraintWeights_0=1) {
    alphaBoolBlend {blendOption=Linear}
    boneToModify {boneName=dyn_skirt_bk_02_l}
    constraintSetup [{targetBone {boneName=pelvis}}]
    constraintWeights [1]
   }
   n50 = "Constraint - dyn_skirt_bk_03_l (P)"(ComponentPose=@n49, Alpha=0, ConstraintWeights_0=1) {
    alphaBoolBlend {blendOption=Linear}
    boneToModify {boneName=dyn_skirt_bk_03_l}
    constraintSetup [{targetBone {boneName=pelvis}}]
    constraintWeights [1]
   }
   n51 = "Constraint - dyn_skirt_bk_01_r (P)"(ComponentPose=@n50, Alpha=1, ConstraintWeights_0=1) {
    alphaBoolBlend {blendOption=Linear}
    boneToModify {boneName=dyn_skirt_bk_01_r}
    constraintSetup [{targetBone {boneName=pelvis}}]
    constraintWeights [1]
   }
   n52 = "Constraint - dyn_skirt_bk_02_r (P)"(ComponentPose=@n51, Alpha=0.3, ConstraintWeights_0=1) {
    alphaBoolBlend {blendOption=Linear}
    boneToModify {boneName=dyn_skirt_bk_02_r}
    constraintSetup [{targetBone {boneName=pelvis}}]
    constraintWeights [1]
   }
   n53 = "Constraint - dyn_skirt_bk_03_r (P)"(ComponentPose=@n52, Alpha=0, ConstraintWeights_0=1) {
    alphaBoolBlend {blendOption=Linear}
    boneToModify {boneName=dyn_skirt_bk_03_r}
    constraintSetup [{targetBone {boneName=pelvis}}]
    constraintWeights [1]
   }
   n54 = "Constraint - dyn_skirt_bk2_01_r (P)"(ComponentPose=@n53, Alpha=1, ConstraintWeights_0=1) {
    alphaBoolBlend {blendOption=Linear}
    boneToModify {boneName=dyn_skirt_bk2_01_r}
    constraintSetup [{targetBone {boneName=pelvis}}]
    constraintWeights [1]
   }
   n55 = "Constraint - dyn_skirt_bk2_02_r (P)"(ComponentPose=@n54, Alpha=0.3, ConstraintWeights_0=1) {
    alphaBoolBlend {blendOption=Linear}
    boneToModify {boneName=dyn_skirt_bk2_02_r}
    constraintSetup [{targetBone {boneName=pelvis}}]
    constraintWeights [1]
   }
   n56 = "Constraint - dyn_skirt_bk2_03_r (P)"(ComponentPose=@n55, Alpha=0, ConstraintWeights_0=1) {
    alphaBoolBlend {blendOption=Linear}
    boneToModify {boneName=dyn_skirt_bk2_03_r}
    constraintSetup [{targetBone {boneName=pelvis}}]
    constraintWeights [1]
   }
   n57 = Component To Local(ComponentPose=@n56)
   n58 = RigidBody(ComponentPose=Local To Component(LocalPose=Use cached pose 'PreDressRBAN'()), OverrideWorldGravity="0,0,0") {
    bClampLinearTranslationLimitToRefPose=true
    bOverrideWorldGravity=true
    overridePhysicsAsset=&F_MED_Crush_Skirt_msLow_Physics
    simulationSpace=BaseBoneSpace
    alphaBoolBlend {blendOption=Linear}
    baseBoneRef {boneName=pelvis}
    componentAppliedLinearAccClamp=<3000 3000 850>
    componentLinearAccScale=<0.4 0.4 0.42>
    overrideWorldGravity {z=-980}
   }
   n59 = RigidBody(ComponentPose=Local To Component(LocalPose=Use cached pose 'PreDressRBAN'()), OverrideWorldGravity="0,0,0") {
    bClampLinearTranslationLimitToRefPose=true
    bOverrideWorldGravity=true
    overridePhysicsAsset=&/F_MED_Crush_Skirt_FE_Physics
    simulationSpace=BaseBoneSpace
    alphaBoolBlend {blendInTime=0.1, blendOption=Linear, blendOutTime=0.1}
    alphaScaleBiasClamp {bInterpResult=true}
    baseBoneRef {boneName=pelvis}
    componentAppliedLinearAccClamp=<3000 3000 1000>
    componentLinearAccScale=<0.4 0.4 0.42>
    overrideWorldGravity {z=-980}
   }
   n60 = Blend Poses by bool(BlendPose_0=Component To Local(ComponentPose=@n59), BlendPose_1=Component To Local(ComponentPose=@n58), BlendTime_0=0.2, BlendTime_1=0.2) {blendType=Linear}
   n61 = "Constraint - dyn_skirt_fr_01_l_MSTR (PP)"!(
    ComponentPose=Local To Component(LocalPose=Blend Poses by bool(BlendPose_0=@n57, BlendPose_1=@n60, BlendTime_0=0.01, BlendTime_1=0.01) {blendType=Linear})
    Alpha=0.9
    ConstraintWeights_0=0
    ConstraintWeights_1:float=1
   ) -> (Pose) {
    alphaBoolBlend {blendOption=Linear}
    boneToModify {boneName=dyn_skirt_fr_01_l_MSTR}
    constraintSetup [{targetBone {boneName=pelvis}}, {targetBone {boneName=dyn_thigh_01_l}}]
    constraintWeights [1,1]
   }
   n62 = "Constraint - dyn_skirt_fr_01_l (PP)"!(ComponentPose=@n61, Alpha=0.9, ConstraintWeights_0=0, ConstraintWeights_1:float=1) -> (Pose) {
    alphaBoolBlend {blendOption=Linear}
    boneToModify {boneName=dyn_skirt_fr_01_l}
    constraintSetup [{targetBone {boneName=pelvis}}, {targetBone {boneName=dyn_thigh_01_l}}]
    constraintWeights [1,1]
   }
   n63 = "Constraint - dyn_skirt_fr_02_l (PP)"!(ComponentPose=@n62, Alpha=0.8, ConstraintWeights_0=0, ConstraintWeights_1:float=1) -> (Pose) {
    alphaBoolBlend {blendOption=Linear}
    boneToModify {boneName=dyn_skirt_fr_02_l}
    constraintSetup [{targetBone {boneName=pelvis}}, {targetBone {boneName=dyn_thigh_01_l}}]
    constraintWeights [1,1]
   }
   n64 = "Constraint - dyn_skirt_fr_03_l (PP)"!(ComponentPose=@n63, Alpha=0.5, ConstraintWeights_0=0, ConstraintWeights_1:float=1) -> (Pose) {
    alphaBoolBlend {blendOption=Linear}
    boneToModify {boneName=dyn_skirt_fr_03_l}
    constraintSetup [{targetBone {boneName=pelvis}}, {targetBone {boneName=dyn_thigh_01_l}}]
    constraintWeights [1,1]
   }
   n65 = "Constraint - dyn_skirt_fr2_01_l_MSTR (PP)"!(ComponentPose=@n64, Alpha=0.9, ConstraintWeights_0=0, ConstraintWeights_1:float=1) -> (Pose) {
    alphaBoolBlend {blendOption=Linear}
    boneToModify {boneName=dyn_skirt_fr2_01_l_MSTR}
    constraintSetup [{targetBone {boneName=pelvis}}, {targetBone {boneName=dyn_thigh_01_l}}]
    constraintWeights [1,1]
   }
   n66 = "Constraint - dyn_skirt_fr2_01_l (PP)"!(ComponentPose=@n65, Alpha=0.9, ConstraintWeights_0=0, ConstraintWeights_1:float=1) -> (Pose) {
    alphaBoolBlend {blendOption=Linear}
    boneToModify {boneName=dyn_skirt_fr2_01_l}
    constraintSetup [{targetBone {boneName=pelvis}}, {targetBone {boneName=dyn_thigh_01_l}}]
    constraintWeights [1,1]
   }
   n67 = "Constraint - dyn_skirt_fr2_02_l (PP)"!(ComponentPose=@n66, Alpha=0.8, ConstraintWeights_0=0, ConstraintWeights_1:float=1) -> (Pose) {
    alphaBoolBlend {blendOption=Linear}
    boneToModify {boneName=dyn_skirt_fr2_02_l}
    constraintSetup [{targetBone {boneName=pelvis}}, {targetBone {boneName=dyn_thigh_01_l}}]
    constraintWeights [1,1]
   }
   n68 = "Constraint - dyn_skirt_fr2_03_l (PP)"!(ComponentPose=@n67, Alpha=0.5, ConstraintWeights_0=0, ConstraintWeights_1:float=1) -> (Pose) {
    alphaBoolBlend {blendOption=Linear}
    boneToModify {boneName=dyn_skirt_fr2_03_l}
    constraintSetup [{targetBone {boneName=pelvis}}, {targetBone {boneName=dyn_thigh_01_l}}]
    constraintWeights [1,1]
   }
   n69 = "Constraint - dyn_skirt_fr_01_r_MSTR (PP)"!(ComponentPose=@n68, Alpha=0.9, ConstraintWeights_0=0, ConstraintWeights_1:float=1) -> (Pose) {
    alphaBoolBlend {blendOption=Linear}
    boneToModify {boneName=dyn_skirt_fr_01_r_MSTR}
    constraintSetup [{targetBone {boneName=pelvis}}, {targetBone {boneName=dyn_thigh_01_r}}]
    constraintWeights [1,1]
   }
   n70 = "Constraint - dyn_skirt_fr_01_r (PP)"!(ComponentPose=@n69, Alpha=0.9, ConstraintWeights_0=0, ConstraintWeights_1:float=1) -> (Pose) {
    alphaBoolBlend {blendOption=Linear}
    boneToModify {boneName=dyn_skirt_fr_01_r}
    constraintSetup [{targetBone {boneName=pelvis}}, {targetBone {boneName=dyn_thigh_01_r}}]
    constraintWeights [1,1]
   }
   n71 = "Constraint - dyn_skirt_fr_02_r (PP)"!(ComponentPose=@n70, Alpha=0.5, ConstraintWeights_0=0, ConstraintWeights_1:float=1) -> (Pose) {
    alphaBoolBlend {blendOption=Linear}
    boneToModify {boneName=dyn_skirt_fr_02_r}
    constraintSetup [{targetBone {boneName=pelvis}}, {targetBone {boneName=dyn_thigh_01_r}}]
    constraintWeights [1,1]
   }
   n72 = "Constraint - dyn_skirt_fr_03_r (PP)"!(ComponentPose=@n71, Alpha=0.5, ConstraintWeights_0=0, ConstraintWeights_1:float=1) -> (Pose) {
    alphaBoolBlend {blendOption=Linear}
    boneToModify {boneName=dyn_skirt_fr_03_r}
    constraintSetup [{targetBone {boneName=pelvis}}, {targetBone {boneName=dyn_thigh_01_r}}]
    constraintWeights [1,1]
   }
   n73 = "Constraint - dyn_skirt_fr2_01_r_MSTR (PP)"!(ComponentPose=@n72, Alpha=0.9, ConstraintWeights_0=0, ConstraintWeights_1:float=1) -> (Pose) {
    alphaBoolBlend {blendOption=Linear}
    boneToModify {boneName=dyn_skirt_fr2_01_r_MSTR}
    constraintSetup [{targetBone {boneName=pelvis}}, {targetBone {boneName=dyn_thigh_01_r}}]
    constraintWeights [1,1]
   }
   n74 = "Constraint - dyn_skirt_fr2_01_r (PP)"!(ComponentPose=@n73, Alpha=0.9, ConstraintWeights_0=0, ConstraintWeights_1:float=1) -> (Pose) {
    alphaBoolBlend {blendOption=Linear}
    boneToModify {boneName=dyn_skirt_fr2_01_r}
    constraintSetup [{targetBone {boneName=pelvis}}, {targetBone {boneName=dyn_thigh_01_r}}]
    constraintWeights [1,1]
   }
   n75 = "Constraint - dyn_skirt_fr2_02_r (PP)"!(ComponentPose=@n74, Alpha=0.5, ConstraintWeights_0=0, ConstraintWeights_1:float=1) -> (Pose) {
    alphaBoolBlend {blendOption=Linear}
    boneToModify {boneName=dyn_skirt_fr2_02_r}
    constraintSetup [{targetBone {boneName=pelvis}}, {targetBone {boneName=dyn_thigh_01_r}}]
    constraintWeights [1,1]
   }
   n76 = "Constraint - dyn_skirt_fr2_03_r (PP)"!(ComponentPose=@n75, Alpha=0.5, ConstraintWeights_0=0, ConstraintWeights_1:float=1) -> (Pose) {
    alphaBoolBlend {blendOption=Linear}
    boneToModify {boneName=dyn_skirt_fr2_03_r}
    constraintSetup [{targetBone {boneName=pelvis}}, {targetBone {boneName=dyn_thigh_01_r}}]
    constraintWeights [1,1]
   }
   n77 = "Constraint - dyn_skirt_fr_01_MSTR (PP)"!(ComponentPose=@n76, Alpha=0.9, ConstraintWeights_0=0.5, ConstraintWeights_1:float=0.5) -> (Pose) {
    alphaBoolBlend {blendOption=Linear}
    boneToModify {boneName=dyn_skirt_fr_01_MSTR}
    constraintSetup [{targetBone {boneName=dyn_skirt_fr_01_l_MSTR}}, {targetBone {boneName=dyn_skirt_fr_01_r_MSTR}}]
    constraintWeights [1,1]
   }
   n78 = "Constraint - dyn_skirt_fr_01 (PP)"!(ComponentPose=@n77, Alpha=0.9, ConstraintWeights_0=0.5, ConstraintWeights_1:float=0.5) -> (Pose) {
    alphaBoolBlend {blendOption=Linear}
    boneToModify {boneName=dyn_skirt_fr_01}
    constraintSetup [{targetBone {boneName=dyn_skirt_fr_01_l}}, {targetBone {boneName=dyn_skirt_fr_01_r}}]
    constraintWeights [1,1]
   }
   n79 = "Constraint - dyn_skirt_fr_02 (PP)"!(ComponentPose=@n78, Alpha=0.6, ConstraintWeights_0=0.5, ConstraintWeights_1:float=0.5) -> (Pose) {
    alphaBoolBlend {blendOption=Linear}
    boneToModify {boneName=dyn_skirt_fr_02}
    constraintSetup [{targetBone {boneName=dyn_skirt_fr_02_l}}, {targetBone {boneName=dyn_skirt_fr_02_r}}]
    constraintWeights [1,1]
   }
   n80 = "Constraint - dyn_skirt_fr_03 (PP)"!(ComponentPose=@n79, Alpha=0.5, ConstraintWeights_0=0.5, ConstraintWeights_1:float=0.5) -> (Pose) {
    alphaBoolBlend {blendOption=Linear}
    boneToModify {boneName=dyn_skirt_fr_03}
    constraintSetup [{targetBone {boneName=dyn_skirt_fr_03_l}}, {targetBone {boneName=dyn_skirt_fr_03_r}}]
    constraintWeights [1,1]
   }
   n81 = "Constraint - dyn_skirt_01_l_MSTR (P)"(ComponentPose=@n80, Alpha=0.5, ConstraintWeights_0=1) {
    alphaBoolBlend {blendOption=Linear}
    boneToModify {boneName=dyn_skirt_01_l_MSTR}
    constraintSetup [{targetBone {boneName=dyn_skirt_fr2_01_l_MSTR}}]
    constraintWeights [1]
   }
   n82 = "Constraint - dyn_skirt_01_l (P)"(ComponentPose=@n81, Alpha=0.5, ConstraintWeights_0=1) {
    alphaBoolBlend {blendOption=Linear}
    boneToModify {boneName=dyn_skirt_01_l}
    constraintSetup [{targetBone {boneName=dyn_skirt_fr2_01_l}}]
    constraintWeights [1]
   }
   n83 = "Constraint - dyn_skirt_02_l (P)"(ComponentPose=@n82, Alpha=0.5, ConstraintWeights_0=1) {
    alphaBoolBlend {blendOption=Linear}
    boneToModify {boneName=dyn_skirt_02_l}
    constraintSetup [{targetBone {boneName=dyn_skirt_fr2_02_l}}]
    constraintWeights [1]
   }
   n84 = "Constraint - dyn_skirt_03_l (P)"(ComponentPose=@n83, Alpha=0.5, ConstraintWeights_0=1) {
    alphaBoolBlend {blendOption=Linear}
    boneToModify {boneName=dyn_skirt_03_l}
    constraintSetup [{targetBone {boneName=dyn_skirt_fr2_03_l}}]
    constraintWeights [1]
   }
   n85 = "Constraint - dyn_skirt_bk2_01_l_MSTR (P)"(ComponentPose=@n84, Alpha=0.5, ConstraintWeights_0=1) {
    alphaBoolBlend {blendOption=Linear}
    boneToModify {boneName=dyn_skirt_bk2_01_l_MSTR}
    constraintSetup [{targetBone {boneName=dyn_skirt_01_l_MSTR}}]
    constraintWeights [1]
   }
   n86 = "Constraint - dyn_skirt_bk2_01_l (P)"(ComponentPose=@n85, Alpha=0.5, ConstraintWeights_0=1) {
    alphaBoolBlend {blendOption=Linear}
    boneToModify {boneName=dyn_skirt_bk2_01_l}
    constraintSetup [{targetBone {boneName=dyn_skirt_01_l}}]
    constraintWeights [1]
   }
   n87 = "Constraint - dyn_skirt_bk2_02_l (P)"(ComponentPose=@n86, Alpha=0.5, ConstraintWeights_0=1) {
    alphaBoolBlend {blendOption=Linear}
    boneToModify {boneName=dyn_skirt_bk2_02_l}
    constraintSetup [{targetBone {boneName=dyn_skirt_02_l}}]
    constraintWeights [1]
   }
   n88 = "Constraint - dyn_skirt_bk2_03_l (P)"(ComponentPose=@n87, Alpha=0.5, ConstraintWeights_0=1) {
    alphaBoolBlend {blendOption=Linear}
    boneToModify {boneName=dyn_skirt_bk2_03_l}
    constraintSetup [{targetBone {boneName=dyn_skirt_03_l}}]
    constraintWeights [1]
   }
   n89 = "Constraint - dyn_skirt_01_r_MSTR (P)"(ComponentPose=@n88, Alpha=0.5, ConstraintWeights_0=1) {
    alphaBoolBlend {blendOption=Linear}
    boneToModify {boneName=dyn_skirt_01_r_MSTR}
    constraintSetup [{targetBone {boneName=dyn_skirt_fr2_01_r_MSTR}}]
    constraintWeights [1]
   }
   n90 = "Constraint - dyn_skirt_01_r (P)"(ComponentPose=@n89, Alpha=0.5, ConstraintWeights_0=1) {
    alphaBoolBlend {blendOption=Linear}
    boneToModify {boneName=dyn_skirt_01_r}
    constraintSetup [{targetBone {boneName=dyn_skirt_fr2_01_r}}]
    constraintWeights [1]
   }
   n91 = "Constraint - dyn_skirt_02_r (P)"(ComponentPose=@n90, Alpha=0.5, ConstraintWeights_0=1) {
    alphaBoolBlend {blendOption=Linear}
    boneToModify {boneName=dyn_skirt_02_r}
    constraintSetup [{targetBone {boneName=dyn_skirt_fr2_02_r}}]
    constraintWeights [1]
   }
   n92 = "Constraint - dyn_skirt_03_r (P)"(ComponentPose=@n91, Alpha=0.5, ConstraintWeights_0=1) {
    alphaBoolBlend {blendOption=Linear}
    boneToModify {boneName=dyn_skirt_03_r}
    constraintSetup [{targetBone {boneName=dyn_skirt_fr2_03_r}}]
    constraintWeights [1]
   }
   n93 = "Constraint - dyn_skirt_bk2_01_r_MSTR (P)"(ComponentPose=@n92, Alpha=0.5, ConstraintWeights_0=1) {
    alphaBoolBlend {blendOption=Linear}
    boneToModify {boneName=dyn_skirt_bk2_01_r_MSTR}
    constraintSetup [{targetBone {boneName=dyn_skirt_01_r_MSTR}}]
    constraintWeights [1]
   }
   n94 = "Constraint - dyn_skirt_bk2_01_r (P)"(ComponentPose=@n93, Alpha=0.5, ConstraintWeights_0=1) {
    alphaBoolBlend {blendOption=Linear}
    boneToModify {boneName=dyn_skirt_bk2_01_r}
    constraintSetup [{targetBone {boneName=dyn_skirt_01_r}}]
    constraintWeights [1]
   }
   n95 = "Constraint - dyn_skirt_bk2_02_r (P)"(ComponentPose=@n94, Alpha=0.5, ConstraintWeights_0=1) {
    alphaBoolBlend {blendOption=Linear}
    boneToModify {boneName=dyn_skirt_bk2_02_r}
    constraintSetup [{targetBone {boneName=dyn_skirt_02_r}}]
    constraintWeights [1]
   }
   n96 = "Constraint - dyn_skirt_bk2_03_r (P)"(ComponentPose=@n95, Alpha=0.5, ConstraintWeights_0=1) {
    alphaBoolBlend {blendOption=Linear}
    boneToModify {boneName=dyn_skirt_bk2_03_r}
    constraintSetup [{targetBone {boneName=dyn_skirt_03_r}}]
    constraintWeights [1]
   }
   n97 = "Transform (Modify) Bone - Bone: dyn_skirt_fr2_01_l_MSTR"(ComponentPose=@n96, Alpha=Get isCrouchingAlpha() -> (isCrouchingAlpha:double), Translation="0,0,1", Rotation="0,0,0", Scale="1,1,1") {
    rotationMode=BMM_Additive
    rotationSpace=BCS_BoneSpace
    translationMode=BMM_Additive
    alphaBoolBlend {blendOption=Linear}
    boneToModify {boneName=dyn_skirt_fr2_01_l_MSTR}
    translation {z=1}
   }
   n98 = "Transform (Modify) Bone - Bone: dyn_skirt_fr2_01_l_MSTR"(ComponentPose=@n97, Alpha=Get isCrouchingAlpha(), Translation="2,0,0", Rotation="0,0,0", Scale="1,1,1") {
    rotationMode=BMM_Additive
    rotationSpace=BCS_BoneSpace
    translationMode=BMM_Additive
    alphaBoolBlend {blendOption=Linear}
    boneToModify {boneName=dyn_skirt_fr2_01_l_MSTR}
    translation {x=2}
   }
   n99 = "Transform (Modify) Bone - Bone: dyn_skirt_bk2_01_r_MSTR"(ComponentPose=@n98, Alpha=1, Translation="-2,-1.5,0", Rotation="0,0,0", Scale="1,1,1") {
    rotationMode=BMM_Additive
    scaleMode=BMM_Additive
    translationMode=BMM_Additive
    alphaBoolBlend {blendOption=Linear}
    boneToModify {boneName=dyn_skirt_bk2_01_r_MSTR}
    translation=<-2 -1.5>
   }
   n100 = "Transform (Modify) Bone - Bone: dyn_skirt_fr_01_MSTR"!(ComponentPose=@n99, Alpha:float=Get isCrouchingAlpha(), Translation:Vector="0,0,0", Rotation="5,0,0", Scale:Vector="1,1,1") -> (Pose) {
    rotationMode=BMM_Additive
    rotationSpace=BCS_BoneSpace
    scaleMode=BMM_Additive
    scaleSpace=BCS_BoneSpace
    translationMode=BMM_Additive
    translationSpace=BCS_BoneSpace
    alphaBoolBlend {blendOption=Linear}
    boneToModify {boneName=dyn_skirt_fr_01_MSTR}
    rotation {pitch=5}
   }
   n101 = "Transform (Modify) Bone - Bone: dyn_skirt_01_r_MSTR"(ComponentPose=@n100, Alpha=Get isCrouchingAlpha(), Translation="0,0,0", Rotation="-5,0,0", Scale="1,1,1") {
    rotationMode=BMM_Additive
    scaleMode=BMM_Additive
    translationMode=BMM_Additive
    alphaBoolBlend {blendOption=Linear}
    boneToModify {boneName=dyn_skirt_01_r_MSTR}
    rotation {pitch=-5}
   }
   n102 = "Constraint - dyn_belt_gun_01 (PPPPPPPP)"!(
    ComponentPose=@n101
    Alpha=1
    ConstraintWeights_0=0.0157
    ConstraintWeights_1:float=0.0128
    ConstraintWeights_2:float=0.6203
    ConstraintWeights_3:float=0.0516
    ConstraintWeights_4:float=0.0188
    ConstraintWeights_5:float=0.1322
    ConstraintWeights_6:float=0.1378
    ConstraintWeights_7:float=0.0108
   ) -> (Pose) {
    alphaBoolBlend {blendOption=Linear}
    boneToModify {boneName=dyn_belt_gun_01}
    constraintSetup [
     {targetBone {boneName=dyn_skirt_fr2_01_l}}
     {targetBone {boneName=dyn_skirt_bk2_02_l}}
     {targetBone {boneName=dyn_skirt_01_l}}
     {targetBone {boneName=dyn_skirt_bk_01_l}}
     {targetBone {boneName=dyn_skirt_02_l}}
     {targetBone {boneName=pelvis}}
     {targetBone {boneName=dyn_skirt_bk2_01_l}}
     {targetBone {boneName=dyn_skirt_bk_02_l}}
    ]
    constraintWeights [1,1,1,1,1,1,1,1]
   }
   n103 = "Constraint - dyn_hipHeart (P)"(ComponentPose=@n102, Alpha=1, ConstraintWeights_0=1) {
    alphaBoolBlend {blendOption=Linear}
    boneToModify {boneName=dyn_hipHeart}
    constraintSetup [{targetBone {boneName=dyn_skirt_01_r}}]
    constraintWeights [1]
   }
   n104 = "Constraint - dyn_l_arm_zipper_MSTR (PPPP)"!(
    ComponentPose=@n103
    Alpha=1
    ConstraintWeights_0=0.3002
    ConstraintWeights_1:float=0.2248
    ConstraintWeights_2:float=0.4725
    ConstraintWeights_3:float=0.0025
   ) -> (Pose) {
    alphaBoolBlend {blendOption=Linear}
    boneToModify {boneName=dyn_l_arm_zipper_MSTR}
    constraintSetup [
     {targetBone {boneName=lowerarm_twist_01_l}}
     {targetBone {boneName=lowerarm_l}}
     {targetBone {boneName=lowerarm_twist_02_l}}
     {targetBone {boneName=upperarm_l}}
    ]
    constraintWeights [1,1,1,1]
   }
   n105 = "Constraint - dyn_r_arm_zipper_MSTR (PPPP)"!(
    ComponentPose=@n104
    Alpha=1
    ConstraintWeights_0=0.4777
    ConstraintWeights_1:float=0.0026
    ConstraintWeights_2:float=0.2401
    ConstraintWeights_3:float=0.2796
   ) -> (Pose) {
    alphaBoolBlend {blendOption=Linear}
    boneToModify {boneName=dyn_r_arm_zipper_MSTR}
    constraintSetup [
     {targetBone {boneName=lowerarm_twist_02_r}}
     {targetBone {boneName=upperarm_r}}
     {targetBone {boneName=lowerarm_r}}
     {targetBone {boneName=lowerarm_twist_01_r}}
    ]
    constraintWeights [1,1,1,1]
   }
   n106 = "Transform (Modify) Bone - Bone: dyn_kneeCorrect"!(ComponentPose=@n105, bAlphaBoolEnabled:bool=True, Translation="0,0,0", Rotation="0,0,0", Scale="4,2,2") -> (Pose) {
    alphaInputType=Bool
    rotationMode=BMM_Additive
    scaleMode=BMM_Additive
    translationMode=BMM_Additive
    alphaBoolBlend {blendOption=Linear}
    boneToModify {boneName=dyn_kneeCorrect}
    scale=<4 2 2>
   }
   n107 = "Transform (Modify) Bone - Bone: dyn_kneeCorrect"!(ComponentPose=@n106, Alpha:float=1, Translation="0,0,0", Rotation="0,0,0", Scale="4,2,2") -> (Pose) {
    rotationMode=BMM_Additive
    scaleMode=BMM_Additive
    translationMode=BMM_Additive
    alphaBoolBlend {blendOption=Linear}
    boneToModify {boneName=dyn_kneeCorrect}
    scale=<4 2 2>
   }
   Save cached pose 'Cache_pose_Crush_01'(Pose=Component To Local(ComponentPose=@n107)) {cacheName=Cache_pose_Crush_01}
  ]}
 EventGraph {nodes [
   Event BlueprintUpdateAnimation()
   > n1 = Sequence!(execute) -> (then_0, then_1, then_2:Exec, then_3:Exec, then_4:Exec)
   n2 = Break Fort Anim Input Skydiving External Force(FortAnimInput_SkydivingExternalForce=Get SkydivingExternalForceInput()) -> (
    bUseSkydivingVectorForce=~
    SkydivingVectorForce=~
    HeadToPelvisDirection=~
    FloatingMultiplier=~
    FloatingAdditive=~
    DivingMultiplier=~
    DivingAdditive=~
    ParachutingMultiplier=~
    ParachutingAdditive=~
   )
   > n3 = Set windVector(windVector:Vector=SelectVector(A=vector * float(A=@n2.SkydivingVectorForce, B=0.01), B="0, 0,-980", bPickA=Get bIsSkyDiving())) -> (Output_Get:Vector)
   n4 = GetOwnerSkeletalMeshComponent()
   n5 = BreakRotator(InRot=BreakTransform(InTransform=GetDeltaTransformFromRefPose(self=@n4, BoneName=pelvis, BaseName=root)).Rotation)
   n6 = SelectFloat(A=980, B=-980, bPickA="float > float"(A="Absolute (Float)"(A=@n5.Yaw), B=90) {is=title K2Node_CallFunction})
   > Set finalGravity(finalGravity:Vector=SelectVector(A=MakeVector(Y=0, Z=@n6), B=@n3, bPickA=Get bIsPlayingEmote())) -> (Output_Get:Vector)
   n7 = Gravity Override State Function DEPRECATED(
    execute=@n1.then_1
    CharacterPartAnimInstance=Self-Reference()
    GameTimeInSeconds=GetGameTimeInSeconds()
    GravityOverrideParams=Get Gravity Override Params()
   )
   > Set GravityOverride(GravityOverride=@n7)
   > DrawVectorFromJointFunction!(
    execute
    SkelMeshComponent:Scene Component Object Reference
    JointName:Name=pelvis
    LineEnd:Vector="0, 0, 0"
    JointOffset:Vector="0, 0, 0"
    LineColor:Linear Color Structure=(R=1,G=0,B=0,A=1)
    Thickness:double=2
    Foreground:EDrawDebugSceneDepthPriorityGroup Enum=Foreground
   ) -> (then)
   n8 = Branch(execute=@n1.then_2, Condition=Get bIsPlayingEmote())
   > Set SkirtDynAlpha(SkirtDynAlpha:double=0.5) -> (Output_Get:double)
   Set SkirtDynAlpha(execute=@n8.else, SkirtDynAlpha=1)
   n9 = Get bIsSkyDiving()
   n10 = Set IsSkydiveOrMode(execute=@n1.then_3, IsSkydiveOrMode:bool=OR Boolean(A=@n9, B=Get bIsSkydiveDiveMode())) -> (Output_Get:bool)
   > Set SkyDiveHipAccAlpha(SkyDiveHipAccAlpha:float=SelectFloat(A=0.2, B=1, bPickA=@n9)) -> (Output_Get:float)
   > Set Shrink the hips Alpha(Shrink the hips Alpha:float=SelectFloat(A=0.5, B=1, bPickA=@n10)) -> (Output_Get:float)
   > Set isJumpingOrFalling(isJumpingOrFalling:bool=OR Boolean(A=Get bIsJumping(), B=Get bIsFalling())) -> (Output_Get:bool)
   Event OnParachuteOpened()
   > n11 = Delay(Duration=1.5)
   Branch(execute=[@n11.then, @n12.then, @n13.then, @n14.then, @n15.then, @n17.then, @n18.then, @n20.then], Condition=false)
   > ResetDynamics(InTeleportType=TeleportPhysics)
   > ResetDynamics(InTeleportType=ResetPhysics)
   n12 = Event OnNotifyPartsStarted!() -> (OutputDelegate, then, Context:Name=None) {is=title K2Node_Event}
   n13 = Event OnNotifyPartsStopped!() -> (OutputDelegate, then, Context:Name=None) {is=title K2Node_Event}
   n14 = Event OnBeginSkydiving()
   n15 = Event OnEndSkydiving()
   Event OnBeginCrouching()
   > Set isCrouchingAlpha(isCrouchingAlpha:double=1) -> (Output_Get:double)
   n16 = Event OnEndCrouching()
   Set isCrouchingAlpha(execute=[@n16.then, @n19.then], isCrouchingAlpha=0)
   n17 = Event OnEndPlayingEmote() {is=title K2Node_Event}
   n18 = Event OnBeginPlayingEmote() {is=title K2Node_Event}
   n19 = Event BlueprintInitializeAnimation()
   n20 = Event OnCharacterCustomizationLoaded()
  ]}
]
variables [
 finalGravity {type=Vector}
 Gravity Override Params {
  type=GravityOverrideParamsStruct
  default {
   skydive_DivingDown {gravityOverride=<-5000 50 -500>, windAmplitude=<500 800 -500>, windFrequency=<-12 -12 9>}
   skydive_FloatingDown {gravityOverride=<0 0 -100>, windAmplitude=<100 -90 -120>, windOffset=<0 -60 40>}
   skydive_GoingUp {gravityOverride=<0 0 -100>, windOffset=<0 -600 400>}
   skydive_Parachuting {gravityOverride=<0 0 -980>, windOffset=<0 -600 400>}
   swimming {gravityOverride=<0 0 0>}
  }
 }
 GravityOverride {type=Vector}
 GravityOverride_FE {type=Vector}
 isCrouchingAlpha {type=number}
 isJumpingOrFalling {type=boolean, default=true}
 IsSkydiveOrMode {type=boolean}
 Shrink the hips Alpha {type=number, default=1}
 SkirtDynAlpha {type=number}
 SkyDiveHipAccAlpha {type=number, default=1}
 windVector {type=Vector}
]
