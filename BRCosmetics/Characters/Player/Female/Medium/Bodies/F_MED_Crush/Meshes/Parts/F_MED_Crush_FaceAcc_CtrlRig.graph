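# Control Rig for the F_MED_Crush face accessory skeleton that applies corrective offsets to hair bones as the head turns.
# Head rotation relative to neck_01 becomes Euler angles, whose negative Z range feeds remaps driving additive global bone transforms.
# Right shoulder hair master and its A_07 bone get translation offsets.
# All rotation remaps and every dyn_hairBack_D_MSTR remap target zero, so those have no effect.
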ControlRigBlueprint /BRCosmetics/Characters/Player/Female/Medium/Bodies/F_MED_Crush/Meshes/Parts/F_MED_Crush_FaceAcc_CtrlRig
graphs [
 RigVMFunctionLibrary {links [], nodes []}
 RigVMModel {nodes [
   RigUnit_BeginExecution = &RigUnit_BeginExecution()
   n1 = &RigVMFunction_MathQuaternionToEuler(RotationOrder=ZYX, Value=&RigUnit_GetRelativeBoneTransform(Bone=head, Space=neck_01).Transform.Rotation)
   /Script/ControlRig.RigUnit_ModifyBoneTransforms(
    BoneToModify=((Bone="dyn_hairShoulder_R_A_MSTR",Transform=(Rotation=(X=0,Y=0,Z=0,W=1),Translation=(X=0,Y=0,Z=0),Scale3D=(X=1,Y=1,Z=1))))
    Mode=AdditiveGlobal
    Weight=1
    WeightMaximum=1
    BoneToModify.0.Transform.Translation.Z=&RigVMFunction_MathFloatRemap(SourceMaximum=-50, SourceMinimum=-19, TargetMaximum=2, TargetMinimum=0, bClamp=true, Value=@n1.Result.Z).Result
    BoneToModify.0.Transform.Translation.Y=&RigVMFunction_MathFloatRemap(SourceMaximum=-50, SourceMinimum=0, TargetMaximum=-3, TargetMinimum=0, bClamp=true, Value=@n1.Result.Z).Result
    BoneToModify.0.Transform.Rotation=&RigVMFunction_MathQuaternionFromEuler(
     Euler=(X=0,Y=0,Z=0)
     Euler.Y=&RigVMFunction_MathFloatRemap(SourceMaximum=-50, SourceMinimum=-19, TargetMaximum=0, TargetMinimum=0, bClamp=true, Value=@n1.Result.Z).Result
    ).Result
    ExecutePin=@RigUnit_BeginExecution.ExecutePin
   ) {is=struct RigVMUnitNode}
   > &RigUnit_ModifyBoneTransforms(
    BoneToModify=((Bone="dyn_hairBack_D_MSTR",Transform=(Rotation=(X=0,Y=0,Z=0,W=1),Translation=(X=0,Y=0,Z=0),Scale3D=(X=1,Y=1,Z=1))))
    Mode=AdditiveGlobal
    Weight=1
    WeightMaximum=1
    BoneToModify.0.Transform.Translation.Z=&RigVMFunction_MathFloatRemap(SourceMaximum=-50, SourceMinimum=-30, TargetMaximum=0, TargetMinimum=0, bClamp=true, Value=@n1.Result.Z).Result
    BoneToModify.0.Transform.Translation.Y=&RigVMFunction_MathFloatRemap(SourceMaximum=-50, SourceMinimum=0, TargetMaximum=0, TargetMinimum=0, bClamp=true, Value=@n1.Result.Z).Result
    BoneToModify.0.Transform.Rotation=&RigVMFunction_MathQuaternionFromEuler(
     Euler=(X=0,Y=0,Z=0)
     Euler.Y=&RigVMFunction_MathFloatRemap(SourceMaximum=-50, SourceMinimum=-19, TargetMaximum=0, TargetMinimum=0, bClamp=False, Value=@n1.Result.Z).Result
    ).Result
   )
   > &RigUnit_ModifyBoneTransforms(
    BoneToModify=((Bone="dyn_hairShoulder_R_A_07",Transform=(Rotation=(X=0,Y=0,Z=0,W=1),Translation=(X=0,Y=0,Z=0),Scale3D=(X=1,Y=1,Z=1))))
    Mode=AdditiveGlobal
    Weight=1
    WeightMaximum=1
    BoneToModify.0.Transform.Translation.Z=&RigVMFunction_MathFloatRemap(SourceMaximum=-50, SourceMinimum=0, TargetMaximum=1.5, TargetMinimum=0, bClamp=true, Value=@n1.Result.Z).Result
    BoneToModify.0.Transform.Rotation=&RigVMFunction_MathQuaternionFromEuler(
     Euler=(X=0,Y=0,Z=0)
     Euler.Y=&RigVMFunction_MathFloatRemap(SourceMaximum=-50, SourceMinimum=-19, TargetMaximum=0, TargetMinimum=0, bClamp=False, Value=@n1.Result.Z).Result
    ).Result
   )
  ]}
]
hierarchy [
 root {loc=<0 0 0>, rot=<0 0 0 1>}
 attach {parent=root, loc=<0 0 0>, rot=<0 0 0 1>}
 pelvis {parent=root, loc=<0 -0.5301 98.0571>, rot=<0.0127 0.707 -0.0127 -0.707>}
 spine_01 {parent=pelvis, loc=<6.1208 -0.9328 0>, rot=<0 0 0.034 -0.9994>}
 spine_02 {parent=spine_01, loc=<6.5996 -0.2095 0>, rot=<0 0 0.0238 -0.9997>}
 spine_03 {parent=spine_02, loc=<6.4914 -0.1237 0>, rot=<0 0 0.0149 -0.9999>}
 spine_04 {parent=spine_03, loc=<6.9598 0.0794 0>, rot=<0 0 -0.0132 -0.9999>}
 spine_05 {parent=spine_04, loc=<6.9987 0 0>, rot=<0 0 -0.0015 -1>}
 clavicle_l {parent=spine_05, loc=<5.4831 -0.1053 -3.2946>, rot=<0.0465 -0.7289 -0.009 -0.6829>}
 clavicle_r {parent=spine_05, loc=<5.4831 -0.1053 3.2946>, rot=<0.7289 0.0465 -0.6829 0.009>}
 neck_01 {parent=spine_05, loc=<11.3138 0.1529 0>, rot=<0 0 -0.2003 -0.9797>}
 neck_02 {parent=neck_01, loc=<4.5627 0 0>, rot=<0 0 0.0504 -0.9987>}
 head {parent=neck_02, loc=<5.3914 0 0>, rot=<0 0 0.115 -0.9934>}
 dyn_earing_L {parent=head, loc=<0.6576 0.0235 -6.0006>, rot=<0.0041 -0.9955 -0.0004 -0.0952>}
 dyn_bow_C_MSTR {parent=head, loc=<8.3685 -8.3345 -0.1542>, rot=<0.0029 -0.7071 -0.0029 -0.7071>}
 dyn_bow_C_L_01 {parent=dyn_bow_C_MSTR, loc=<1.2565 -0.4669 0.047>, rot=<-0.6315 0.2362 0.1787 -0.7166>}
 dyn_bow_C_L_02 {parent=dyn_bow_C_L_01, loc=<1.8257 0 0>, rot=<0.0093 0.0004 0.082 -0.9966>}
 dyn_bow_C_R_01 {parent=dyn_bow_C_MSTR, loc=<-1.235 -0.4179 0.0146>, rot=<0.2649 -0.5996 0.7332 -0.181>}
 dyn_bow_C_R_02 {parent=dyn_bow_C_R_01, loc=<1.7339 0 0>, rot=<-0.0098 -0.0279 -0.0636 -0.9975>}
 dyn_bow_B_MSTR {parent=head, loc=<11.3966 -7.6294 -0.0731>, rot=<-0.7071 -0.0029 -0.0029 -0.7071>}
 dyn_bow_B_L_01 {parent=dyn_bow_B_MSTR, loc=<0.1394 -2.1245 -0.0286>, rot=<0.1258 -0.084 0.6809 -0.7166>}
 dyn_bow_B_L_02 {parent=dyn_bow_B_L_01, loc=<2.6379 0 0>, rot=<-0.013 0.017 -0.0012 -0.9998>}
 dyn_bow_B_R_01 {parent=dyn_bow_B_MSTR, loc=<0.0212 1.9644 -0.0558>, rot=<-0.0723 0.0393 -0.6952 -0.7141>}
 dyn_bow_B_R_02 {parent=dyn_bow_B_R_01, loc=<2.8983 0 0>, rot=<-0.0007 -0.0201 -0.074 -0.9971>}
 dyn_bow_A_MSTR {parent=head, loc=<13.3747 -5.1507 -0.1964>, rot=<0.046 -0.0329 0.4677 -0.8821>}
 dyn_bow_A_R_01 {parent=dyn_bow_A_MSTR, loc=<-0.0231 -0.1467 1.5036>, rot=<0 0 0 1>}
 dyn_bow_A_R_02 {parent=dyn_bow_A_R_01, loc=<-0.0403 0.0876 1.9505>, rot=<0 0 0 1>}
 dyn_bow_A_L_01 {parent=dyn_bow_A_MSTR, loc=<0.1541 0.1052 -1.0487>, rot=<0.9619 0.2544 -0.0808 -0.0585>}
 dyn_bow_A_L_02 {parent=dyn_bow_A_L_01, loc=<0.143 -0.1007 2.3234>, rot=<0 0 0 1>}
 dyn_bow_D_MSTR {parent=head, loc=<6.0608 -8.592 -0.0588>, rot=<0.0029 -0.7071 -0.0029 -0.7071>}
 dyn_bow_D_L_01 {parent=dyn_bow_D_MSTR, loc=<1.3154 -0.3195 -0.1039>, rot=<-0.6865 0.15 0.3572 -0.6153>}
 dyn_bow_D_L_02 {parent=dyn_bow_D_L_01, loc=<1.4215 0 0>, rot=<0.0158 -0.0236 -0.0454 -0.9986>}
 dyn_bow_D_R_01 {parent=dyn_bow_D_MSTR, loc=<-1.229 -0.3603 -0.0446>, rot=<0.1146 -0.6622 0.7154 -0.1913>}
 dyn_bow_D_R_02 {parent=dyn_bow_D_R_01, loc=<1.3451 0 0>, rot=<0.0011 -0.0014 -0.0125 -0.9999>}
 dyn_bow_E_MSTR {parent=head, loc=<3.6081 -8.3495 -0.0957>, rot=<0.0029 -0.7071 -0.0029 -0.7071>}
 dyn_bow_E_R_01 {parent=dyn_bow_E_MSTR, loc=<-1.23 -0.1275 -0.162>, rot=<-0.0095 -0.6935 0.6989 -0.1745>}
 dyn_bow_E_R_02 {parent=dyn_bow_E_R_01, loc=<0.9717 0 0>, rot=<0.0171 -0.0154 -0.0331 -0.9992>}
 dyn_bow_E_L_01 {parent=dyn_bow_E_MSTR, loc=<1.2145 -0.2172 -0.2151>, rot=<-0.6828 0.0255 0.2896 -0.6703>}
 dyn_bow_E_L_02 {parent=dyn_bow_E_L_01, loc=<1.0057 0 0>, rot=<-0.0123 -0.0306 -0.0605 -0.9976>}
 dyn_hairPoof_R {parent=head, loc=<9.464 -1.4449 5.2611>, rot=<0 0 0 1>}
 dyn_hairPoof_L {parent=head, loc=<8.691 -2.4403 -4.9005>, rot=<0 0 0 -1>}
 dyn_earing_R {parent=head, loc=<0.6417 0.0055 5.918>, rot=<0.1142 0.0164 0.9933 -0.0035>}
 dyn_hairBack_A_MSTR {parent=head, loc=<1.5266 -5.1745 -5.4276>, rot=<0.0364 -0.8572 0.5084 -0.0734>}
 dyn_hairBack_A_01 {parent=dyn_hairBack_A_MSTR, loc=<0 0 0>, rot=<0 0 0 -1>}
 dyn_hairBack_A_02 {parent=dyn_hairBack_A_01, loc=<5.2696 0 0>, rot=<0 0.0965 0 -0.9953>}
 dyn_hairBack_A_03 {parent=dyn_hairBack_A_02, loc=<5.2696 0 0>, rot=<0 0.0327 0 -0.9995>}
 dyn_hairBack_A_04 {parent=dyn_hairBack_A_03, loc=<4.684 0 0.1263>, rot=<0 -0.106 0 -0.9944>}
 dyn_hairBack_B_MSTR {parent=head, loc=<1.6888 -6.9154 -2.5699>, rot=<0.0518 -0.7829 0.6164 -0.0661>}
 dyn_hairBack_B_01 {parent=dyn_hairBack_B_MSTR, loc=<0 0 0>, rot=<0 0 0 1>}
 dyn_hairBack_B_02 {parent=dyn_hairBack_B_01, loc=<5.2696 0 0>, rot=<0 0.0965 0 -0.9953>}
 dyn_hairBack_B_03 {parent=dyn_hairBack_B_02, loc=<5.2696 0 0>, rot=<0 -0.0162 0 -0.9999>}
 dyn_hairBack_B_04 {parent=dyn_hairBack_B_03, loc=<5.2696 0 0>, rot=<0 -0.106 0 -0.9944>}
 dyn_hairBack_B_05 {parent=dyn_hairBack_B_04, loc=<5.2696 0 0>, rot=<0 0 0 1>}
 dyn_hairShoulder_L_A_MSTR {parent=head, loc=<-1.6224 -2.7402 -6.2412>, rot=<-0.091 -0.9574 -0.2638 -0.074>}
 dyn_hairShoulder_L_A_01 {parent=dyn_hairShoulder_L_A_MSTR, loc=<0 0 0>, rot=<0 0 0 -1>}
 dyn_hairShoulder_L_A_02 {parent=dyn_hairShoulder_L_A_01, loc=<6.1825 0.145 -0.5381>, rot=<0.2862 0.1359 -0.2651 -0.9107>}
 dyn_hairShoulder_L_A_03 {parent=dyn_hairShoulder_L_A_02, loc=<3.944 0 0>, rot=<-0.0406 -0.0142 0.1421 -0.9889>}
 dyn_hairShoulder_L_A_04 {parent=dyn_hairShoulder_L_A_03, loc=<3.9405 0 0>, rot=<0.1977 -0.0218 -0.3017 -0.9324>}
 dyn_hairShoulder_L_B_01 {parent=dyn_hairShoulder_L_A_02, loc=<2.9237 -2.4586 -0.304>, rot=<-0.1128 0.1749 0.3872 -0.8982>}
 dyn_hairShoulder_L_B_02 {parent=dyn_hairShoulder_L_B_01, loc=<2.4971 0 0>, rot=<0.0283 -0.0133 -0.2376 -0.9709>}
 dyn_hair_R_A_MSTR {parent=head, loc=<7.0007 4.7795 6.4189>, rot=<-0.0175 0.0434 0.9989 0.0006>}
 dyn_hair_R_A_01 {parent=dyn_hair_R_A_MSTR, loc=<0 0 0>, rot=<0 0 0 1>}
 dyn_hair_R_A_02 {parent=dyn_hair_R_A_01, loc=<5.7839 0 0>, rot=<-0.011 0.0977 -0.0311 -0.9947>}
 dyn_hair_R_A_03 {parent=dyn_hair_R_A_02, loc=<4.5012 0 0>, rot=<0.0172 -0.019 -0.015 -0.9996>}
 dyn_hair_R_A_04 {parent=dyn_hair_R_A_03, loc=<3.9371 0 0>, rot=<-0.0284 -0.3382 0.0006 -0.9406>}
 dyn_hair_R_A_05 {parent=dyn_hair_R_A_04, loc=<4.0649 0 0>, rot=<0.01 0.14 0.1306 -0.9815>}
 dyn_hair_011 {parent=dyn_hair_R_A_05, loc=<2.4463 0 0>, rot=<-0.0008 0.0076 0.0436 -0.999>}
 dyn_hair_R_A_06 {parent=dyn_hair_R_A_05, loc=<2.4463 0 0>, rot=<-0.0008 0.0076 0.0436 -0.999>}
 dyn_hair_R_B_01 {parent=dyn_hair_R_A_02, loc=<3.7534 -2.0774 -0.2202>, rot=<-0.2573 -0.1011 0.2071 -0.9384>}
 dyn_hair_R_B_02 {parent=dyn_hair_R_B_01, loc=<3.3443 0 0>, rot=<-0.0035 -0.1955 -0.0361 -0.98>}
 dyn_hair_R_B_03 {parent=dyn_hair_R_B_02, loc=<2.2961 0 0>, rot=<0.0767 0.1392 -0.3617 -0.9187>}
 dyn_bang_R_MSTR {parent=head, loc=<9.7598 8.9875 1.0694>, rot=<0.6091 0.3376 0.7177 -0.0047>}
 dyn_bang_R_01 {parent=dyn_bang_R_MSTR, loc=<0 0 0>, rot=<0 0 0 -1>}
 dyn_bang_R_02 {parent=dyn_bang_R_01, loc=<2.8854 0 0>, rot=<-0.0503 -0.4126 0.027 -0.9091>}
 dyn_bang_R_03 {parent=dyn_bang_R_02, loc=<2.5414 0 0>, rot=<0.0353 -0.2204 0.0103 -0.9747>}
 dyn_bang_L_MSTR {parent=head, loc=<9.7833 9.3622 -0.7914>, rot=<-0.0191 -0.7453 -0.3235 -0.5827>}
 dyn_bang_L_01 {parent=dyn_bang_L_MSTR, loc=<0 0 0>, rot=<0 0 0 -1>}
 dyn_bang_L_02 {parent=dyn_bang_L_01, loc=<2.6025 0 0>, rot=<0.0949 -0.3681 0.0782 -0.9216>}
 dyn_hairBack_C_MSTR {parent=head, loc=<1.7987 -7.0344 0.3459>, rot=<0.0646 -0.6633 0.7435 -0.0554>}
 dyn_hairBack_C_01 {parent=dyn_hairBack_C_MSTR, loc=<0 0 0>, rot=<0 0 0 1>}
 dyn_hairBack_C_02 {parent=dyn_hairBack_C_01, loc=<5.2696 0 0>, rot=<0 0.0965 0 -0.9953>}
 dyn_hairBack_C_03 {parent=dyn_hairBack_C_02, loc=<5.2696 0 0>, rot=<0 -0.0162 0 -0.9999>}
 dyn_hairBack_C_04 {parent=dyn_hairBack_C_03, loc=<5.2696 0 0>, rot=<0 -0.106 0 -0.9944>}
 dyn_hairBack_C_05 {parent=dyn_hairBack_C_04, loc=<5.2696 0 0>, rot=<0 0 0 1>}
 dyn_hairBack_D_MSTR {parent=head, loc=<1.904 -6.1074 2.9313>, rot=<0.1023 -0.4684 0.8764 -0.0458>}
 dyn_hairBack_D_01 {parent=dyn_hairBack_D_MSTR, loc=<0 0 0>, rot=<0 0 0 -1>}
 dyn_hairBack_D_02 {parent=dyn_hairBack_D_01, loc=<5.2696 0 0>, rot=<0 0.0965 0 -0.9953>}
 dyn_hairBack_D_03 {parent=dyn_hairBack_D_02, loc=<5.2696 0 0>, rot=<0 -0.0162 0 -0.9999>}
 dyn_hairBack_D_04 {parent=dyn_hairBack_D_03, loc=<5.2696 0 0>, rot=<0 -0.106 0 -0.9944>}
 dyn_hairShoulder_R_A_MSTR {parent=head, loc=<2.3939 -3.7348 5.2516>, rot=<0.0469 -0.1015 0.9937 -0.002>}
 dyn_hairShoulder_R_A_01 {parent=dyn_hairShoulder_R_A_MSTR, loc=<0 0 0>, rot=<0 0 0 1>}
 dyn_hairShoulder_R_A_02 {parent=dyn_hairShoulder_R_A_01, loc=<6.1733 0 0>, rot=<-0.0146 0.0796 0.0334 -0.9962>}
 dyn_hairShoulder_R_A_03 {parent=dyn_hairShoulder_R_A_02, loc=<3.3765 0 0>, rot=<-0.137 0.0072 0.3481 -0.9274>}
 dyn_hairShoulder_R_A_04 {parent=dyn_hairShoulder_R_A_03, loc=<5.0277 0 0>, rot=<-0.1548 -0.1316 0.245 -0.948>}
 dyn_hairShoulder_R_A_05 {parent=dyn_hairShoulder_R_A_04, loc=<4.0159 0 0>, rot=<-0.1108 -0.0738 -0.0428 -0.9902>}
 dyn_hairShoulder_R_A_07 {parent=dyn_hairShoulder_R_A_03, loc=<0.3175 2.2313 -0.7235>, rot=<-0.1307 0.6252 -0.3336 -0.6934>}
 dyn_hairShoulder_R_B_01 {parent=dyn_hairShoulder_R_A_02, loc=<2.9019 -2.6034 -1.5366>, rot=<-0.3157 -0.1807 0.3856 -0.8479>}
 dyn_hairShoulder_R_B_02 {parent=dyn_hairShoulder_R_B_01, loc=<3.2966 0 0>, rot=<-0.096 -0.1212 0.0781 -0.9849>}
 dyn_hairShoulder_R_B_03 {parent=dyn_hairShoulder_R_B_02, loc=<2.9485 0 0>, rot=<-0.1442 0.0784 -0.2028 -0.9654>}
 dyn_hair_L_A_MSTR {parent=head, loc=<11.4978 6.3599 -3.8812>, rot=<0.8906 0.1399 -0.3686 -0.2266>}
 dyn_hair_L_A_01 {parent=dyn_hair_L_A_MSTR, loc=<0 0 0>, rot=<0 0 0 1>}
 dyn_hair_L_B_01 {parent=dyn_hair_L_A_01, loc=<-1.734 -1.8618 5.6592>, rot=<-0.3747 0.0018 0.9271 0.0045>}
 dyn_hair_L_B_02 {parent=dyn_hair_L_B_01, loc=<0.5997 0.0771 2.695>, rot=<0 0 0 1>}
 dyn_hair_L_B_03 {parent=dyn_hair_L_B_02, loc=<0.1796 0.6376 1.9157>, rot=<0 0 0 1>}
 dyn_hair_L_A_02 {parent=dyn_hair_L_A_01, loc=<-2.2513 2.2724 4.351>, rot=<0.3929 0.1733 0.8981 -0.0957>}
 dyn_hair_L_A_03 {parent=dyn_hair_L_A_02, loc=<4.8861 0 0>, rot=<-0.0023 -0.1349 0.0525 -0.9895>}
 dyn_hair_L_A_04 {parent=dyn_hair_L_A_03, loc=<4.6846 0 0>, rot=<0.015 0.2719 0.0971 -0.9573>}
 dyn_hair_L_A_05 {parent=dyn_hair_L_A_04, loc=<4.7195 0 0>, rot=<0.0619 -0.0475 -0.1337 -0.9879>}
 dyn_hair_L_A_06 {parent=dyn_hair_L_A_05, loc=<2.828 0 0>, rot=<0.1258 -0.2038 -0.3858 -0.8909>}
 dyn_hair_L_A_07 {parent=dyn_hair_L_A_06, loc=<2.3612 0 0>, rot=<0.0365 0.0714 -0.1487 -0.9856>}
 dyn_hair_L_C_01 {parent=dyn_hair_L_A_03, loc=<2.8385 2.3788 -0.1028>, rot=<0.1366 0.04 -0.0755 -0.9869>}
 dyn_hair_L_C_02 {parent=dyn_hair_L_C_01, loc=<3.4207 0 0>, rot=<-0.0695 0.1953 0.163 -0.9646>}
 dyn_hair_L_C_03 {parent=dyn_hair_L_C_02, loc=<2.8333 0 0>, rot=<0.0375 -0.269 0.1354 -0.9528>}
]
