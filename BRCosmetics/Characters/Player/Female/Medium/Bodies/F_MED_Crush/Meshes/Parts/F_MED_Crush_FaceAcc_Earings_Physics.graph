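# Physics asset for the earrings face accessory mesh of the Crush medium female player body, giving both earrings a physics setup for dangling motion.
# Identical box bodies sit on the left and right earring bones, alongside a capsule body on the head bone.
# Each earring is constrained to the head with moderate swing limits and a much tighter twist limit.
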PhysicsAsset /BRCosmetics/Characters/Player/Female/Medium/Bodies/F_MED_Crush/Meshes/Parts/F_MED_Crush_FaceAcc_Earings_Physics
bodies {
 dyn_earing_L [{shapeType=Box, extentX=4.1359, extentY=2.8612, extentZ=1.1199, center=<2.0859 0 0>}]
 dyn_earing_R [{shapeType=Box, extentX=4.1359, extentY=2.8612, extentZ=1.1199, center=<2.0859 0 0>}]
 head [{shapeType=Capsule, length=5.1113, radius=5.2093, center=<-1.2175 2.0174 0>, rotation=<rot 90 0 58.8895>}]
}
constraints [
 {bone1Name=dyn_earing_L, bone2Name=head, swing1LimitDegrees=20, swing2LimitDegrees=20, twistLimitDegrees=4}
 {bone1Name=dyn_earing_R, bone2Name=head, swing1LimitDegrees=20, swing2LimitDegrees=20, twistLimitDegrees=4}
]
