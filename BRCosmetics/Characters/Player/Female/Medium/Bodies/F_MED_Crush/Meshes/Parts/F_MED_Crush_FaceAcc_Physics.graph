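# Physics asset for the F_MED_Crush face-accessory mesh, setting up physics for the female character's hair, bangs, hair poofs and earrings.
# Capsules cover upper spine, clavicles, neck and head, alongside mostly box bodies on the dyn_ bones.
# Swing/twist-limited chains, some branching, hang from small MSTR bodies jointed to the head with free rotation; poofs and earrings attach directly.
# Free-rotation cross-links tie back-hair chains A and C to B.
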PhysicsAsset /BRCosmetics/Characters/Player/Female/Medium/Bodies/F_MED_Crush/Meshes/Parts/F_MED_Crush_FaceAcc_Physics
bodies {
 clavicle_l [{shapeType=Capsule, length=11.6702, radius=5.8752, center=<5.5144 -0.8243 -1.1015>, rotation=<rot 85.1267 179.9999 -180>}]
 clavicle_r [{shapeType=Capsule, length=14.2727, radius=7.3527, center=<-2.9939 0.5044 1.8959>, rotation=<rot 0 -89.9999 89.1164>}]
 dyn_bang_L_01 [{shapeType=Box, extentX=2.8048, extentY=2.7717, extentZ=1.6851, center=<1.5301 0 0>}]
 dyn_bang_L_02 [{shapeType=Box, extentX=2.8048, extentY=2.7717, extentZ=1.6851, center=<1.5301 0 0>}]
 dyn_bang_L_MSTR [{shapeType=Capsule, length=0.505, radius=0.505}]
 dyn_bang_R_01 [{shapeType=Box, extentX=2.8048, extentY=2.7717, extentZ=1.6851, center=<1.5301 0 0>}]
 dyn_bang_R_02 [{shapeType=Box, extentX=2.8048, extentY=2.7717, extentZ=1.6851, center=<1.5301 0 0>}]
 dyn_bang_R_03 [{shapeType=Box, extentX=2.8048, extentY=2.7717, extentZ=1.6851, center=<1.5301 0 0>}]
 dyn_bang_R_MSTR [{shapeType=Capsule, length=0.505, radius=0.505}]
 dyn_earing_L [{shapeType=Box, extentX=4.1359, extentY=2.8612, extentZ=1.1199, center=<2.0859 0 0>}]
 dyn_earing_R [{shapeType=Box, extentX=4.1359, extentY=2.8612, extentZ=1.1199, center=<2.0859 0 0>}]
 dyn_hairBack_A_01 [{shapeType=Box, extentX=5.419, extentY=8.3413, extentZ=1.9717, center=<2.65 0.0656 -0.1659>, rotation=<rot -3.5829 1.4174 18.5503>}]
 dyn_hairBack_A_02 [{shapeType=Box, extentX=5.1174, extentY=7.6674, extentZ=3.054, center=<2.6674 -0.1682 0.0206>, rotation=<rot -3.5829 1.4174 -2.9782>}]
 dyn_hairBack_A_03 [{shapeType=Box, extentX=4.5743, extentY=7.5789, extentZ=2.2313, center=<2.3518 -0.0151 0.2617>, rotation=<rot 12.5505 -5.8392 -24.5373>}]
 dyn_hairBack_A_04 [{shapeType=Box, extentX=5.2323, extentY=7.762, extentZ=2.3711, center=<2.5818 -0.2824 0.4204>, rotation=<rot 5.0638 -3.7353 -30.8638>}]
 dyn_hairBack_A_MSTR [{shapeType=Capsule, length=0.505, radius=0.505}]
 dyn_hairBack_B_01 [{shapeType=Box, extentX=5, extentY=8.4363, extentZ=1.7495, center=<2.6252 0 0>}]
 dyn_hairBack_B_02 [{shapeType=Box, extentX=5, extentY=8.4363, extentZ=1.7495, center=<2.5547 0 0>}]
 dyn_hairBack_B_03 [{shapeType=Box, extentX=5, extentY=8.4363, extentZ=1.7495, center=<2.5171 0 0>}]
 dyn_hairBack_B_04 [{shapeType=Box, extentX=5, extentY=8.4363, extentZ=1.7495, center=<2.566 0 0>}]
 dyn_hairBack_B_05 [{shapeType=Box, extentX=5, extentY=8.4363, extentZ=1.7495, center=<2.3919 0 -0.0001>}]
 dyn_hairBack_B_MSTR [{shapeType=Capsule, length=0.505, radius=0.505}]
 dyn_hairBack_C_01 [{shapeType=Box, extentX=4.9998, extentY=5.9616, extentZ=3.2079, center=<2.5785 0 -0.0001>, rotation=<rot 0 0 0.2717>}]
 dyn_hairBack_C_02 [{shapeType=Box, extentX=5.491, extentY=9.3622, extentZ=3.6147, center=<2.5785 0 -0.1998>}]
 dyn_hairBack_C_03 [{shapeType=Box, extentX=5.4066, extentY=11.5474, extentZ=3.2083, center=<2.5785 0 -0.0001>}]
 dyn_hairBack_C_04 [{shapeType=Box, extentX=5.346, extentY=9.3351, extentZ=2.883, center=<3.0775 0 -0.0001>, rotation=<rot 0 0 4.0049>}]
 dyn_hairBack_C_05 [{shapeType=Box, extentX=4.4154, extentY=8.7369, extentZ=1.1667, center=<2.157 0 -0.0001>}]
 dyn_hairBack_C_MSTR [{shapeType=Capsule, length=0.505, radius=0.505}]
 dyn_hairBack_D_01 [{shapeType=Box, extentX=4.3189, extentY=5.4261, extentZ=1.3776, center=<2.8795 0.1024 0.2329>, rotation=<rot 0 0 18.3503>}]
 dyn_hairBack_D_02 [{shapeType=Box, extentX=10.1553, extentY=6.0814, extentZ=1.8823, center=<5.3281 0.1024 0.2328>, rotation=<rot 0 0 17.8862>}]
 dyn_hairBack_D_03 [{shapeType=Box, extentX=5.1896, extentY=6.2245, extentZ=0.8641, center=<2.6506 0.2527 0.567>, rotation=<rot 0 0 17.8859>}]
 dyn_hairBack_D_04 [{shapeType=Box, extentX=5.1896, extentY=6.2245, extentZ=0.7488, center=<2.5785 0.2527 0.5746>, rotation=<rot 0 0 29.9252>}]
 dyn_hairBack_D_MSTR [{shapeType=Capsule, length=0.505, radius=0.505}]
 dyn_hairPoof_L [{shapeType=Box, extentX=2.3983, extentY=5, extentZ=5, center=<-2.1647 0 0>}]
 dyn_hairPoof_R [{shapeType=Box, extentX=2.3983, extentY=5, extentZ=5, center=<-2.1647 0 0>}]
 dyn_hairShoulder_L_A_01 [{shapeType=Box, extentX=5.633, extentY=4.0109, extentZ=2.3774, center=<2.939 0 0>, rotation=<rot 0 0 -27.035>}]
 dyn_hairShoulder_L_A_02 [{shapeType=Capsule, length=2.2672, radius=1.1924, center=<1.6841 -0.3273 -0.1426>, rotation=<rot -81.2829 178.3304 158.4141>}]
 dyn_hairShoulder_L_A_03 [{shapeType=Capsule, length=2.4098, radius=1.0498, center=<1.8985 -0.0001 0.0048>, rotation=<rot -89.8533 0.015 -0.015>}]
 dyn_hairShoulder_L_A_04 [{shapeType=Capsule, length=2.1407, radius=0.5571, center=<1.5798 -0.0001 0.004>, rotation=<rot -89.8533 0.015 -0.015>}]
 dyn_hairShoulder_L_A_MSTR [{shapeType=Capsule, length=0.505, radius=0.505}]
 dyn_hairShoulder_L_B_01 [{shapeType=Box, extentX=2.4057, extentY=1.9701, extentZ=0.767, center=<1.2925 0 0.0736>}]
 dyn_hairShoulder_L_B_02 [{shapeType=Box, extentX=2.4057, extentY=1.9701, extentZ=0.767, center=<1.2896 -0.0036 0.0735>}]
 dyn_hairShoulder_R_A_01 [{shapeType=Box, extentX=3.6641, extentY=6.7433, extentZ=1.8108, center=<2.8368 0 -0.0001>, rotation=<rot 0 0 0.1926>}]
 dyn_hairShoulder_R_A_02 [{shapeType=Box, extentX=1.4492, extentY=5.7962, extentZ=0.8366, center=<0.7458 -0.0096 0.0005>, rotation=<rot -0.0421 0.7848 -0.0771>}]
 dyn_hairShoulder_R_A_03 [{shapeType=Capsule, length=2.139, radius=1.0594, center=<2.5712 0 0.0001>, rotation=<rot 90 90 111.1512>}]
 dyn_hairShoulder_R_A_04 [{shapeType=Capsule, length=2.0867, radius=0.9133, center=<1.8142 0 0>, rotation=<rot 90 0 0>}]
 dyn_hairShoulder_R_A_05 [{shapeType=Capsule, length=2.018, radius=0.982, center=<1.838 0 0.0001>, rotation=<rot 90 0 0>}]
 dyn_hairShoulder_R_A_MSTR [{shapeType=Capsule, length=0.505, radius=0.505}]
 dyn_hairShoulder_R_B_01 [{shapeType=Capsule, length=2, radius=1, center=<1.8753 0 0>, rotation=<rot 90 0 0>}]
 dyn_hairShoulder_R_B_02 [{shapeType=Capsule, length=2, radius=1, center=<1.9746 0 0>, rotation=<rot 90 0 0>}]
 dyn_hairShoulder_R_B_03 [{shapeType=Capsule, length=2, radius=1, center=<1.8682 0 -0.0001>, rotation=<rot 90 0 0>}]
 dyn_hair_L_A_01 [{shapeType=Box, extentX=3.6207, extentY=6.4795, extentZ=6.577, center=<-0.9259 -0.0449 2.9602>, rotation=<rot 16.5845 2.7793 -2.4944>}]
 dyn_hair_L_A_02 [{shapeType=Box, extentX=4.937, extentY=4.7523, extentZ=1.6589, center=<2.4164 0 0>, rotation=<rot 1.8422 9.1471 -3.9261>}]
 dyn_hair_L_A_03 [{shapeType=Box, extentX=4.5584, extentY=7.1822, extentZ=3.7051, center=<2.3249 0 0>}]
 dyn_hair_L_A_04 [{shapeType=Box, extentX=4.5584, extentY=4.7866, extentZ=3.3985, center=<2.4663 0.0169 0.1209>, rotation=<rot -3.8187 -0.534 7.9669>}]
 dyn_hair_L_A_05 [{shapeType=Box, extentX=2.9314, extentY=3.4608, extentZ=3.6038, center=<1.5457 0 0>}]
 dyn_hair_L_A_06 [{shapeType=Box, extentX=2.3685, extentY=2.2415, extentZ=2.4074, center=<1.1267 0 0>}]
 dyn_hair_L_A_07 [{shapeType=Box, extentX=2.3888, extentY=1.3409, extentZ=2.4074, center=<1.002 0 0>}]
 dyn_hair_L_A_MSTR [{shapeType=Capsule, length=0.505, radius=0.505}]
 dyn_hair_L_B_01 [{shapeType=Box, extentX=1.0594, extentY=2.041, extentZ=3, center=<0.2859 -0.0037 1.424>, rotation=<rot -8.8316 -0.7507 -0.2585>}]
 dyn_hair_L_B_02 [{shapeType=Box, extentX=1.0594, extentY=1.3208, extentZ=2.1466, center=<0.2645 0.4641 0.9243>, rotation=<rot -4.1945 -1.8708 19.5108>}]
 dyn_hair_L_B_03 [{shapeType=Box, extentX=1.0594, extentY=0.8123, extentZ=2.1466, center=<0.2645 0.4641 0.9243>, rotation=<rot -4.1945 -1.8708 19.5108>}]
 dyn_hair_L_C_01 [{shapeType=Box, extentX=3.1965, extentY=2.779, extentZ=1.4233, center=<1.6425 0 0>}]
 dyn_hair_L_C_02 [{shapeType=Box, extentX=3.1965, extentY=1.928, extentZ=1.4233, center=<1.6425 0 0>}]
 dyn_hair_L_C_03 [{shapeType=Box, extentX=2.7009, extentY=1.6122, extentZ=1.4233, center=<1.3888 0 -0.0001>}]
 dyn_hair_R_A_01 [{shapeType=Box, extentX=5.723, extentY=7.6055, extentZ=1.4856, center=<2.8924 0 0.1886>}]
 dyn_hair_R_A_02 [{shapeType=Box, extentX=5.723, extentY=6.3136, extentZ=1.1429, center=<2.8924 0 0>, rotation=<rot -7.9346 -0.2395 -0.1943>}]
 dyn_hair_R_A_03 [{shapeType=Box, extentX=3, extentY=4.3607, extentZ=1.5195, center=<1.5308 0 0>}]
 dyn_hair_R_A_04 [{shapeType=Capsule, length=1.6162, radius=1.4802, center=<1.9866 0 0.0001>, rotation=<rot 90 0 0>}]
 dyn_hair_R_A_05 [{shapeType=Box, extentX=3, extentY=3.086, extentZ=1.5195, center=<1.5308 0 0>, rotation=<rot 0 0 94.3977>}]
 dyn_hair_R_A_06 [{shapeType=Box, extentX=3, extentY=4.3607, extentZ=1.5195, center=<1.5308 0 0>, rotation=<rot -0.0001 0 -93.8852>}]
 dyn_hair_R_A_MSTR [{shapeType=Capsule, length=0.505, radius=0.505}]
 dyn_hair_R_B_01 [{shapeType=Box, extentX=3.2406, extentY=2.4362, extentZ=0.9759, center=<1.6807 0 0>}]
 dyn_hair_R_B_02 [{shapeType=Capsule, length=1.4628, radius=0.5001, center=<1.1752 0 -0.0001>, rotation=<rot 89.7513 51.2978 51.0757>}]
 dyn_hair_R_B_03 [{shapeType=Capsule, length=1.4628, radius=0.5001, center=<1.1752 0 -0.0001>, rotation=<rot 90 0 0>}]
 head [
  {shapeType=Capsule, length=5.9467, radius=5.541, center=<0.8118 3.4354 0>, rotation=<rot 90 0 0>}
  {shapeType=Capsule, length=6.7157, radius=4.772, center=<0.8118 1.1771 0>, rotation=<rot 90 0 0>}
  {shapeType=Capsule, length=9.347, radius=4.522, center=<1.8068 -1.0039 -0.1194>, rotation=<rot 90 0 -10.0004>}
 ]
 neck_01 [{shapeType=Capsule, length=7.5228, radius=4.7659, center=<0.1502 0 0>, rotation=<rot -90 0 0>}]
 neck_02 [{shapeType=Capsule, length=5.1073, radius=4.4521, rotation=<rot -90 0 0>}]
 spine_04 [{shapeType=Capsule, length=23.6364, radius=8.5815, rotation=<rot 0 0 0.1926>}]
 spine_05 [{shapeType=Capsule, length=23.2164, radius=9.0015}]
}
constraints [
 {bone1Name=dyn_hairBack_A_01, bone2Name=dyn_hairBack_A_MSTR, swing1LimitDegrees=10, swing2LimitDegrees=45, twistLimitDegrees=20}
 {bone1Name=dyn_hairBack_A_02, bone2Name=dyn_hairBack_A_01, swing1LimitDegrees=10, swing2LimitDegrees=45, twistLimitDegrees=20}
 {bone1Name=dyn_hairBack_A_03, bone2Name=dyn_hairBack_A_02, swing1LimitDegrees=10, swing2LimitDegrees=45, twistLimitDegrees=20}
 {bone1Name=dyn_hairBack_A_04, bone2Name=dyn_hairBack_A_03, swing1LimitDegrees=10, swing2LimitDegrees=45, twistLimitDegrees=20}
 {bone1Name=dyn_hairBack_C_02, bone2Name=dyn_hairBack_C_01, swing1LimitDegrees=25, swing2LimitDegrees=50, twistLimitDegrees=20}
 {bone1Name=dyn_hairBack_C_03, bone2Name=dyn_hairBack_C_02, swing1LimitDegrees=25, swing2LimitDegrees=50, twistLimitDegrees=20}
 {bone1Name=dyn_hairBack_C_04, bone2Name=dyn_hairBack_C_03, swing1LimitDegrees=25, swing2LimitDegrees=50, twistLimitDegrees=20}
 {bone1Name=dyn_hairBack_D_02, bone2Name=dyn_hairBack_D_01, swing1LimitDegrees=5, swing2LimitDegrees=45, twistLimitDegrees=5}
 {bone1Name=dyn_hairBack_D_03, bone2Name=dyn_hairBack_D_02, swing1LimitDegrees=10, swing2LimitDegrees=45, twistLimitDegrees=20}
 {bone1Name=dyn_hairBack_D_04, bone2Name=dyn_hairBack_D_03, swing1LimitDegrees=20, swing2LimitDegrees=45, twistLimitDegrees=20}
 {bone1Name=dyn_hairShoulder_L_A_01, bone2Name=dyn_hairShoulder_L_A_MSTR, swing1LimitDegrees=25, swing2LimitDegrees=10, twistLimitDegrees=10}
 {bone1Name=dyn_hairShoulder_L_A_02, bone2Name=dyn_hairShoulder_L_A_01, swing1LimitDegrees=25, swing2LimitDegrees=10, twistLimitDegrees=10}
 {bone1Name=dyn_hairShoulder_L_A_03, bone2Name=dyn_hairShoulder_L_A_02, swing1LimitDegrees=25, swing2LimitDegrees=10, twistLimitDegrees=10}
 {bone1Name=dyn_hairShoulder_L_A_04, bone2Name=dyn_hairShoulder_L_A_03, swing1LimitDegrees=25, swing2LimitDegrees=10, twistLimitDegrees=10}
 {bone1Name=dyn_hairShoulder_L_B_01, bone2Name=dyn_hairShoulder_L_A_02, swing1LimitDegrees=25, swing2LimitDegrees=10, twistLimitDegrees=10}
 {bone1Name=dyn_hairShoulder_L_B_02, bone2Name=dyn_hairShoulder_L_B_01, swing1LimitDegrees=25, swing2LimitDegrees=10, twistLimitDegrees=10}
 {bone1Name=dyn_hair_R_A_01, bone2Name=dyn_hair_R_A_MSTR, swing1LimitDegrees=20, swing2LimitDegrees=40, twistLimitDegrees=10}
 {bone1Name=dyn_hair_R_A_02, bone2Name=dyn_hair_R_A_01, swing1LimitDegrees=20, swing2LimitDegrees=40, twistLimitDegrees=10}
 {bone1Name=dyn_hair_R_A_03, bone2Name=dyn_hair_R_A_02, swing1LimitDegrees=20, swing2LimitDegrees=40, twistLimitDegrees=10}
 {bone1Name=dyn_hair_R_A_04, bone2Name=dyn_hair_R_A_03, swing1LimitDegrees=20, swing2LimitDegrees=40, twistLimitDegrees=10}
 {bone1Name=dyn_hair_R_A_05, bone2Name=dyn_hair_R_A_04, swing1LimitDegrees=20, swing2LimitDegrees=40, twistLimitDegrees=10}
 {bone1Name=dyn_hair_R_A_06, bone2Name=dyn_hair_R_A_05, swing1LimitDegrees=20, swing2LimitDegrees=20, twistLimitDegrees=10}
 {bone1Name=dyn_hair_R_B_01, bone2Name=dyn_hair_R_A_02, swing1LimitDegrees=5, swing2LimitDegrees=5, twistLimitDegrees=5}
 {bone1Name=dyn_hair_L_B_01, bone2Name=dyn_hair_L_A_01, swing1LimitDegrees=15, swing2LimitDegrees=15, twistLimitDegrees=15}
 {bone1Name=dyn_hair_L_B_02, bone2Name=dyn_hair_L_B_01, swing1LimitDegrees=15, swing2LimitDegrees=15, twistLimitDegrees=15}
 {bone1Name=dyn_hair_L_B_03, bone2Name=dyn_hair_L_B_02, swing1LimitDegrees=15, swing2LimitDegrees=15, twistLimitDegrees=15}
 {bone1Name=dyn_hair_L_A_02, bone2Name=dyn_hair_L_A_01, swing1LimitDegrees=10, swing2LimitDegrees=10, twistLimitDegrees=10}
 {bone1Name=dyn_hair_L_A_03, bone2Name=dyn_hair_L_A_02, swing1LimitDegrees=20, swing2LimitDegrees=20, twistLimitDegrees=5}
 {bone1Name=dyn_hair_L_A_04, bone2Name=dyn_hair_L_A_03, swing1LimitDegrees=6, swing2LimitDegrees=20, twistLimitDegrees=5}
 {bone1Name=dyn_hair_L_A_05, bone2Name=dyn_hair_L_A_04, swing1LimitDegrees=45, swing2LimitDegrees=60, twistLimitDegrees=10}
 {bone1Name=dyn_hair_L_A_06, bone2Name=dyn_hair_L_A_05, swing1LimitDegrees=50, swing2LimitDegrees=60, twistLimitDegrees=10}
 {bone1Name=dyn_hair_L_A_07, bone2Name=dyn_hair_L_A_06, swing1LimitDegrees=90, swing2LimitDegrees=60, twistLimitDegrees=10}
 {bone1Name=dyn_hair_L_C_01, bone2Name=dyn_hair_L_A_03, swing1LimitDegrees=10, swing2LimitDegrees=10, twistLimitDegrees=10}
 {bone1Name=dyn_hair_L_C_02, bone2Name=dyn_hair_L_C_01, swing1LimitDegrees=10, swing2LimitDegrees=10, twistLimitDegrees=10}
 {bone1Name=dyn_hair_L_C_03, bone2Name=dyn_hair_L_C_02, swing1LimitDegrees=10, swing2LimitDegrees=10, twistLimitDegrees=10}
 {bone1Name=dyn_bang_R_01, bone2Name=dyn_bang_R_MSTR, swing1LimitDegrees=15, swing2LimitDegrees=15, twistLimitDegrees=15}
 {bone1Name=dyn_bang_R_02, bone2Name=dyn_bang_R_01, swing1LimitDegrees=10, swing2LimitDegrees=10, twistLimitDegrees=10}
 {bone1Name=dyn_bang_R_03, bone2Name=dyn_bang_R_02, swing1LimitDegrees=20, swing2LimitDegrees=20, twistLimitDegrees=20}
 {bone1Name=dyn_bang_L_01, bone2Name=dyn_bang_L_MSTR, swing1LimitDegrees=15, swing2LimitDegrees=15, twistLimitDegrees=15}
 {bone1Name=dyn_bang_L_02, bone2Name=dyn_bang_L_01, swing1LimitDegrees=10, swing2LimitDegrees=10, twistLimitDegrees=10}
 {
  bone1Name=clavicle_r
  bone2Name=spine_05
  swing1LimitDegrees=45
  swing1Motion=Free
  swing2LimitDegrees=45
  swing2Motion=Free
  twistLimitDegrees=45
  twistMotion=Free
 }
 {bone1Name=dyn_hairPoof_R, bone2Name=head, swing1LimitDegrees=45, swing2LimitDegrees=45, twistLimitDegrees=5}
 {bone1Name=dyn_hairPoof_L, bone2Name=head, swing1LimitDegrees=45, swing2LimitDegrees=45, twistLimitDegrees=5}
 {
  bone1Name=dyn_hairBack_A_MSTR
  bone2Name=head
  swing1LimitDegrees=45
  swing1Motion=Free
  swing2LimitDegrees=45
  swing2Motion=Free
  twistLimitDegrees=45
  twistMotion=Free
 }
 {
  bone1Name=dyn_hairBack_B_MSTR
  bone2Name=head
  swing1LimitDegrees=45
  swing1Motion=Free
  swing2LimitDegrees=45
  swing2Motion=Free
  twistLimitDegrees=45
  twistMotion=Free
 }
 {
  bone1Name=dyn_hairShoulder_L_A_MSTR
  bone2Name=head
  swing1LimitDegrees=45
  swing1Motion=Free
  swing2LimitDegrees=45
  swing2Motion=Free
  twistLimitDegrees=45
  twistMotion=Free
 }
 {
  bone1Name=dyn_hair_R_A_MSTR
  bone2Name=head
  swing1LimitDegrees=45
  swing1Motion=Free
  swing2LimitDegrees=45
  swing2Motion=Free
  twistLimitDegrees=45
  twistMotion=Free
 }
 {
  bone1Name=dyn_bang_R_MSTR
  bone2Name=head
  swing1LimitDegrees=45
  swing1Motion=Free
  swing2LimitDegrees=45
  swing2Motion=Free
  twistLimitDegrees=45
  twistMotion=Free
 }
 {
  bone1Name=dyn_bang_L_MSTR
  bone2Name=head
  swing1LimitDegrees=45
  swing1Motion=Free
  swing2LimitDegrees=45
  swing2Motion=Free
  twistLimitDegrees=45
  twistMotion=Free
 }
 {bone1Name=dyn_hair_R_B_02, bone2Name=dyn_hair_R_B_01, swing1LimitDegrees=5, swing2LimitDegrees=5, twistLimitDegrees=5}
 {bone1Name=dyn_hair_R_B_03, bone2Name=dyn_hair_R_B_02, swing1LimitDegrees=5, swing2LimitDegrees=5, twistLimitDegrees=5}
 {
  bone1Name=head
  bone2Name=neck_02
  swing1LimitDegrees=45
  swing1Motion=Free
  swing2LimitDegrees=45
  swing2Motion=Free
  twistLimitDegrees=45
  twistMotion=Free
 }
 {
  bone1Name=neck_01
  bone2Name=spine_05
  swing1LimitDegrees=45
  swing1Motion=Free
  swing2LimitDegrees=45
  swing2Motion=Free
  twistLimitDegrees=45
  twistMotion=Free
 }
 {
  bone1Name=neck_02
  bone2Name=neck_01
  swing1LimitDegrees=45
  swing1Motion=Free
  swing2LimitDegrees=45
  swing2Motion=Free
  twistLimitDegrees=45
  twistMotion=Free
 }
 {
  bone1Name=spine_05
  bone2Name=spine_04
  swing1LimitDegrees=45
  swing1Motion=Free
  swing2LimitDegrees=45
  swing2Motion=Free
  twistLimitDegrees=45
  twistMotion=Free
 }
 {
  bone1Name=clavicle_l
  bone2Name=spine_05
  swing1LimitDegrees=45
  swing1Motion=Free
  swing2LimitDegrees=45
  swing2Motion=Free
  twistLimitDegrees=45
  twistMotion=Free
 }
 {bone1Name=dyn_hairBack_C_01, bone2Name=dyn_hairBack_C_MSTR, swing1LimitDegrees=25, swing2LimitDegrees=50, twistLimitDegrees=20}
 {
  bone1Name=dyn_hairBack_C_MSTR
  bone2Name=head
  swing1LimitDegrees=45
  swing1Motion=Free
  swing2LimitDegrees=45
  swing2Motion=Free
  twistLimitDegrees=45
  twistMotion=Free
 }
 {bone1Name=dyn_hairBack_D_01, bone2Name=dyn_hairBack_D_MSTR, swing1LimitDegrees=5, swing2LimitDegrees=45, twistLimitDegrees=5}
 {
  bone1Name=dyn_hairBack_D_MSTR
  bone2Name=head
  swing1LimitDegrees=45
  swing1Motion=Free
  swing2LimitDegrees=45
  swing2Motion=Free
  twistLimitDegrees=45
  twistMotion=Free
 }
 {
  bone1Name=dyn_hairShoulder_R_A_MSTR
  bone2Name=head
  swing1LimitDegrees=45
  swing1Motion=Free
  swing2LimitDegrees=45
  swing2Motion=Free
  twistLimitDegrees=45
  twistMotion=Free
 }
 {bone1Name=dyn_hair_L_A_01, bone2Name=dyn_hair_L_A_MSTR, swing1LimitDegrees=5, swing2LimitDegrees=5, twistLimitDegrees=5}
 {
  bone1Name=dyn_hair_L_A_MSTR
  bone2Name=head
  swing1LimitDegrees=45
  swing1Motion=Free
  swing2LimitDegrees=45
  swing2Motion=Free
  twistLimitDegrees=45
  twistMotion=Free
 }
 {bone1Name=dyn_hairShoulder_R_A_01, bone2Name=dyn_hairShoulder_R_A_MSTR, swing1LimitDegrees=30, swing2LimitDegrees=10, twistLimitDegrees=5}
 {bone1Name=dyn_hairShoulder_R_A_02, bone2Name=dyn_hairShoulder_R_A_01, swing1LimitDegrees=30, swing2LimitDegrees=10, twistLimitDegrees=5}
 {bone1Name=dyn_hairShoulder_R_A_03, bone2Name=dyn_hairShoulder_R_A_02, swing1LimitDegrees=30, swing2LimitDegrees=10, twistLimitDegrees=5}
 {bone1Name=dyn_hairShoulder_R_A_04, bone2Name=dyn_hairShoulder_R_A_03, swing1LimitDegrees=30, swing2LimitDegrees=10, twistLimitDegrees=5}
 {bone1Name=dyn_hairShoulder_R_A_05, bone2Name=dyn_hairShoulder_R_A_04, swing1LimitDegrees=30, swing2LimitDegrees=15, twistLimitDegrees=15}
 {bone1Name=dyn_hairShoulder_R_B_01, bone2Name=dyn_hairShoulder_R_A_02, swing1LimitDegrees=15, swing2LimitDegrees=15, twistLimitDegrees=15}
 {bone1Name=dyn_hairShoulder_R_B_02, bone2Name=dyn_hairShoulder_R_B_01, swing1LimitDegrees=15, swing2LimitDegrees=15, twistLimitDegrees=15}
 {bone1Name=dyn_hairShoulder_R_B_03, bone2Name=dyn_hairShoulder_R_B_02, swing1LimitDegrees=15, swing2LimitDegrees=15, twistLimitDegrees=15}
 {bone1Name=dyn_hairBack_C_05, bone2Name=dyn_hairBack_C_04, swing1LimitDegrees=20, swing2LimitDegrees=50, twistLimitDegrees=20}
 {bone1Name=dyn_hairBack_B_01, bone2Name=dyn_hairBack_B_MSTR, swing1LimitDegrees=10, swing2LimitDegrees=50, twistLimitDegrees=20}
 {bone1Name=dyn_hairBack_B_02, bone2Name=dyn_hairBack_B_01, swing1LimitDegrees=10, swing2LimitDegrees=50, twistLimitDegrees=20}
 {bone1Name=dyn_hairBack_B_03, bone2Name=dyn_hairBack_B_02, swing1LimitDegrees=10, swing2LimitDegrees=50, twistLimitDegrees=20}
 {bone1Name=dyn_hairBack_B_04, bone2Name=dyn_hairBack_B_03, swing1LimitDegrees=10, swing2LimitDegrees=50, twistLimitDegrees=20}
 {bone1Name=dyn_hairBack_B_05, bone2Name=dyn_hairBack_B_04, swing1LimitDegrees=10, swing2LimitDegrees=50, twistLimitDegrees=20}
 {
  bone1Name=dyn_hairBack_A_01
  bone2Name=dyn_hairBack_B_01
  swing1LimitDegrees=45
  swing1Motion=Free
  swing2LimitDegrees=45
  swing2Motion=Free
  twistLimitDegrees=45
  twistMotion=Free
 }
 {
  bone1Name=dyn_hairBack_C_01
  bone2Name=dyn_hairBack_B_01
  swing1LimitDegrees=45
  swing1Motion=Free
  swing2LimitDegrees=45
  swing2Motion=Free
  twistLimitDegrees=45
  twistMotion=Free
 }
 {
  bone1Name=dyn_hairBack_A_02
  bone2Name=dyn_hairBack_B_02
  swing1LimitDegrees=45
  swing1Motion=Free
  swing2LimitDegrees=45
  swing2Motion=Free
  twistLimitDegrees=45
  twistMotion=Free
 }
 {
  bone1Name=dyn_hairBack_C_02
  bone2Name=dyn_hairBack_B_02
  swing1LimitDegrees=45
  swing1Motion=Free
  swing2LimitDegrees=45
  swing2Motion=Free
  twistLimitDegrees=45
  twistMotion=Free
 }
 {
  bone1Name=dyn_hairBack_A_03
  bone2Name=dyn_hairBack_B_03
  swing1LimitDegrees=45
  swing1Motion=Free
  swing2LimitDegrees=45
  swing2Motion=Free
  twistLimitDegrees=45
  twistMotion=Free
 }
 {
  bone1Name=dyn_hairBack_C_03
  bone2Name=dyn_hairBack_B_03
  swing1LimitDegrees=45
  swing1Motion=Free
  swing2LimitDegrees=45
  swing2Motion=Free
  twistLimitDegrees=45
  twistMotion=Free
 }
 {
  bone1Name=dyn_hairBack_C_04
  bone2Name=dyn_hairBack_B_04
  swing1LimitDegrees=45
  swing1Motion=Free
  swing2LimitDegrees=45
  swing2Motion=Free
  twistLimitDegrees=45
  twistMotion=Free
 }
 {
  bone1Name=dyn_hairBack_C_05
  bone2Name=dyn_hairBack_B_05
  swing1LimitDegrees=45
  swing1Motion=Free
  swing2LimitDegrees=45
  swing2Motion=Free
  twistLimitDegrees=45
  twistMotion=Free
 }
 {bone1Name=dyn_earing_L, bone2Name=head, swing1LimitDegrees=20, swing2LimitDegrees=20, twistLimitDegrees=4}
 {bone1Name=dyn_earing_R, bone2Name=head, swing1LimitDegrees=20, swing2LimitDegrees=20, twistLimitDegrees=4}
]
