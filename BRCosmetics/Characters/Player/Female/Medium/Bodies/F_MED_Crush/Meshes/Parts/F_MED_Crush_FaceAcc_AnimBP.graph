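# Post-process style secondary-motion AnimBP for the Crush character's FaceAcc part, animating hair, bows and earrings on a pose copied from the parent mesh.
# Front end uses full RigidBody physics plus an earrings pass; gameplay uses LowMS physics, switching setups on crouch.
# Bone offsets and a Control Rig follow, then chained AnimDynamics for bows and hair strands.
# Partial-alpha constraints tie hair to neighboring strands, head and spine.
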AnimBlueprint /BRCosmetics/Characters/Player/Female/Medium/Bodies/F_MED_Crush/Meshes/Parts/F_MED_Crush_FaceAcc_AnimBP
target_skeleton=&/F_MED_Crush_FaceAcc_Skeleton
graphs [
 AnimGraph {nodes [
   n1 = RigidBody!(
    ComponentPose=Local To Component(LocalPose=Copy Pose From Mesh() {bCopyCurves=true, bUseAttachedParent=true})
    Alpha
    OverrideWorldGravity:Vector=Get GravityOverride()
    ExternalForce
   ) -> (Pose) {
    bClampLinearTranslationLimitToRefPose=true
    bOverrideWorldGravity=true
    overridePhysicsAsset=&/F_MED_Crush_FaceAcc_LowMS_Physics
    simulationSpace=BaseBoneSpace
    alphaBoolBlend {blendOption=Linear}
    baseBoneRef {boneName=neck_01}
    componentAppliedLinearAccClamp=<2000 2000 2000>
    componentLinearAccScale=<0.37 0.37 0.5>
    overrideWorldGravity {z=-980}
   }
   n2 = RigidBody!(ComponentPose=Local To Component(LocalPose=Copy Pose From Mesh() {bCopyCurves=true, bUseAttachedParent=true}), Alpha, ExternalForce) -> (Pose) {
    bClampLinearTranslationLimitToRefPose=true
    bOverrideWorldGravity=true
    overridePhysicsAsset=&F_MED_Crush_FaceAcc_LowMS_Physics
    simulationSpace=BaseBoneSpace
    alphaBoolBlend {blendOption=Linear}
    baseBoneRef {boneName=root}
    componentLinearAccScale=<0.5 0.5 0.5>
    overrideWorldGravity {z=-980}
   }
   n3 = Get bIsMoving2D()
   n4 = "Transform (Modify) Bone - Bone: dyn_hairBack_A_02"!(ComponentPose=@n2, bAlphaBoolEnabled:bool=@n3, Translation="0,0,0", Rotation="0.000,0.000,20", Scale="1,1,1") -> (Pose) {
    alphaInputType=Bool
    rotationMode=BMM_Additive
    scaleMode=BMM_Additive
    translationMode=BMM_Additive
    alphaBoolBlend {blendInTime=0.1, blendOption=Linear, blendOutTime=0.1}
    boneToModify {boneName=dyn_hairBack_A_02}
    rotation {roll=20}
   }
   n5 = "Transform (Modify) Bone - Bone: dyn_hairBack_B_02"!(ComponentPose=@n4, bAlphaBoolEnabled:bool=@n3, Translation="0,0,0", Rotation="0.000,0.000,20", Scale="1,1,1") -> (Pose) {
    alphaInputType=Bool
    rotationMode=BMM_Additive
    scaleMode=BMM_Additive
    translationMode=BMM_Additive
    alphaBoolBlend {blendInTime=0.1, blendOption=Linear, blendOutTime=0.1}
    boneToModify {boneName=dyn_hairBack_B_02}
    rotation {roll=20}
   }
   n6 = "Transform (Modify) Bone - Bone: dyn_hairBack_C_02"!(ComponentPose=@n5, bAlphaBoolEnabled:bool=@n3, Translation="0,0,0", Rotation="0.000,0,20", Scale="1,1,1") -> (Pose) {
    alphaInputType=Bool
    rotationMode=BMM_Additive
    scaleMode=BMM_Additive
    translationMode=BMM_Additive
    alphaBoolBlend {blendInTime=0.1, blendOption=Linear, blendOutTime=0.1}
    boneToModify {boneName=dyn_hairBack_C_02}
    rotation {roll=20}
   }
   n7 = "Transform (Modify) Bone - Bone: dyn_hairBack_D_02"!(ComponentPose=@n6, bAlphaBoolEnabled:bool=@n3, Translation="0,0,0", Rotation="0.000,0.000,20", Scale="1,1,1") -> (Pose) {
    alphaInputType=Bool
    rotationMode=BMM_Additive
    scaleMode=BMM_Additive
    translationMode=BMM_Additive
    alphaBoolBlend {blendInTime=0.1, blendOption=Linear, blendOutTime=0.1}
    boneToModify {boneName=dyn_hairBack_D_02}
    rotation {roll=20}
   }
   n8 = Blend Poses by bool(
    BlendPose_0=Component To Local(ComponentPose=@n7)
    BlendPose_1=Component To Local(ComponentPose=@n1)
    BlendTime_0=0.4
    BlendTime_1=0.8
    bActiveValue=Get bIsCrouching()
   ) {blendType=Linear}
   n9 = RigidBody!(
    ComponentPose=Local To Component(LocalPose=Copy Pose From Mesh() {bCopyCurves=true, bUseAttachedParent=true})
    Alpha
    OverrideWorldGravity:Vector=Get GravityOverride()
    ExternalForce
   ) -> (Pose) {
    bClampLinearTranslationLimitToRefPose=true
    bOverrideWorldGravity=true
    overridePhysicsAsset=&/F_MED_Crush_FaceAcc_Physics
    simulationSpace=BaseBoneSpace
    alphaBoolBlend {blendOption=Linear}
    baseBoneRef {boneName=neck_01}
    componentAppliedLinearAccClamp=<2000 2000 2000>
    componentLinearAccScale=<0.37 0.37 0.5>
    overrideWorldGravity {z=-980}
   }
   n10 = RigidBody!(ComponentPose=@n9, Alpha, ExternalForce) -> (Pose) {
    bClampLinearTranslationLimitToRefPose=true
    bOverrideWorldGravity=true
    overridePhysicsAsset=&/F_MED_Crush_FaceAcc_Earings_Physics
    alphaBoolBlend {blendOption=Linear}
    componentAppliedLinearAccClamp=<2500 2500 2500>
    componentLinearAccScale=<0.01 0.01 0.01>
    overrideWorldGravity {z=-1100}
   }
   n11 = Component To Local(ComponentPose=@n10)
   n12 = Get bIsInFrontEnd()
   n13 = "Transform (Modify) Bone - Bone: dyn_hairBack_D_02"(
    ComponentPose=Local To Component(LocalPose=Blend Poses by bool(BlendPose_0=@n11, BlendPose_1=@n8, BlendTime_0=0.4, BlendTime_1=0.8, bActiveValue=@n12) {blendType=Linear})
    bAlphaBoolEnabled=True
    Translation="0,0,0"
    Rotation="4.584,0,15.000"
    Scale="1,1,1"
   ) {
    alphaInputType=Bool
    rotationMode=BMM_Additive
    scaleMode=BMM_Additive
    translationMode=BMM_Additive
    alphaBoolBlend {blendInTime=0.2, blendOption=Linear, blendOutTime=0.2}
    boneToModify {boneName=dyn_hairBack_D_02}
    rotation {pitch=4.584, roll=15}
   }
   n14 = "Constraint - dyn_hairShoulder_R_A_07 (PP)"!(ComponentPose=@n13, Alpha=1, ConstraintWeights_0=0.5, ConstraintWeights_1:float=0.5) -> (Pose) {
    alphaBoolBlend {blendOption=Linear}
    boneToModify {boneName=dyn_hairShoulder_R_A_07}
    constraintSetup [{targetBone {boneName=dyn_hairBack_D_03}}, {targetBone {boneName=dyn_hairShoulder_R_A_03}}]
    constraintWeights [1,1]
   }
   n15 = Control Rig(Source=Component To Local(ComponentPose=@n14)) {
    alphaBoolBlend {blendOption=Linear}
    controlRigAssetReference {blueprintRigClass=&/F_MED_Crush_FaceAcc_CtrlRig.F_MED_Crush_FaceAcc_CtrlRig_C}
   }
   n16 = "To Float (Boolean)"(InBool=Get bIsCrouching())
   n17 = "Transform (Modify) Bone - Bone: dyn_hairBack_D_02"!(ComponentPose=Local To Component(LocalPose=@n15), Alpha:float=@n16, Translation="0,0,1", Rotation="0,0,0", Scale="1,1,1") -> (Pose) {
    rotationMode=BMM_Additive
    scaleMode=BMM_Additive
    translationMode=BMM_Additive
    alphaBoolBlend {blendOption=Linear}
    boneToModify {boneName=dyn_hairBack_D_02}
    translation {z=1}
   }
   n18 = "Transform (Modify) Bone - Bone: dyn_hairShoulder_R_A_02"(ComponentPose=@n17, Alpha=@n16, Translation="0,0,1", Rotation="0,0,0", Scale="1,1,1") {
    rotationMode=BMM_Additive
    scaleMode=BMM_Additive
    translationMode=BMM_Additive
    alphaBoolBlend {blendOption=Linear}
    boneToModify {boneName=dyn_hairShoulder_R_A_02}
    translation {z=1}
   }
   n19 = "Constraint - dyn_hairBack_C_01 (P)"(ComponentPose=@n18, Alpha=0.2, ConstraintWeights_0=1) {
    alphaBoolBlend {blendOption=Linear}
    boneToModify {boneName=dyn_hairBack_C_01}
    constraintSetup [{targetBone {boneName=dyn_hairBack_D_01}}]
    constraintWeights [1]
   }
   n20 = "Constraint - dyn_hairBack_C_02 (P)"(ComponentPose=@n19, Alpha=0.15, ConstraintWeights_0=1) {
    alphaBoolBlend {blendOption=Linear}
    boneToModify {boneName=dyn_hairBack_C_02}
    constraintSetup [{targetBone {boneName=dyn_hairBack_D_02}}]
    constraintWeights [1]
   }
   n21 = "Constraint - dyn_hairBack_C_03 (P)"(ComponentPose=@n20, Alpha=0.05, ConstraintWeights_0=1) {
    alphaBoolBlend {blendOption=Linear}
    boneToModify {boneName=dyn_hairBack_C_03}
    constraintSetup [{targetBone {boneName=dyn_hairBack_D_03}}]
    constraintWeights [1]
   }
   n22 = "Constraint - dyn_hairBack_C_04 (P)"(ComponentPose=@n21, Alpha=0.01, ConstraintWeights_0=1) {
    alphaBoolBlend {blendOption=Linear}
    boneToModify {boneName=dyn_hairBack_C_04}
    constraintSetup [{targetBone {boneName=dyn_hairBack_D_04}}]
    constraintWeights [1]
   }
   n23 = Anim Dynamics - Bone: dyn_bow_C_L_01(ComponentPose=@n22, Alpha=0.35, ExternalForce="0,0,0") {
    angularDampingOverride=0.9999
    bGravityOverrideInSimSpace=true
    bOverrideAngularDamping=true
    gravityScale=0.25
    alphaBoolBlend {blendOption=Linear}
    boundBone {boneName=dyn_bow_C_L_01}
    componentLinearAccScale=<0.4 0.35 0.25>
    physicsBodyDefinitions [{
      sphereCollisionRadius=0
      boundBone {boneName=dyn_bow_C_L_01}
      boxExtents=<5 5 5>
      constraintSetup {angularLimitsMax=<0.2866 10 75.9554>, angularLimitsMin=<-17.4841 -45 -7.1656>, angularTarget=<0 0 0>}
      localJointOffset=<1 0 -0.3>
     }]
   }
   n24 = Anim Dynamics - Bone: dyn_bow_C_R_01(ComponentPose=@n23, Alpha=0.35, ExternalForce="0,0,0") {
    angularDampingOverride=0.9999
    bGravityOverrideInSimSpace=true
    bOverrideAngularDamping=true
    gravityScale=0.25
    alphaBoolBlend {blendOption=Linear}
    boundBone {boneName=dyn_bow_C_R_01}
    componentLinearAccScale=<0.4 0.35 0.25>
    physicsBodyDefinitions [{
      sphereCollisionRadius=0
      boundBone {boneName=dyn_bow_C_R_01}
      boxExtents=<5 5 5>
      constraintSetup {angularLimitsMax=<0.2866 10 75.9554>, angularLimitsMin=<-17.4841 -45 -7.1656>, angularTarget=<0 0 0>}
      localJointOffset=<1 0 -0.3>
     }]
   }
   n25 = Anim Dynamics - Bone: dyn_bow_B_L_01(ComponentPose=@n24, Alpha=0.4, ExternalForce="0,0,0") {
    angularDampingOverride=0.9999
    bGravityOverrideInSimSpace=true
    bOverrideAngularDamping=true
    gravityScale=0.25
    alphaBoolBlend {blendOption=Linear}
    boundBone {boneName=dyn_bow_B_L_01}
    componentLinearAccScale=<0.4 0.35 0.25>
    physicsBodyDefinitions [{
      sphereCollisionRadius=0
      boundBone {boneName=dyn_bow_B_L_01}
      boxExtents=<5 5 5>
      constraintSetup {angularLimitsMax=<20 10 20>, angularLimitsMin=<-20 -45 -20>, angularTarget=<0 0 0>}
      localJointOffset=<1 0 -0.3>
     }]
   }
   n26 = Anim Dynamics - Bone: dyn_bow_B_R_01(ComponentPose=@n25, Alpha=0.2, ExternalForce="0,0,0") {
    angularDampingOverride=0.9999
    bAngularSpring=true
    bGravityOverrideInSimSpace=true
    bOverrideAngularDamping=true
    gravityScale=0.25
    alphaBoolBlend {blendOption=Linear}
    boundBone {boneName=dyn_bow_B_R_01}
    componentLinearAccScale=<0.4 0.35 0.25>
    physicsBodyDefinitions [{
      sphereCollisionRadius=0
      boundBone {boneName=dyn_bow_B_R_01}
      boxExtents=<5 5 5>
      constraintSetup {angularLimitsMax=<20 10 20>, angularLimitsMin=<-20 -45 -20>, angularTarget=<0 0 0>}
      localJointOffset=<1 0 -0.3>
     }]
   }
   n27 = Anim Dynamics - Bone: dyn_bow_A_L_01(ComponentPose=@n26, Alpha=0.2, ExternalForce="0,0,0") {
    angularDampingOverride=0.9999
    bGravityOverrideInSimSpace=true
    bOverrideAngularDamping=true
    gravityScale=0.25
    alphaBoolBlend {blendOption=Linear}
    boundBone {boneName=dyn_bow_A_L_01}
    componentLinearAccScale=<0.4 0.35 0.25>
    physicsBodyDefinitions [{
      sphereCollisionRadius=0
      boundBone {boneName=dyn_bow_A_L_01}
      boxExtents=<5 5 5>
      constraintSetup {angularLimitsMax=<0.2866 10 75.9554>, angularLimitsMin=<-17.4841 -45 -7.1656>, angularTarget=<0 0 0>}
      localJointOffset=<1 0 -0.3>
     }]
   }
   n28 = Anim Dynamics - Bone: dyn_bow_A_R_01(ComponentPose=@n27, Alpha=0.2, ExternalForce="0,0,0") {
    angularDampingOverride=0.9999
    bGravityOverrideInSimSpace=true
    bOverrideAngularDamping=true
    gravityScale=0.25
    alphaBoolBlend {blendOption=Linear}
    boundBone {boneName=dyn_bow_A_R_01}
    componentLinearAccScale=<0.4 0.35 0.25>
    physicsBodyDefinitions [{
      sphereCollisionRadius=0
      boundBone {boneName=dyn_bow_A_R_01}
      boxExtents=<5 5 5>
      constraintSetup {angularLimitsMax=<0.2866 10 75.9554>, angularLimitsMin=<-17.4841 -45 -7.1656>, angularTarget=<0 0 0>}
      localJointOffset=<1 0 -0.3>
     }]
   }
   n29 = Anim Dynamics - Bone: dyn_bow_D_L_01(ComponentPose=@n28, Alpha=0.1, ExternalForce="0,0,0") {
    angularDampingOverride=0.9999
    bGravityOverrideInSimSpace=true
    bOverrideAngularDamping=true
    gravityScale=0.25
    alphaBoolBlend {blendOption=Linear}
    boundBone {boneName=dyn_bow_D_L_01}
    componentLinearAccScale=<0.4 0.35 0.25>
    physicsBodyDefinitions [{
      sphereCollisionRadius=0
      boundBone {boneName=dyn_bow_D_L_01}
      boxExtents=<5 5 5>
      constraintSetup {angularLimitsMax=<0.2866 10 75.9554>, angularLimitsMin=<-17.4841 -45 -7.1656>, angularTarget=<0 0 0>}
      localJointOffset=<1 0 -0.3>
     }]
   }
   n30 = Anim Dynamics - Bone: dyn_bow_D_R_01(ComponentPose=@n29, Alpha=0.1, ExternalForce="0,0,0") {
    angularDampingOverride=0.9999
    bGravityOverrideInSimSpace=true
    bOverrideAngularDamping=true
    gravityScale=0.25
    alphaBoolBlend {blendOption=Linear}
    boundBone {boneName=dyn_bow_D_R_01}
    componentLinearAccScale=<0.4 0.35 0.25>
    physicsBodyDefinitions [{
      sphereCollisionRadius=0
      boundBone {boneName=dyn_bow_D_R_01}
      boxExtents=<5 5 5>
      constraintSetup {angularLimitsMax=<0.2866 10 75.9554>, angularLimitsMin=<-17.4841 -45 -7.1656>, angularTarget=<0 0 0>}
      localJointOffset=<1 0 -0.3>
     }]
   }
   n31 = Anim Dynamics - Bone: dyn_bow_E_L_01(ComponentPose=@n30, Alpha=0.15, ExternalForce="0,0,0") {
    angularDampingOverride=0.9999
    bGravityOverrideInSimSpace=true
    bOverrideAngularDamping=true
    gravityScale=0.25
    alphaBoolBlend {blendOption=Linear}
    boundBone {boneName=dyn_bow_E_L_01}
    componentLinearAccScale=<0.4 0.35 0.25>
    physicsBodyDefinitions [{
      sphereCollisionRadius=0
      boundBone {boneName=dyn_bow_E_L_01}
      boxExtents=<5 5 5>
      constraintSetup {angularLimitsMax=<0.2866 10 75.9554>, angularLimitsMin=<-17.4841 -45 -7.1656>, angularTarget=<0 0 0>}
      localJointOffset=<1 0 -0.3>
     }]
   }
   n32 = Anim Dynamics - Bone: dyn_bow_E_R_01(ComponentPose=@n31, Alpha=0.15, ExternalForce="0,0,0") {
    angularDampingOverride=0.9999
    bGravityOverrideInSimSpace=true
    bOverrideAngularDamping=true
    gravityScale=0.25
    alphaBoolBlend {blendOption=Linear}
    boundBone {boneName=dyn_bow_E_R_01}
    componentLinearAccScale=<0.4 0.35 0.25>
    physicsBodyDefinitions [{
      sphereCollisionRadius=0
      boundBone {boneName=dyn_bow_E_R_01}
      boxExtents=<5 5 5>
      constraintSetup {angularLimitsMax=<0.2866 10 75.9554>, angularLimitsMin=<-17.4841 -45 -7.1656>, angularTarget=<0 0 0>}
      localJointOffset=<1 0 -0.3>
     }]
   }
   n33 = Get GravityOverride()
   n34 = Anim Dynamics - Bone: dyn_hairBack_A_01!(ComponentPose=@n32, Alpha=1, GravityOverride:Vector=@n33, ExternalForce="0,0,0") -> (Pose) {
    angularDampingOverride=0.9999
    bAngularSpring=true
    bGravityOverrideInSimSpace=true
    bOverrideAngularDamping=true
    bUseGravityOverride=true
    simulationSpace=RootRelative
    alphaBoolBlend {blendOption=Linear}
    boundBone {boneName=dyn_hairBack_A_01}
    componentLinearAccScale=<2 2 2>
    componentLinearVelScale=<2 2 2>
    physicsBodyDefinitions [{
      sphereCollisionRadius=0
      boundBone {boneName=dyn_hairBack_A_01}
      boxExtents=<20 20 10>
      constraintSetup {angularLimitsMax=<45 -5 45>, angularLimitsMin=<-45 -45 -45>, angularTarget=<0 0 0>}
      localJointOffset=<1 0 -0.5>
     }]
   }
   n35 = Anim Dynamics - Bone: dyn_hairBack_A_02!(ComponentPose=@n34, Alpha=1, GravityOverride:Vector=@n33, ExternalForce="0,0,0") -> (Pose) {
    angularDampingOverride=0.9999
    bAngularSpring=true
    bGravityOverrideInSimSpace=true
    bOverrideAngularDamping=true
    bUseGravityOverride=true
    simulationSpace=RootRelative
    alphaBoolBlend {blendOption=Linear}
    boundBone {boneName=dyn_hairBack_A_02}
    componentLinearAccScale=<2 2 2>
    componentLinearVelScale=<2 2 2>
    physicsBodyDefinitions [{
      sphereCollisionRadius=0
      boundBone {boneName=dyn_hairBack_A_02}
      boxExtents=<20 20 10>
      constraintSetup {angularLimitsMax=<45 -5 45>, angularLimitsMin=<-45 -45 -45>, angularTarget=<0 0 0>}
      localJointOffset=<1 0 -0.5>
     }]
   }
   n36 = Anim Dynamics - Bone: dyn_hairBack_A_03!(ComponentPose=@n35, Alpha=1, GravityOverride:Vector=@n33, ExternalForce="0,0,0") -> (Pose) {
    angularDampingOverride=0.9999
    bAngularSpring=true
    bGravityOverrideInSimSpace=true
    bOverrideAngularDamping=true
    bUseGravityOverride=true
    simulationSpace=RootRelative
    alphaBoolBlend {blendOption=Linear}
    boundBone {boneName=dyn_hairBack_A_03}
    componentLinearAccScale=<2 2 2>
    componentLinearVelScale=<2 2 2>
    physicsBodyDefinitions [{
      sphereCollisionRadius=0
      boundBone {boneName=dyn_hairBack_A_03}
      boxExtents=<20 20 10>
      constraintSetup {angularLimitsMax=<45 -5 45>, angularLimitsMin=<-45 -45 -45>, angularTarget=<0 0 0>}
      localJointOffset=<1 0 -0.5>
     }]
   }
   n37 = Anim Dynamics - Bone: dyn_hairBack_A_04!(ComponentPose=@n36, Alpha=1, GravityOverride:Vector=@n33, ExternalForce="0,0,0") -> (Pose) {
    angularDampingOverride=0.9999
    bAngularSpring=true
    bGravityOverrideInSimSpace=true
    bOverrideAngularDamping=true
    bUseGravityOverride=true
    simulationSpace=RootRelative
    alphaBoolBlend {blendOption=Linear}
    boundBone {boneName=dyn_hairBack_A_04}
    componentLinearAccScale=<2 2 2>
    componentLinearVelScale=<2 2 2>
    physicsBodyDefinitions [{
      sphereCollisionRadius=0
      boundBone {boneName=dyn_hairBack_A_04}
      boxExtents=<20 20 10>
      constraintSetup {angularLimitsMax=<45 -5 45>, angularLimitsMin=<-45 -45 -45>, angularTarget=<0 0 0>}
      localJointOffset=<1 0 -0.5>
     }]
   }
   n38 = Anim Dynamics - Bone: dyn_hairBack_B_01!(ComponentPose=@n37, Alpha=1, GravityOverride:Vector=@n33, ExternalForce="0,0,0") -> (Pose) {
    angularDampingOverride=0.9999
    bAngularSpring=true
    bGravityOverrideInSimSpace=true
    bOverrideAngularDamping=true
    bUseGravityOverride=true
    simulationSpace=RootRelative
    alphaBoolBlend {blendOption=Linear}
    boundBone {boneName=dyn_hairBack_B_01}
    componentLinearAccScale=<2 2 2>
    componentLinearVelScale=<2 2 2>
    physicsBodyDefinitions [{
      sphereCollisionRadius=0
      boundBone {boneName=dyn_hairBack_B_01}
      boxExtents=<20 20 10>
      constraintSetup {angularLimitsMax=<45 -5 45>, angularLimitsMin=<-45 -45 -45>, angularTarget=<0 0 0>}
      localJointOffset=<1 0 -0.5>
     }]
   }
   n39 = Anim Dynamics - Bone: dyn_hairBack_B_02!(ComponentPose=@n38, Alpha=1, GravityOverride:Vector=@n33, ExternalForce="0,0,0") -> (Pose) {
    angularDampingOverride=0.9999
    bAngularSpring=true
    bGravityOverrideInSimSpace=true
    bOverrideAngularDamping=true
    bUseGravityOverride=true
    simulationSpace=RootRelative
    alphaBoolBlend {blendOption=Linear}
    boundBone {boneName=dyn_hairBack_B_02}
    componentLinearAccScale=<2 2 2>
    componentLinearVelScale=<2 2 2>
    physicsBodyDefinitions [{
      sphereCollisionRadius=0
      boundBone {boneName=dyn_hairBack_B_02}
      boxExtents=<20 20 10>
      constraintSetup {angularLimitsMax=<45 -5 45>, angularLimitsMin=<-45 -45 -45>, angularTarget=<0 0 0>}
      localJointOffset=<1 0 -0.5>
     }]
   }
   n40 = Anim Dynamics - Bone: dyn_hairBack_B_03!(ComponentPose=@n39, Alpha=1, GravityOverride:Vector=@n33, ExternalForce="0,0,0") -> (Pose) {
    angularDampingOverride=0.9999
    bAngularSpring=true
    bGravityOverrideInSimSpace=true
    bOverrideAngularDamping=true
    bUseGravityOverride=true
    simulationSpace=RootRelative
    alphaBoolBlend {blendOption=Linear}
    boundBone {boneName=dyn_hairBack_B_03}
    componentLinearAccScale=<2 2 2>
    componentLinearVelScale=<2 2 2>
    physicsBodyDefinitions [{
      sphereCollisionRadius=0
      boundBone {boneName=dyn_hairBack_B_03}
      boxExtents=<20 20 10>
      constraintSetup {angularLimitsMax=<45 -5 45>, angularLimitsMin=<-45 -45 -45>, angularTarget=<0 0 0>}
      localJointOffset=<1 0 -0.5>
     }]
   }
   n41 = Anim Dynamics - Bone: dyn_hairBack_B_04!(ComponentPose=@n40, Alpha=1, GravityOverride:Vector=@n33, ExternalForce="0,0,0") -> (Pose) {
    angularDampingOverride=0.9999
    bAngularSpring=true
    bGravityOverrideInSimSpace=true
    bOverrideAngularDamping=true
    bUseGravityOverride=true
    simulationSpace=RootRelative
    alphaBoolBlend {blendOption=Linear}
    boundBone {boneName=dyn_hairBack_B_04}
    componentLinearAccScale=<2 2 2>
    componentLinearVelScale=<2 2 2>
    physicsBodyDefinitions [{
      sphereCollisionRadius=0
      boundBone {boneName=dyn_hairBack_B_04}
      boxExtents=<20 20 10>
      constraintSetup {angularLimitsMax=<45 -5 45>, angularLimitsMin=<-45 -45 -45>, angularTarget=<0 0 0>}
      localJointOffset=<1 0 -0.5>
     }]
   }
   n42 = Anim Dynamics - Bone: dyn_hairBack_B_05!(ComponentPose=@n41, Alpha=1, GravityOverride:Vector=@n33, ExternalForce="0,0,0") -> (Pose) {
    angularDampingOverride=0.9999
    bAngularSpring=true
    bGravityOverrideInSimSpace=true
    bOverrideAngularDamping=true
    bUseGravityOverride=true
    simulationSpace=RootRelative
    alphaBoolBlend {blendOption=Linear}
    boundBone {boneName=dyn_hairBack_B_05}
    componentLinearAccScale=<2 2 2>
    componentLinearVelScale=<2 2 2>
    physicsBodyDefinitions [{
      sphereCollisionRadius=0
      boundBone {boneName=dyn_hairBack_B_05}
      boxExtents=<20 20 10>
      constraintSetup {angularLimitsMax=<45 -5 45>, angularLimitsMin=<-45 -45 -45>, angularTarget=<0 0 0>}
      localJointOffset=<1 0 -0.5>
     }]
   }
   n43 = Anim Dynamics - Bone: dyn_hairBack_C_01!(ComponentPose=@n42, Alpha=1, GravityOverride:Vector=@n33, ExternalForce="0,0,0") -> (Pose) {
    angularDampingOverride=0.9999
    bAngularSpring=true
    bGravityOverrideInSimSpace=true
    bOverrideAngularDamping=true
    bUseGravityOverride=true
    simulationSpace=RootRelative
    alphaBoolBlend {blendOption=Linear}
    boundBone {boneName=dyn_hairBack_C_01}
    componentLinearAccScale=<2 2 2>
    componentLinearVelScale=<2 2 2>
    physicsBodyDefinitions [{
      sphereCollisionRadius=0
      boundBone {boneName=dyn_hairBack_C_01}
      boxExtents=<20 20 10>
      constraintSetup {angularLimitsMax=<45 -5 45>, angularLimitsMin=<-45 -45 -45>, angularTarget=<0 0 0>}
      localJointOffset=<1 0 -0.5>
     }]
   }
   n44 = Anim Dynamics - Bone: dyn_hairBack_C_02!(ComponentPose=@n43, Alpha=1, GravityOverride:Vector=@n33, ExternalForce="0,0,0") -> (Pose) {
    angularDampingOverride=0.9999
    bAngularSpring=true
    bGravityOverrideInSimSpace=true
    bOverrideAngularDamping=true
    bUseGravityOverride=true
    simulationSpace=RootRelative
    alphaBoolBlend {blendOption=Linear}
    boundBone {boneName=dyn_hairBack_C_02}
    componentLinearAccScale=<2 2 2>
    componentLinearVelScale=<2 2 2>
    physicsBodyDefinitions [{
      sphereCollisionRadius=0
      boundBone {boneName=dyn_hairBack_C_02}
      boxExtents=<20 20 10>
      constraintSetup {angularLimitsMax=<45 -5 45>, angularLimitsMin=<-45 -45 -45>, angularTarget=<0 0 0>}
      localJointOffset=<1 0 -0.5>
     }]
   }
   n45 = Anim Dynamics - Bone: dyn_hairBack_C_03!(ComponentPose=@n44, Alpha=1, GravityOverride:Vector=@n33, ExternalForce="0,0,0") -> (Pose) {
    angularDampingOverride=0.9999
    bAngularSpring=true
    bGravityOverrideInSimSpace=true
    bOverrideAngularDamping=true
    bUseGravityOverride=true
    simulationSpace=RootRelative
    alphaBoolBlend {blendOption=Linear}
    boundBone {boneName=dyn_hairBack_C_03}
    componentLinearAccScale=<2 2 2>
    componentLinearVelScale=<2 2 2>
    physicsBodyDefinitions [{
      sphereCollisionRadius=0
      boundBone {boneName=dyn_hairBack_C_03}
      boxExtents=<20 20 10>
      constraintSetup {angularLimitsMax=<45 -5 45>, angularLimitsMin=<-45 -45 -45>, angularTarget=<0 0 0>}
      localJointOffset=<1 0 -0.5>
     }]
   }
   n46 = Anim Dynamics - Bone: dyn_hairBack_C_04!(ComponentPose=@n45, Alpha=1, GravityOverride:Vector=@n33, ExternalForce="0,0,0") -> (Pose) {
    angularDampingOverride=0.9999
    bAngularSpring=true
    bGravityOverrideInSimSpace=true
    bOverrideAngularDamping=true
    bUseGravityOverride=true
    simulationSpace=RootRelative
    alphaBoolBlend {blendOption=Linear}
    boundBone {boneName=dyn_hairBack_C_04}
    componentLinearAccScale=<2 2 2>
    componentLinearVelScale=<2 2 2>
    physicsBodyDefinitions [{
      sphereCollisionRadius=0
      boundBone {boneName=dyn_hairBack_C_04}
      boxExtents=<20 20 10>
      constraintSetup {angularLimitsMax=<45 -5 45>, angularLimitsMin=<-45 -45 -45>, angularTarget=<0 0 0>}
      localJointOffset=<1 0 -0.5>
     }]
   }
   n47 = Anim Dynamics - Bone: dyn_hairBack_C_05!(ComponentPose=@n46, Alpha=1, GravityOverride:Vector=@n33, ExternalForce="0,0,0") -> (Pose) {
    angularDampingOverride=0.9999
    bAngularSpring=true
    bGravityOverrideInSimSpace=true
    bOverrideAngularDamping=true
    bUseGravityOverride=true
    simulationSpace=RootRelative
    alphaBoolBlend {blendOption=Linear}
    boundBone {boneName=dyn_hairBack_C_05}
    componentLinearAccScale=<2 2 2>
    componentLinearVelScale=<2 2 2>
    physicsBodyDefinitions [{
      sphereCollisionRadius=0
      boundBone {boneName=dyn_hairBack_C_05}
      boxExtents=<20 20 10>
      constraintSetup {angularLimitsMax=<45 -5 45>, angularLimitsMin=<-45 -45 -45>, angularTarget=<0 0 0>}
      localJointOffset=<1 0 -0.5>
     }]
   }
   n48 = Anim Dynamics - Bone: dyn_hairBack_D_01!(ComponentPose=@n47, Alpha=1, GravityOverride:Vector=@n33, ExternalForce="0,0,0") -> (Pose) {
    angularDampingOverride=0.9999
    bAngularSpring=true
    bGravityOverrideInSimSpace=true
    bOverrideAngularDamping=true
    bUseGravityOverride=true
    simulationSpace=RootRelative
    alphaBoolBlend {blendOption=Linear}
    boundBone {boneName=dyn_hairBack_D_01}
    componentLinearAccScale=<2 2 2>
    componentLinearVelScale=<2 2 2>
    physicsBodyDefinitions [{
      sphereCollisionRadius=0
      boundBone {boneName=dyn_hairBack_D_01}
      boxExtents=<20 20 10>
      constraintSetup {angularLimitsMax=<45 -5 45>, angularLimitsMin=<-45 -45 -45>, angularTarget=<0 0 0>}
      localJointOffset=<1 0 -0.5>
     }]
   }
   n49 = Anim Dynamics - Bone: dyn_hairBack_D_02!(ComponentPose=@n48, Alpha=1, GravityOverride:Vector=@n33, ExternalForce="0,0,0") -> (Pose) {
    angularDampingOverride=0.9999
    bAngularSpring=true
    bGravityOverrideInSimSpace=true
    bOverrideAngularDamping=true
    bUseGravityOverride=true
    simulationSpace=RootRelative
    alphaBoolBlend {blendOption=Linear}
    boundBone {boneName=dyn_hairBack_D_02}
    componentLinearAccScale=<2 2 2>
    componentLinearVelScale=<2 2 2>
    physicsBodyDefinitions [{
      sphereCollisionRadius=0
      boundBone {boneName=dyn_hairBack_D_02}
      boxExtents=<20 20 10>
      constraintSetup {angularLimitsMax=<45 -5 45>, angularLimitsMin=<-45 -45 -45>, angularTarget=<0 0 0>}
      localJointOffset=<1 0 -0.5>
     }]
   }
   n50 = Anim Dynamics - Bone: dyn_hairBack_D_03!(ComponentPose=@n49, Alpha=1, GravityOverride:Vector=@n33, ExternalForce="0,0,0") -> (Pose) {
    angularDampingOverride=0.9999
    bAngularSpring=true
    bGravityOverrideInSimSpace=true
    bOverrideAngularDamping=true
    bUseGravityOverride=true
    simulationSpace=RootRelative
    alphaBoolBlend {blendOption=Linear}
    boundBone {boneName=dyn_hairBack_D_03}
    componentLinearAccScale=<2 2 2>
    componentLinearVelScale=<2 2 2>
    physicsBodyDefinitions [{
      sphereCollisionRadius=0
      boundBone {boneName=dyn_hairBack_D_03}
      boxExtents=<20 20 10>
      constraintSetup {angularLimitsMax=<45 -5 45>, angularLimitsMin=<-45 -45 -45>, angularTarget=<0 0 0>}
      localJointOffset=<1 0 -0.5>
     }]
   }
   n51 = Anim Dynamics - Bone: dyn_hairBack_D_04!(ComponentPose=@n50, Alpha=1, GravityOverride:Vector=@n33, ExternalForce="0,0,0") -> (Pose) {
    angularDampingOverride=0.9999
    bAngularSpring=true
    bGravityOverrideInSimSpace=true
    bOverrideAngularDamping=true
    bUseGravityOverride=true
    simulationSpace=RootRelative
    alphaBoolBlend {blendOption=Linear}
    boundBone {boneName=dyn_hairBack_D_04}
    componentLinearAccScale=<2 2 2>
    componentLinearVelScale=<2 2 2>
    physicsBodyDefinitions [{
      sphereCollisionRadius=0
      boundBone {boneName=dyn_hairBack_D_04}
      boxExtents=<20 20 10>
      constraintSetup {angularLimitsMax=<45 -5 45>, angularLimitsMin=<-45 -45 -45>, angularTarget=<0 0 0>}
      localJointOffset=<1 0 -0.5>
     }]
   }
   n52 = Anim Dynamics - Bone: dyn_hair_R_A_01(ComponentPose=@n51, Alpha=1, ExternalForce="0,0,0") {
    angularDampingOverride=0.9999
    angularSpringConstant=75
    bAngularSpring=true
    bGravityOverrideInSimSpace=true
    bOverrideAngularDamping=true
    simulationSpace=RootRelative
    alphaBoolBlend {blendOption=Linear}
    boundBone {boneName=dyn_hair_R_A_01}
    componentLinearAccScale=<1 1 1>
    componentLinearVelScale=<1 1 1>
    physicsBodyDefinitions [{
      sphereCollisionRadius=0
      boundBone {boneName=dyn_hair_R_A_01}
      boxExtents=<10 20 20>
      constraintSetup {angularLimitsMax=<45 45 45>, angularLimitsMin=<-45 -45 -45>, angularTarget=<0 0 0>}
      localJointOffset=<1 0 -0.2>
     }]
   }
   n53 = Anim Dynamics - Bone: dyn_hair_R_A_02(ComponentPose=@n52, Alpha=1, ExternalForce="0,0,0") {
    angularDampingOverride=0.9999
    angularSpringConstant=75
    bAngularSpring=true
    bGravityOverrideInSimSpace=true
    bOverrideAngularDamping=true
    simulationSpace=RootRelative
    alphaBoolBlend {blendOption=Linear}
    boundBone {boneName=dyn_hair_R_A_02}
    componentLinearAccScale=<1 1 1>
    componentLinearVelScale=<1 1 1>
    physicsBodyDefinitions [{
      sphereCollisionRadius=0
      boundBone {boneName=dyn_hair_R_A_02}
      boxExtents=<10 20 20>
      constraintSetup {angularLimitsMax=<45 45 45>, angularLimitsMin=<-45 -45 -45>, angularTarget=<0 0 0>}
      localJointOffset=<1 0 -0.2>
     }]
   }
   n54 = Anim Dynamics - Bone: dyn_hair_R_A_03(ComponentPose=@n53, Alpha=1, ExternalForce="0,0,0") {
    angularDampingOverride=0.9999
    angularSpringConstant=75
    bAngularSpring=true
    bGravityOverrideInSimSpace=true
    bOverrideAngularDamping=true
    simulationSpace=RootRelative
    alphaBoolBlend {blendOption=Linear}
    boundBone {boneName=dyn_hair_R_A_03}
    componentLinearAccScale=<1 1 1>
    componentLinearVelScale=<1 1 1>
    physicsBodyDefinitions [{
      sphereCollisionRadius=0
      boundBone {boneName=dyn_hair_R_A_03}
      boxExtents=<10 20 20>
      constraintSetup {angularLimitsMax=<45 45 45>, angularLimitsMin=<-45 -45 -45>, angularTarget=<0 0 0>}
      localJointOffset=<1 0 -0.2>
     }]
   }
   n55 = Anim Dynamics - Bone: dyn_hair_R_A_04(ComponentPose=@n54, Alpha=1, ExternalForce="0,0,0") {
    angularDampingOverride=0.9999
    angularSpringConstant=75
    bAngularSpring=true
    bGravityOverrideInSimSpace=true
    bOverrideAngularDamping=true
    simulationSpace=RootRelative
    alphaBoolBlend {blendOption=Linear}
    boundBone {boneName=dyn_hair_R_A_04}
    componentLinearAccScale=<1 1 1>
    componentLinearVelScale=<1 1 1>
    physicsBodyDefinitions [{
      sphereCollisionRadius=0
      boundBone {boneName=dyn_hair_R_A_04}
      boxExtents=<10 20 20>
      constraintSetup {angularLimitsMax=<45 45 45>, angularLimitsMin=<-45 -45 -45>, angularTarget=<0 0 0>}
      localJointOffset=<1 0 -0.2>
     }]
   }
   n56 = Anim Dynamics - Bone: dyn_hair_R_A_05(ComponentPose=@n55, Alpha=1, ExternalForce="0,0,0") {
    angularDampingOverride=0.9999
    angularSpringConstant=75
    bAngularSpring=true
    bGravityOverrideInSimSpace=true
    bOverrideAngularDamping=true
    simulationSpace=RootRelative
    alphaBoolBlend {blendOption=Linear}
    boundBone {boneName=dyn_hair_R_A_05}
    componentLinearAccScale=<1 1 1>
    componentLinearVelScale=<1 1 1>
    physicsBodyDefinitions [{
      sphereCollisionRadius=0
      boundBone {boneName=dyn_hair_R_A_05}
      boxExtents=<10 20 20>
      constraintSetup {angularLimitsMax=<45 45 45>, angularLimitsMin=<-45 -45 -45>, angularTarget=<0 0 0>}
      localJointOffset=<1 0 -0.2>
     }]
   }
   n57 = Anim Dynamics - Bone: dyn_hair_L_A_02(ComponentPose=@n56, Alpha=1, ExternalForce="0,0,0") {
    angularDampingOverride=0.9999
    angularSpringConstant=75
    bAngularSpring=true
    bGravityOverrideInSimSpace=true
    bOverrideAngularDamping=true
    simulationSpace=RootRelative
    alphaBoolBlend {blendOption=Linear}
    boundBone {boneName=dyn_hair_L_A_02}
    componentLinearAccScale=<1 1 1>
    componentLinearVelScale=<1 1 1>
    physicsBodyDefinitions [{
      sphereCollisionRadius=0
      boundBone {boneName=dyn_hair_L_A_02}
      boxExtents=<10 20 20>
      constraintSetup {angularLimitsMax=<45 45 45>, angularLimitsMin=<-45 -45 -45>, angularTarget=<0 0 0>}
      localJointOffset=<1 0 -0.2>
     }]
   }
   n58 = Anim Dynamics - Bone: dyn_hair_L_A_03(ComponentPose=@n57, Alpha=1, ExternalForce="0,0,0") {
    angularDampingOverride=0.9999
    angularSpringConstant=75
    bAngularSpring=true
    bGravityOverrideInSimSpace=true
    bOverrideAngularDamping=true
    simulationSpace=RootRelative
    alphaBoolBlend {blendOption=Linear}
    boundBone {boneName=dyn_hair_L_A_03}
    componentLinearAccScale=<1 1 1>
    componentLinearVelScale=<1 1 1>
    physicsBodyDefinitions [{
      sphereCollisionRadius=0
      boundBone {boneName=dyn_hair_L_A_03}
      boxExtents=<10 20 20>
      constraintSetup {angularLimitsMax=<45 45 45>, angularLimitsMin=<-45 -45 -45>, angularTarget=<0 0 0>}
      localJointOffset=<1 0 -0.2>
     }]
   }
   n59 = Anim Dynamics - Bone: dyn_hair_L_A_04(ComponentPose=@n58, Alpha=1, ExternalForce="0,0,0") {
    angularDampingOverride=0.9999
    angularSpringConstant=75
    bAngularSpring=true
    bGravityOverrideInSimSpace=true
    bOverrideAngularDamping=true
    simulationSpace=RootRelative
    alphaBoolBlend {blendOption=Linear}
    boundBone {boneName=dyn_hair_L_A_04}
    componentLinearAccScale=<1 1 1>
    componentLinearVelScale=<1 1 1>
    physicsBodyDefinitions [{
      sphereCollisionRadius=0
      boundBone {boneName=dyn_hair_L_A_04}
      boxExtents=<10 20 20>
      constraintSetup {angularLimitsMax=<45 45 45>, angularLimitsMin=<-45 -45 -45>, angularTarget=<0 0 0>}
      localJointOffset=<1 0 -0.2>
     }]
   }
   n60 = Anim Dynamics - Bone: dyn_hair_L_A_05(ComponentPose=@n59, Alpha=1, ExternalForce="0,0,0") {
    angularDampingOverride=0.9999
    angularSpringConstant=75
    bAngularSpring=true
    bGravityOverrideInSimSpace=true
    bOverrideAngularDamping=true
    simulationSpace=RootRelative
    alphaBoolBlend {blendOption=Linear}
    boundBone {boneName=dyn_hair_L_A_05}
    componentLinearAccScale=<1 1 1>
    componentLinearVelScale=<1 1 1>
    physicsBodyDefinitions [{
      sphereCollisionRadius=0
      boundBone {boneName=dyn_hair_L_A_05}
      boxExtents=<10 20 20>
      constraintSetup {angularLimitsMax=<45 45 45>, angularLimitsMin=<-45 -45 -45>, angularTarget=<0 0 0>}
      localJointOffset=<1 0 -0.2>
     }]
   }
   n61 = Anim Dynamics - Bone: dyn_hair_L_A_06(ComponentPose=@n60, Alpha=1, ExternalForce="0,0,0") {
    angularDampingOverride=0.9999
    angularSpringConstant=75
    bAngularSpring=true
    bGravityOverrideInSimSpace=true
    bOverrideAngularDamping=true
    simulationSpace=RootRelative
    alphaBoolBlend {blendOption=Linear}
    boundBone {boneName=dyn_hair_L_A_06}
    componentLinearAccScale=<1 1 1>
    componentLinearVelScale=<1 1 1>
    physicsBodyDefinitions [{
      sphereCollisionRadius=0
      boundBone {boneName=dyn_hair_L_A_06}
      boxExtents=<10 20 20>
      constraintSetup {angularLimitsMax=<45 45 45>, angularLimitsMin=<-45 -45 -45>, angularTarget=<0 0 0>}
      localJointOffset=<1 0 -0.2>
     }]
   }
   n62 = Anim Dynamics - Bone: dyn_hair_L_A_07(ComponentPose=@n61, Alpha=1, ExternalForce="0,0,0") {
    angularDampingOverride=0.9999
    angularSpringConstant=75
    bAngularSpring=true
    bGravityOverrideInSimSpace=true
    bOverrideAngularDamping=true
    simulationSpace=RootRelative
    alphaBoolBlend {blendOption=Linear}
    boundBone {boneName=dyn_hair_L_A_07}
    componentLinearAccScale=<1 1 1>
    componentLinearVelScale=<1 1 1>
    physicsBodyDefinitions [{
      sphereCollisionRadius=0
      boundBone {boneName=dyn_hair_L_A_07}
      boxExtents=<10 20 20>
      constraintSetup {angularLimitsMax=<45 45 45>, angularLimitsMin=<-45 -45 -45>, angularTarget=<0 0 0>}
      localJointOffset=<1 0 -0.2>
     }]
   }
   n63 = "Constraint - dyn_hair_R_A_01 (P)"(ComponentPose=@n62, Alpha=0.25, ConstraintWeights_0=1) {
    alphaBoolBlend {blendOption=Linear}
    boneToModify {boneName=dyn_hair_R_A_01}
    constraintSetup [{targetBone {boneName=head}}]
    constraintWeights [1]
   }
   n64 = "Constraint - dyn_hair_R_A_02 (P)"(ComponentPose=@n63, Alpha=0.1, ConstraintWeights_0=1) {
    alphaBoolBlend {blendOption=Linear}
    boneToModify {boneName=dyn_hair_R_A_02}
    constraintSetup [{targetBone {boneName=head}}]
    constraintWeights [1]
   }
   n65 = "Constraint - dyn_hair_R_A_03 (P)"(ComponentPose=@n64, Alpha=0.1, ConstraintWeights_0=1) {
    alphaBoolBlend {blendOption=Linear}
    boneToModify {boneName=dyn_hair_R_A_03}
    constraintSetup [{targetBone {boneName=head}}]
    constraintWeights [1]
   }
   n66 = "Constraint - dyn_hair_R_A_04 (P)"(ComponentPose=@n65, Alpha=0, ConstraintWeights_0=1) {
    alphaBoolBlend {blendOption=Linear}
    boneToModify {boneName=dyn_hair_R_A_04}
    constraintSetup [{targetBone {boneName=head}}]
    constraintWeights [1]
   }
   n67 = "Constraint - dyn_hair_R_A_05 (P)"(ComponentPose=@n66, Alpha=0, ConstraintWeights_0=1) {
    alphaBoolBlend {blendOption=Linear}
    boneToModify {boneName=dyn_hair_R_A_05}
    constraintSetup [{targetBone {boneName=head}}]
    constraintWeights [1]
   }
   n68 = "Constraint - dyn_hair_L_A_02 (P)"(ComponentPose=@n67, Alpha=0.65, ConstraintWeights_0=1) {
    alphaBoolBlend {blendOption=Linear}
    boneToModify {boneName=dyn_hair_L_A_02}
    constraintSetup [{targetBone {boneName=head}}]
    constraintWeights [1]
   }
   n69 = "Constraint - dyn_hair_L_A_03 (P)"(ComponentPose=@n68, Alpha=0.1, ConstraintWeights_0=1) {
    alphaBoolBlend {blendOption=Linear}
    boneToModify {boneName=dyn_hair_L_A_03}
    constraintSetup [{targetBone {boneName=head}}]
    constraintWeights [1]
   }
   n70 = "Constraint - dyn_hair_L_A_04 (P)"(ComponentPose=@n69, Alpha=0.1, ConstraintWeights_0=1) {
    alphaBoolBlend {blendOption=Linear}
    boneToModify {boneName=dyn_hair_L_A_04}
    constraintSetup [{targetBone {boneName=head}}]
    constraintWeights [1]
   }
   n71 = "Constraint - dyn_hair_L_A_05 (P)"(ComponentPose=@n70, Alpha=0, ConstraintWeights_0=1) {
    alphaBoolBlend {blendOption=Linear}
    boneToModify {boneName=dyn_hair_L_A_05}
    constraintSetup [{targetBone {boneName=head}}]
    constraintWeights [1]
   }
   n72 = "Constraint - dyn_hair_L_A_06 (P)"(ComponentPose=@n71, Alpha=0, ConstraintWeights_0=1) {
    alphaBoolBlend {blendOption=Linear}
    boneToModify {boneName=dyn_hair_L_A_06}
    constraintSetup [{targetBone {boneName=head}}]
    constraintWeights [1]
   }
   n73 = "Constraint - dyn_hair_L_A_07 (P)"(ComponentPose=@n72, Alpha=0.1, ConstraintWeights_0=1) {
    alphaBoolBlend {blendOption=Linear}
    boneToModify {boneName=dyn_hair_L_A_07}
    constraintSetup [{targetBone {boneName=head}}]
    constraintWeights [1]
   }
   n74 = "Constraint - dyn_hairShoulder_L_A_02 (P)"(ComponentPose=@n73, Alpha=1, ConstraintWeights_0=1) {
    alphaBoolBlend {blendOption=Linear}
    boneToModify {boneName=dyn_hairShoulder_L_A_02}
    constraintSetup [{targetBone {boneName=spine_05}}]
    constraintWeights [1]
   }
   n75 = "Constraint - dyn_hairShoulder_R_A_03 (P)"(ComponentPose=@n74, Alpha=1, ConstraintWeights_0=1) {
    alphaBoolBlend {blendOption=Linear}
    boneToModify {boneName=dyn_hairShoulder_R_A_03}
    constraintSetup [{targetBone {boneName=spine_05}}]
    constraintWeights [1]
   }
   n76 = "Constraint - dyn_hairShoulder_R_B_01 (P)"(ComponentPose=@n75, Alpha=0.7, ConstraintWeights_0=1) {
    alphaBoolBlend {blendOption=Linear}
    boneToModify {boneName=dyn_hairShoulder_R_B_01}
    constraintSetup [{targetBone {boneName=spine_05}}]
    constraintWeights [1]
   }
   n77 = "Constraint - dyn_hairShoulder_R_B_02 (P)"(ComponentPose=@n76, Alpha=0.5, ConstraintWeights_0=1) {
    alphaBoolBlend {blendOption=Linear}
    boneToModify {boneName=dyn_hairShoulder_R_B_02}
    constraintSetup [{targetBone {boneName=spine_05}}]
    constraintWeights [1]
   }
   n78 = "Constraint - dyn_hairShoulder_R_B_03 (P)"(ComponentPose=@n77, Alpha=0.5, ConstraintWeights_0=1) {
    alphaBoolBlend {blendOption=Linear}
    boneToModify {boneName=dyn_hairShoulder_R_B_03}
    constraintSetup [{targetBone {boneName=spine_05}}]
    constraintWeights [1]
   }
   n79 = "Constraint - dyn_hairBack_B_01 (P)"(ComponentPose=@n78, Alpha=0.2, ConstraintWeights_0=1) {
    alphaBoolBlend {blendOption=Linear}
    boneToModify {boneName=dyn_hairBack_B_01}
    constraintSetup [{targetBone {boneName=spine_05}}]
    constraintWeights [1]
   }
   n80 = "Constraint - dyn_hairBack_C_01 (P)"(ComponentPose=@n79, Alpha=0.2, ConstraintWeights_0=1) {
    alphaBoolBlend {blendOption=Linear}
    boneToModify {boneName=dyn_hairBack_C_01}
    constraintSetup [{targetBone {boneName=spine_05}}]
    constraintWeights [1]
   }
   n81 = "Constraint - dyn_hairBack_A_01 (P)"(ComponentPose=@n80, Alpha=0.3, ConstraintWeights_0=1) {
    alphaBoolBlend {blendOption=Linear}
    boneToModify {boneName=dyn_hairBack_A_01}
    constraintSetup [{targetBone {boneName=spine_05}}]
    constraintWeights [1]
   }
   n82 = "Constraint - dyn_hairBack_B_02 (P)"(ComponentPose=@n81, Alpha=0.3, ConstraintWeights_0=1) {
    alphaBoolBlend {blendOption=Linear}
    boneToModify {boneName=dyn_hairBack_B_02}
    constraintSetup [{targetBone {boneName=spine_05}}]
    constraintWeights [1]
   }
   n83 = "Constraint - dyn_hairBack_C_02 (P)"(ComponentPose=@n82, Alpha=0.3, ConstraintWeights_0=1) {
    alphaBoolBlend {blendOption=Linear}
    boneToModify {boneName=dyn_hairBack_C_02}
    constraintSetup [{targetBone {boneName=spine_05}}]
    constraintWeights [1]
   }
   n84 = "Constraint - dyn_hairBack_D_01 (P)"(ComponentPose=@n83, Alpha=0.3, ConstraintWeights_0=1) {
    alphaBoolBlend {blendOption=Linear}
    boneToModify {boneName=dyn_hairBack_D_01}
    constraintSetup [{targetBone {boneName=spine_05}}]
    constraintWeights [1]
   }
   n85 = "Constraint - dyn_hairBack_A_02 (P)"(ComponentPose=@n84, Alpha=0.4, ConstraintWeights_0=1) {
    alphaBoolBlend {blendOption=Linear}
    boneToModify {boneName=dyn_hairBack_A_02}
    constraintSetup [{targetBone {boneName=spine_05}}]
    constraintWeights [1]
   }
   n86 = "Constraint - dyn_hairBack_B_03 (P)"(ComponentPose=@n85, Alpha=0.4, ConstraintWeights_0=1) {
    alphaBoolBlend {blendOption=Linear}
    boneToModify {boneName=dyn_hairBack_B_03}
    constraintSetup [{targetBone {boneName=spine_05}}]
    constraintWeights [1]
   }
   n87 = "Constraint - dyn_hairBack_C_03 (P)"(ComponentPose=@n86, Alpha=0.4, ConstraintWeights_0=1) {
    alphaBoolBlend {blendOption=Linear}
    boneToModify {boneName=dyn_hairBack_C_03}
    constraintSetup [{targetBone {boneName=spine_05}}]
    constraintWeights [1]
   }
   n88 = "Constraint - dyn_hairBack_D_02 (P)"(ComponentPose=@n87, Alpha=0.4, ConstraintWeights_0=1) {
    alphaBoolBlend {blendOption=Linear}
    boneToModify {boneName=dyn_hairBack_D_02}
    constraintSetup [{targetBone {boneName=spine_05}}]
    constraintWeights [1]
   }
   n89 = "Constraint - dyn_hairBack_A_03 (P)"(ComponentPose=@n88, Alpha=0.4, ConstraintWeights_0=1) {
    alphaBoolBlend {blendOption=Linear}
    boneToModify {boneName=dyn_hairBack_A_03}
    constraintSetup [{targetBone {boneName=spine_05}}]
    constraintWeights [1]
   }
   n90 = "Constraint - dyn_hairBack_B_04 (P)"(ComponentPose=@n89, Alpha=0.4, ConstraintWeights_0=1) {
    alphaBoolBlend {blendOption=Linear}
    boneToModify {boneName=dyn_hairBack_B_04}
    constraintSetup [{targetBone {boneName=spine_05}}]
    constraintWeights [1]
   }
   n91 = "Constraint - dyn_hairBack_C_04 (P)"(ComponentPose=@n90, Alpha=0.4, ConstraintWeights_0=1) {
    alphaBoolBlend {blendOption=Linear}
    boneToModify {boneName=dyn_hairBack_C_04}
    constraintSetup [{targetBone {boneName=spine_05}}]
    constraintWeights [1]
   }
   n92 = "Constraint - dyn_hairBack_D_03 (P)"(ComponentPose=@n91, Alpha=0.4, ConstraintWeights_0=1) {
    alphaBoolBlend {blendOption=Linear}
    boneToModify {boneName=dyn_hairBack_D_03}
    constraintSetup [{targetBone {boneName=spine_05}}]
    constraintWeights [1]
   }
   n93 = "Constraint - dyn_hairBack_A_04 (P)"(ComponentPose=@n92, Alpha=0, ConstraintWeights_0=1) {
    alphaBoolBlend {blendOption=Linear}
    boneToModify {boneName=dyn_hairBack_A_04}
    constraintSetup [{targetBone {boneName=spine_05}}]
    constraintWeights [1]
   }
   n94 = "Constraint - dyn_hairBack_B_05 (P)"(ComponentPose=@n93, Alpha=0, ConstraintWeights_0=1) {
    alphaBoolBlend {blendOption=Linear}
    boneToModify {boneName=dyn_hairBack_B_05}
    constraintSetup [{targetBone {boneName=spine_05}}]
    constraintWeights [1]
   }
   n95 = "Constraint - dyn_hairBack_C_05 (P)"(ComponentPose=@n94, Alpha=0, ConstraintWeights_0=1) {
    alphaBoolBlend {blendOption=Linear}
    boneToModify {boneName=dyn_hairBack_C_05}
    constraintSetup [{targetBone {boneName=spine_05}}]
    constraintWeights [1]
   }
   n96 = "Constraint - dyn_hairBack_D_04 (P)"(ComponentPose=@n95, Alpha=0, ConstraintWeights_0=1) {
    alphaBoolBlend {blendOption=Linear}
    boneToModify {boneName=dyn_hairBack_D_04}
    constraintSetup [{targetBone {boneName=spine_05}}]
    constraintWeights [1]
   }
   n97 = OR Boolean(A=Get bIsJumping(), B=Get bIsFalling())
   n98 = "Transform (Modify) Bone - Bone: dyn_hairBack_A_02"(ComponentPose=@n96, bAlphaBoolEnabled=@n97, Translation="0,0,0", Rotation="0,0,0", Scale="1,1,1") {
    alphaInputType=Bool
    rotationMode=BMM_Additive
    scaleMode=BMM_Additive
    translationMode=BMM_Additive
    alphaBoolBlend {blendInTime=0.2, blendOption=Linear, blendOutTime=0.2}
    boneToModify {boneName=dyn_hairBack_A_02}
   }
   n99 = "Transform (Modify) Bone - Bone: dyn_hairBack_A_03"!(ComponentPose=@n98, bAlphaBoolEnabled:bool=@n97, Translation="0,0,0", Rotation="0.000,0.000,10", Scale="1,1,1") -> (Pose) {
    alphaInputType=Bool
    rotationMode=BMM_Additive
    scaleMode=BMM_Additive
    translationMode=BMM_Additive
    alphaBoolBlend {blendInTime=4, blendOption=Linear, blendOutTime=4}
    boneToModify {boneName=dyn_hairBack_A_03}
    rotation {roll=10}
   }
   n100 = "Transform (Modify) Bone - Bone: dyn_hairBack_A_04"!(ComponentPose=@n99, bAlphaBoolEnabled:bool=@n97, Translation="0,0,0", Rotation="0.000,0.000,10", Scale="1,1,1") -> (Pose) {
    alphaInputType=Bool
    rotationMode=BMM_Additive
    scaleMode=BMM_Additive
    translationMode=BMM_Additive
    alphaBoolBlend {blendInTime=4, blendOption=Linear, blendOutTime=4}
    boneToModify {boneName=dyn_hairBack_A_04}
    rotation {roll=10}
   }
   n101 = "Transform (Modify) Bone - Bone: dyn_hairBack_B_02"(ComponentPose=@n100, bAlphaBoolEnabled=@n97, Translation="0,0,0", Rotation="0.000,0.000,15", Scale="1,1,1") {
    alphaInputType=Bool
    rotationMode=BMM_Additive
    scaleMode=BMM_Additive
    translationMode=BMM_Additive
    alphaBoolBlend {blendInTime=0.2, blendOption=Linear, blendOutTime=0.2}
    boneToModify {boneName=dyn_hairBack_B_02}
    rotation {roll=15}
   }
   n102 = "Transform (Modify) Bone - Bone: dyn_hairBack_B_03"!(ComponentPose=@n101, bAlphaBoolEnabled:bool=@n97, Translation="0,0,0", Rotation="0.000,0.000,10", Scale="1,1,1") -> (Pose) {
    alphaInputType=Bool
    rotationMode=BMM_Additive
    scaleMode=BMM_Additive
    translationMode=BMM_Additive
    alphaBoolBlend {blendInTime=4, blendOption=Linear, blendOutTime=4}
    boneToModify {boneName=dyn_hairBack_B_03}
    rotation {roll=10}
   }
   n103 = "Transform (Modify) Bone - Bone: dyn_hairBack_B_04"!(ComponentPose=@n102, bAlphaBoolEnabled:bool=@n97, Translation="0,0,0", Rotation="0.000,0.000,20", Scale="1,1,1") -> (Pose) {
    alphaInputType=Bool
    rotationMode=BMM_Additive
    scaleMode=BMM_Additive
    translationMode=BMM_Additive
    alphaBoolBlend {blendInTime=4, blendOption=Linear, blendOutTime=4}
    boneToModify {boneName=dyn_hairBack_B_04}
    rotation {roll=20}
   }
   n104 = "Transform (Modify) Bone - Bone: dyn_hairBack_B_05"!(ComponentPose=@n103, bAlphaBoolEnabled:bool=@n97, Translation="0,0,0", Rotation="0.000,0.000,10", Scale="1,1,1") -> (Pose) {
    alphaInputType=Bool
    rotationMode=BMM_Additive
    scaleMode=BMM_Additive
    translationMode=BMM_Additive
    alphaBoolBlend {blendInTime=4, blendOption=Linear, blendOutTime=4}
    boneToModify {boneName=dyn_hairBack_B_05}
    rotation {roll=10}
   }
   n105 = "Transform (Modify) Bone - Bone: dyn_hairBack_C_02"(ComponentPose=@n104, bAlphaBoolEnabled=@n97, Translation="0,0,0", Rotation="0.000,0.000,15", Scale="1,1,1") {
    alphaInputType=Bool
    rotationMode=BMM_Additive
    scaleMode=BMM_Additive
    translationMode=BMM_Additive
    alphaBoolBlend {blendInTime=0.2, blendOption=Linear, blendOutTime=0.2}
    boneToModify {boneName=dyn_hairBack_C_02}
    rotation {roll=15}
   }
   n106 = "Transform (Modify) Bone - Bone: dyn_hairBack_C_03"!(ComponentPose=@n105, bAlphaBoolEnabled:bool=@n97, Translation="0,0,0", Rotation="0.000,0.000,10", Scale="1,1,1") -> (Pose) {
    alphaInputType=Bool
    rotationMode=BMM_Additive
    scaleMode=BMM_Additive
    translationMode=BMM_Additive
    alphaBoolBlend {blendInTime=4, blendOption=Linear, blendOutTime=4}
    boneToModify {boneName=dyn_hairBack_C_03}
    rotation {roll=10}
   }
   n107 = "Transform (Modify) Bone - Bone: dyn_hairBack_C_04"!(ComponentPose=@n106, bAlphaBoolEnabled:bool=@n97, Translation="0,0,0", Rotation="0.000,0.000,20", Scale="1,1,1") -> (Pose) {
    alphaInputType=Bool
    rotationMode=BMM_Additive
    scaleMode=BMM_Additive
    translationMode=BMM_Additive
    alphaBoolBlend {blendInTime=4, blendOption=Linear, blendOutTime=4}
    boneToModify {boneName=dyn_hairBack_C_04}
    rotation {roll=20}
   }
   n108 = "Transform (Modify) Bone - Bone: dyn_hairBack_C_05"!(ComponentPose=@n107, bAlphaBoolEnabled:bool=@n97, Translation="0,0,0", Rotation="0.000,0.000,10", Scale="1,1,1") -> (Pose) {
    alphaInputType=Bool
    rotationMode=BMM_Additive
    scaleMode=BMM_Additive
    translationMode=BMM_Additive
    alphaBoolBlend {blendInTime=4, blendOption=Linear, blendOutTime=4}
    boneToModify {boneName=dyn_hairBack_C_05}
    rotation {roll=10}
   }
   n109 = "Transform (Modify) Bone - Bone: dyn_hairBack_D_02"!(ComponentPose=@n108, bAlphaBoolEnabled:bool=@n97, Translation="0,0,0", Rotation="0,0,0", Scale="1,1,1") -> (Pose) {
    alphaInputType=Bool
    rotationMode=BMM_Additive
    scaleMode=BMM_Additive
    translationMode=BMM_Additive
    alphaBoolBlend {blendInTime=0.2, blendOption=Linear, blendOutTime=0.2}
    boneToModify {boneName=dyn_hairBack_D_02}
   }
   n110 = "Transform (Modify) Bone - Bone: dyn_hairBack_D_03"!(ComponentPose=@n109, bAlphaBoolEnabled:bool=@n97, Translation="0,0,0", Rotation="0.000,0.000,10", Scale="1,1,1") -> (Pose) {
    alphaInputType=Bool
    rotationMode=BMM_Additive
    scaleMode=BMM_Additive
    translationMode=BMM_Additive
    alphaBoolBlend {blendInTime=4, blendOption=Linear, blendOutTime=4}
    boneToModify {boneName=dyn_hairBack_D_03}
    rotation {roll=10}
   }
   n111 = "Transform (Modify) Bone - Bone: dyn_hairBack_D_04"!(ComponentPose=@n110, bAlphaBoolEnabled:bool=@n97, Translation="0,0,0", Rotation="0.000,0.000,10", Scale="1,1,1") -> (Pose) {
    alphaInputType=Bool
    rotationMode=BMM_Additive
    scaleMode=BMM_Additive
    translationMode=BMM_Additive
    alphaBoolBlend {blendInTime=4, blendOption=Linear, blendOutTime=4}
    boneToModify {boneName=dyn_hairBack_D_04}
    rotation {roll=10}
   }
   n112 = Anim Dynamics - Bone: dyn_earing_L(ComponentPose=@n111, Alpha=1, ExternalForce="0,0,0") {
    angularDampingOverride=0.9999
    bAngularSpring=true
    bGravityOverrideInSimSpace=true
    bOverrideAngularDamping=true
    simulationSpace=RootRelative
    alphaBoolBlend {blendOption=Linear}
    boundBone {boneName=dyn_earing_L}
    componentLinearAccScale=<2 2 2>
    componentLinearVelScale=<2 2 2>
    physicsBodyDefinitions [{
      sphereCollisionRadius=0
      boundBone {boneName=dyn_earing_L}
      boxExtents=<15 15 8>
      constraintSetup {angularLimitsMax=<30 30 30>, angularLimitsMin=<-30 -30 -30>, angularTarget=<0 0 0>}
      localJointOffset=<2 0 0>
     }]
   }
   n113 = Anim Dynamics - Bone: dyn_earing_R(ComponentPose=@n112, Alpha=1, ExternalForce="0,0,0") {
    angularDampingOverride=0.9999
    bAngularSpring=true
    bGravityOverrideInSimSpace=true
    bOverrideAngularDamping=true
    simulationSpace=RootRelative
    alphaBoolBlend {blendOption=Linear}
    boundBone {boneName=dyn_earing_R}
    componentLinearAccScale=<2 2 2>
    componentLinearVelScale=<2 2 2>
    physicsBodyDefinitions [{
      sphereCollisionRadius=0
      boundBone {boneName=dyn_earing_R}
      boxExtents=<15 15 8>
      constraintSetup {angularLimitsMax=<30 30 30>, angularLimitsMin=<-30 -30 -30>, angularTarget=<0 0 0>}
      localJointOffset=<2 0 0>
     }]
   }
   n114 = "Transform (Modify) Bone - Bone: dyn_hairBack_A_MSTR"!(ComponentPose=@n113, bAlphaBoolEnabled:bool=True, Translation="0,-5,0", Rotation="0.000,0.000,30", Scale="1,1,1") -> (Pose) {
    alphaInputType=Bool
    bAlphaBoolEnabled=false
    rotationMode=BMM_Additive
    scaleMode=BMM_Additive
    scaleSpace=BCS_BoneSpace
    translationMode=BMM_Additive
    translationSpace=BCS_BoneSpace
    alphaBoolBlend {blendInTime=0.1, blendOption=Linear, blendOutTime=0.1}
    boneToModify {boneName=dyn_hairBack_A_MSTR}
    rotation {roll=30}
    translation {y=-5}
   }
   n115 = "Transform (Modify) Bone - Bone: dyn_hairBack_B_MSTR"!(ComponentPose=@n114, bAlphaBoolEnabled:bool=True, Translation="0,-5,0", Rotation="0,0,30", Scale="1,1,1") -> (Pose) {
    alphaInputType=Bool
    bAlphaBoolEnabled=false
    rotationMode=BMM_Additive
    scaleMode=BMM_Additive
    scaleSpace=BCS_BoneSpace
    translationMode=BMM_Additive
    translationSpace=BCS_BoneSpace
    alphaBoolBlend {blendInTime=0.1, blendOption=Linear, blendOutTime=0.1}
    boneToModify {boneName=dyn_hairBack_B_MSTR}
    rotation {roll=30}
    translation {y=-5}
   }
   n116 = "Transform (Modify) Bone - Bone: dyn_hairBack_C_MSTR"!(ComponentPose=@n115, bAlphaBoolEnabled:bool=True, Translation="0,-5,0", Rotation="0,0,30", Scale="1,1,1") -> (Pose) {
    alphaInputType=Bool
    bAlphaBoolEnabled=false
    rotationMode=BMM_Additive
    scaleMode=BMM_Additive
    scaleSpace=BCS_BoneSpace
    translationMode=BMM_Additive
    translationSpace=BCS_BoneSpace
    alphaBoolBlend {blendInTime=0.1, blendOption=Linear, blendOutTime=0.1}
    boneToModify {boneName=dyn_hairBack_C_MSTR}
    rotation {roll=30}
    translation {y=-5}
   }
   n117 = "Transform (Modify) Bone - Bone: dyn_hairBack_D_MSTR"!(ComponentPose=@n116, bAlphaBoolEnabled:bool=True, Translation="0,-5,0", Rotation="0,0,30", Scale="1,1,1") -> (Pose) {
    alphaInputType=Bool
    bAlphaBoolEnabled=false
    rotationMode=BMM_Additive
    scaleMode=BMM_Additive
    scaleSpace=BCS_BoneSpace
    translationMode=BMM_Additive
    translationSpace=BCS_BoneSpace
    alphaBoolBlend {blendInTime=0.1, blendOption=Linear, blendOutTime=0.1}
    boneToModify {boneName=dyn_hairBack_D_MSTR}
    rotation {roll=30}
    translation {y=-5}
   }
   n118 = "Transform (Modify) Bone - Bone: dyn_hairBack_A_MSTR"(ComponentPose=@n117, bAlphaBoolEnabled=True, Translation="0,5,0", Rotation="0.000,0.000,-30", Scale="1,1,1") {
    alphaInputType=Bool
    bAlphaBoolEnabled=false
    rotationMode=BMM_Additive
    scaleMode=BMM_Additive
    scaleSpace=BCS_BoneSpace
    translationMode=BMM_Additive
    translationSpace=BCS_BoneSpace
    alphaBoolBlend {blendInTime=0.1, blendOption=Linear, blendOutTime=0.1}
    boneToModify {boneName=dyn_hairBack_A_MSTR}
    rotation {roll=-30}
    translation {y=5}
   }
   n119 = "Transform (Modify) Bone - Bone: dyn_hairBack_B_MSTR"(ComponentPose=@n118, bAlphaBoolEnabled=True, Translation="0,5,0", Rotation="0,0,-30", Scale="1,1,1") {
    alphaInputType=Bool
    bAlphaBoolEnabled=false
    rotationMode=BMM_Additive
    scaleMode=BMM_Additive
    scaleSpace=BCS_BoneSpace
    translationMode=BMM_Additive
    translationSpace=BCS_BoneSpace
    alphaBoolBlend {blendInTime=0.1, blendOption=Linear, blendOutTime=0.1}
    boneToModify {boneName=dyn_hairBack_B_MSTR}
    rotation {roll=-30}
    translation {y=5}
   }
   n120 = "Transform (Modify) Bone - Bone: dyn_hairBack_C_MSTR"(ComponentPose=@n119, bAlphaBoolEnabled=True, Translation="0,5,0", Rotation="0,0,-30", Scale="1,1,1") {
    alphaInputType=Bool
    bAlphaBoolEnabled=false
    rotationMode=BMM_Additive
    scaleMode=BMM_Additive
    scaleSpace=BCS_BoneSpace
    translationMode=BMM_Additive
    translationSpace=BCS_BoneSpace
    alphaBoolBlend {blendInTime=0.1, blendOption=Linear, blendOutTime=0.1}
    boneToModify {boneName=dyn_hairBack_C_MSTR}
    rotation {roll=-30}
    translation {y=5}
   }
   n121 = "Transform (Modify) Bone - Bone: dyn_hairBack_D_MSTR"(ComponentPose=@n120, bAlphaBoolEnabled=True, Translation="0,5,0", Rotation="0,0,-30", Scale="1,1,1") {
    alphaInputType=Bool
    bAlphaBoolEnabled=false
    rotationMode=BMM_Additive
    scaleMode=BMM_Additive
    scaleSpace=BCS_BoneSpace
    translationMode=BMM_Additive
    translationSpace=BCS_BoneSpace
    alphaBoolBlend {blendInTime=0.1, blendOption=Linear, blendOutTime=0.1}
    boneToModify {boneName=dyn_hairBack_D_MSTR}
    rotation {roll=-30}
    translation {y=5}
   }
   Output Pose(Result=Component To Local(ComponentPose=@n121)) {layerGroup=None}
  ]}
 EventGraph {nodes [
   RifleSwitch() {is=title K2Node_CustomEvent}
   > ResetDynamics(InTeleportType=ResetPhysics)
   Event BlueprintUpdateAnimation()
   > n1 = Sequence()
   n2 = Gravity Override State Function DEPRECATED(
    execute=@n1.then_1
    CharacterPartAnimInstance=Self-Reference()
    GameTimeInSeconds=GetGameTimeInSeconds()
    GravityOverrideParams=Get Gravity Override Params()
   )
   > Set GravityOverride(GravityOverride=@n2)
   n3 = GetDeltaTransformFromRefPose(self=GetOwnerSkeletalMeshComponent(), BoneName=head, BaseName=spine_05)
   Set headUp(execute=@n1.then_0, headUp:double=BreakRotator(InRot=BreakTransform(InTransform=@n3).Rotation).Yaw) -> (Output_Get:double)
   Event BlueprintInitializeAnimation()
   > Set Body_Mesh(Body_Mesh=GetOwnerSkeletalMeshComponent())
   > n4 = Cast To PlayerPawn_Athena(Object=TryGetPawnOwner())
   > n5 = Set Pawn Owner(Pawn Owner:Player Pawn Athena Object Reference=@n4) -> (Output_Get:Player Pawn Athena Object Reference)
   > Set Head_Mesh(Head_Mesh:Skeletal Mesh Component Object Reference=GetSkeletalMeshForPartType(self=@n5, PartType=Head)) -> (Output_Get:Skeletal Mesh Component Object Reference)
   TryGetPawnOwner()
  ]}
]
variables [
 Body_Mesh {type=SkeletalMeshComponent}
 Gravity Override Params {
  type=GravityOverrideParamsStruct
  default {
   onGround_Moving {gravityOverride=<0 0 -1500>}
   onGround_Standing {gravityOverride=<0 0 -1500>}
   skydive_FloatingDown {gravityOverride=<0 0 600>, windAmplitude=<2000 -5000 4000>, windFrequency=<20 40 50>}
   skydive_Parachuting {gravityOverride=<0 -800 -1500>}
   swimming {gravityOverride=<0 0 0>}
  }
 }
 GravityOverride {type=Vector}
 Head_Mesh {type=SkeletalMeshComponent}
 headUp {type=number}
 Pawn Owner {type=PlayerPawn_Athena_C}
]
